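# Control Rig for the F_MED_AnodizedMetal female cosmetic body that correctively poses skirt, belt, charms, katana, masks, collar and shoulder pads as the body moves.
# Helper functions remap relative bone rotations into weight variables, joined by deformRig anim attributes, spherical pose readers and crouch, airborne and parachute states.
# Those weights blend parent constraints and additive transform offsets on base and dyn bones.
# A debug branch lets controls drive limb and neck bones.
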ControlRigBlueprint /BRCosmetics/Characters/Player/Female/Medium/Bodies/F_MED_AnodizedMetal/Meshes/F_MED_AnodizedMetal_CtrlRig
graphs [
 Multiply_3_1_ContainedGraph {nodes [
   n1 = RigVMFunctionEntryNode()
   RigVMFunctionReturnNode(Result=&RigVMFunction_MathDoubleMul(B=@n1.C, A=&RigVMFunction_MathDoubleMul(A=@n1.A, B=@n1.B).Result).Result)
  ]}
 New Function_ContainedGraph {nodes [
   n1 = RigVMFunctionEntryNode()
   RigVMFunction_ControlFlowBranch = &RigVMFunction_ControlFlowBranch(Condition=@n1.Debug Controls, ExecuteContext=@n1.ExecuteContext)
   RigVMFunctionReturnNode(ExecuteContext=@RigVMFunction_ControlFlowBranch.Completed)
   RigUnit_GetTransform = &RigUnit_GetTransform(Item=(Type=Control,Name="thigh_r_ctrl"), bInitial=False)
   &RigUnit_SetTransform(
    Item=(Type=Bone,Name="thigh_r")
    Weight=1
    bInitial=False
    Value=@RigUnit_GetTransform.Transform
    ExecutePin=@RigVMFunction_ControlFlowBranch.True
   )
   > &RigUnit_SetTransform(
    Item=(Type=Bone,Name="thigh_l")
    Weight=1
    bInitial=False
    Value=&RigUnit_GetTransform(Item=(Type=Control,Name="thigh_l_ctrl"), bInitial=False).Transform
   )
   > &RigUnit_SetTransform(
    Item=(Type=Bone,Name="calf_l")
    Weight=1
    bInitial=False
    Value=&RigUnit_GetTransform(Item=(Type=Control,Name="calf_l_ctrl"), bInitial=False).Transform
   )
   > &RigUnit_SetTransform(
    Item=(Type=Bone,Name="calf_r")
    Weight=1
    bInitial=False
    Value=&RigUnit_GetTransform(Item=(Type=Control,Name="calf_r_ctrl"), bInitial=False).Transform
   )
   > &RigUnit_SetTransform(
    Item=(Type=Bone,Name="upperarm_l")
    Weight=1
    bInitial=False
    Value=&RigUnit_GetTransform(Item=(Type=Control,Name="upperarm_l_ctrl"), bInitial=False).Transform
   )
   > &RigUnit_SetTransform(
    Item=(Type=Bone,Name="upperarm_r")
    Weight=1
    bInitial=False
    Value=&RigUnit_GetTransform(Item=(Type=Control,Name="upperarm_r_ctrl"), bInitial=False).Transform
   )
   > &RigUnit_SetTransform(
    Item=(Type=Bone,Name="neck_01")
    Weight=1
    bInitial=False
    Value=&RigUnit_GetTransform(Item=(Type=Control,Name="neck_01_ctrl"), bInitial=False).Transform
   )
  ]}
 New Function_ContainedGraph {nodes [
   n1 = RigVMFunctionEntryNode()
   RigVMFunction_MathQuaternionToEuler = &RigVMFunction_MathQuaternionToEuler(
    RotationOrder=XYZ
    Value=&RigUnit_GetRelativeTransformForItem(
     Child=(Type=Bone,Name="hand_l")
     Parent=(Type=Null,Name="lowerarm_l")
     bChildInitial=False
     bParentInitial=False
     Child.Name=@n1.Child
     Parent.Name=@n1.Parent
    ).RelativeTransform.Rotation
   )
   Remap = &RigVMFunction_MathDoubleRemap(
    TargetMaximum=1
    TargetMinimum=0
    Value=@RigVMFunction_MathQuaternionToEuler.Result.Y
    SourceMaximum=@n1.Source Maximum
    SourceMinimum=@n1.Source Minimum
   )
   > RigVMFunctionReturnNode(Delta Rotation=@Remap.Result)
  ]}
 New Function_ContainedGraph {nodes [
   n1 = RigVMFunctionEntryNode()
   RigVMFunction_MathQuaternionToEuler = &RigVMFunction_MathQuaternionToEuler(
    RotationOrder=XYZ
    Value=&RigUnit_GetRelativeTransformForItem(
     Child=(Type=Bone,Name="hand_l")
     Parent=(Type=Null,Name="lowerarm_l")
     bChildInitial=False
     bParentInitial=False
     Child.Name=@n1.Child
     Parent.Name=@n1.Parent
    ).RelativeTransform.Rotation
   )
   Remap = &RigVMFunction_MathDoubleRemap(
    TargetMaximum=1
    TargetMinimum=0
    Value=@RigVMFunction_MathQuaternionToEuler.Result.Z
    SourceMaximum=@n1.Source Maximum
    SourceMinimum=@n1.Source Minimum
   )
   > RigVMFunctionReturnNode(Delta Rotation=@Remap.Result)
  ]}
 New Function_ContainedGraph {nodes [
   n1 = RigVMFunctionEntryNode()
   RigVMFunction_MathTransformToEulerTransform = /Script/RigVM.RigVMFunction_MathTransformToEulerTransform(Value=&RigVMFunction_MathTransformMakeRelative(
     Global=&RigUnit_GetTransform(Space=LocalSpace, Item=@n1.Item).Transform
     Parent=&RigUnit_GetTransform(Space=LocalSpace, bInitial=true, Item=@n1.Item).Transform
    ).Local) {is=struct RigVMUnitNode, template="Cast::Execute(in Value,out Result)"}
   > RigVMFunctionReturnNode(OutTransform=@RigVMFunction_MathTransformToEulerTransform.Result)
  ]}
 New Function_ContainedGraph {nodes [
   RigVMFunctionEntryNode()
   > RigVMFunction_Sequence_1 = RigVMAggregateNode()
   &RigUnit_ParentConstraint(
    AdvancedSettings=(RotationOrderForFilter=XZY)
    Child=(Type=Bone,Name="base_skirt_01_l")
    Parents=((Item=(Type=Bone,Name="thigh_twist_01_l"),Weight=0.25),(Item=(Type=Bone,Name="pelvis"),Weight=0.75))
    Weight=1
    ExecutePin=@RigVMFunction_Sequence_1.A
   )
   > &RigUnit_ParentConstraint(
    AdvancedSettings=(RotationOrderForFilter=XZY)
    Child=(Type=Bone,Name="base_skirt_01_front_l")
    Filter=(TranslationFilter=(bX=false,bY=false,bZ=false),RotationFilter=(bX=True,bY=True,bZ=True),ScaleFilter=(bX=True,bY=True,bZ=True))
    Parents=((Item=(Type=Bone,Name="thigh_twist_01_l"),Weight=0.5),(Item=(Type=Bone,Name="pelvis"),Weight=0.5))
    Weight=1
   )
   > &RigUnit_ModifyTransforms(
    ItemToModify="((Item=(Type=Bone,Name=\"base_skirt_01_front_l\"),Transform=(Rotation=(X=-0.1137,Y=-0.3859,Z=0.2031,W=0.8927),Translation=(X=0,Y=0,Z=3),Scale3D=(X=1,Y=1,Z=1))),(Item=(Type=Bone,Name=\"base_skirt_01_l\"),Transform=(Rotation=(X=0.0024,Y=-0.2138,Z=0.234,W=0.9484) ...(truncated)"
    ItemToModify.0.Transform.Rotation=&RigVMFunction_MathQuaternionFromEuler(Euler=(X=-10,Y=-45,Z=0)).Result
    Weight=RigVMDispatchNode(BoneName=root, Default=1, Name=deformRig_l_thigh_up_alpha) {template="DISPATCH_RigDispatch_GetAnimAttribute(in Name,in BoneName,in Default,out Value,out Found)"}.Value
   )
   n1 = RigVMVariableNode() {variable=Thigh_L_SideRotation}
   > &RigUnit_ParentConstraint(
    AdvancedSettings=(RotationOrderForFilter=XZY)
    Child=(Type=Bone,Name="base_skirt_01_l")
    Filter=(TranslationFilter=(bX=false,bY=false,bZ=false),RotationFilter=(bX=True,bY=True,bZ=True),ScaleFilter=(bX=True,bY=True,bZ=True))
    Parents=((Item=(Type=Bone,Name="thigh_twist_01_l"),Weight=0.8),(Item=(Type=Bone,Name="pelvis"),Weight=0.2))
    Weight=@n1.Value
   )
   > &RigUnit_ParentConstraint(
    AdvancedSettings=(RotationOrderForFilter=XZY)
    Child=(Type=Bone,Name="base_skirt_01_front_l")
    Filter=(TranslationFilter=(bX=false,bY=false,bZ=false),RotationFilter=(bX=True,bY=True,bZ=True),ScaleFilter=(bX=True,bY=True,bZ=True))
    Parents=((Item=(Type=Bone,Name="thigh_twist_01_l"),Weight=0.8),(Item=(Type=Bone,Name="pelvis"),Weight=0.2))
    Weight=@n1.Value
   )
   > &RigUnit_ModifyTransforms(
    ItemToModify="((Item=(Type=Bone,Name=\"base_skirt_01_l\"),Transform=(Rotation=(X=0.0026,Y=-0.1035,Z=-0.0251,W=0.9943),Translation=(X=0,Y=0,Z=0),Scale3D=(X=1,Y=1,Z=1))),(Item=(Type=Bone,Name=\"base_skirt_01_front_l\"),Transform=(Rotation=(X=-0.0058,Y=0.0018,Z=-0.0064,W=1),Tr ...(truncated)"
    Weight=@n1.Value
   )
   n2 = RigVMDispatchNode(BoneName=root, Default=1, Name=deformRig_r_thigh_up_alpha) {template="DISPATCH_RigDispatch_GetAnimAttribute(in Name,in BoneName,in Default,out Value,out Found)"}
   Remap = &RigVMFunction_MathDoubleRemap(SourceMaximum=1, SourceMinimum=0, TargetMaximum=1, TargetMinimum=0.4, bClamp=False, Value=@n2.Value)
   &RigUnit_ParentConstraint(
    AdvancedSettings=(RotationOrderForFilter=XZY)
    Child=(Type=Bone,Name="base_skirt_01_r")
    Parents=((Item=(Type=Bone,Name="thigh_twist_01_r"),Weight=0.25),(Item=(Type=Bone,Name="pelvis"),Weight=0.75))
    Weight=@Remap.Result
    ExecutePin=@RigVMFunction_Sequence_1.B
   )
   > &RigUnit_ParentConstraint(
    AdvancedSettings=(RotationOrderForFilter=XZY)
    Child=(Type=Bone,Name="base_skirt_01_front_r")
    Filter=(TranslationFilter=(bX=false,bY=false,bZ=false),RotationFilter=(bX=True,bY=True,bZ=True),ScaleFilter=(bX=True,bY=True,bZ=True))
    Parents=((Item=(Type=Bone,Name="thigh_twist_01_r"),Weight=0.6),(Item=(Type=Bone,Name="pelvis"),Weight=0.4))
    Weight=@Remap.Result
   )
   > &RigUnit_ModifyTransforms(
    ItemToModify="((Item=(Type=Bone,Name=\"base_skirt_01_front_r\"),Transform=(Rotation=(X=0,Y=-0.4226,Z=0,W=0.9063),Translation=(X=0.2329,Y=-0.6724,Z=-0.2055),Scale3D=(X=1,Y=1,Z=1))),(Item=(Type=Bone,Name=\"base_skirt_01_r\"),Transform=(Rotation=(X=0.0106,Y=-0.1813,Z=-0.0586,W ...(truncated)"
    Weight=@n2.Value
   )
   > &RigUnit_SphericalPoseReader(
    ActiveRegionScaleFactors=(PositiveWidth=5.5,NegativeWidth=4.5,PositiveHeight=2,NegativeHeight=2)
    ActiveRegionSize=0.1
    Debug=(bDrawDebug=false,bDraw2D=false,DebugScale=25,DebugSegments=20,DebugThickness=0.25)
    DriverAxis=(X=1,Y=0,Z=0)
    DriverItem=(Type=Bone,Name="thigh_r")
    FalloffRegionScaleFactors=(PositiveWidth=1,NegativeWidth=1,PositiveHeight=1,NegativeHeight=1)
    FalloffSize=0.7
    FlipHeightScaling=false
    FlipWidthScaling=true
    OptionalParentItem=(Type=Bone,Name="None")
    RotationOffset=(X=0,Y=50,Z=0)
   )
   n3 = RigVMVariableNode() {variable=Thigh_R_SideRotation}
   n4 = &RigVMFunction_MathDoubleRemap(SourceMaximum=1, SourceMinimum=0, TargetMaximum=1, TargetMinimum=0.4, bClamp=False, Value=@n3.Value)
   > &RigUnit_ParentConstraint(
    AdvancedSettings=(RotationOrderForFilter=XZY)
    Child=(Type=Bone,Name="base_skirt_01_r")
    Filter=(TranslationFilter=(bX=false,bY=false,bZ=false),RotationFilter=(bX=True,bY=True,bZ=True),ScaleFilter=(bX=True,bY=True,bZ=True))
    Parents=((Item=(Type=Bone,Name="thigh_twist_01_r"),Weight=0.8),(Item=(Type=Bone,Name="pelvis"),Weight=0.2))
    Weight=@n4.Result
   )
   > &RigUnit_ParentConstraint(
    AdvancedSettings=(RotationOrderForFilter=XZY)
    Child=(Type=Bone,Name="base_skirt_01_front_r")
    Filter=(TranslationFilter=(bX=false,bY=false,bZ=false),RotationFilter=(bX=True,bY=True,bZ=True),ScaleFilter=(bX=True,bY=True,bZ=True))
    Parents=((Item=(Type=Bone,Name="thigh_twist_01_r"),Weight=0.8),(Item=(Type=Bone,Name="pelvis"),Weight=0.2))
    Weight=@n4.Result
   )
   > &RigUnit_ModifyTransforms(
    ItemToModify="((Item=(Type=Bone,Name=\"base_skirt_01_r\"),Transform=(Rotation=(X=-0.0032,Y=-0.1054,Z=0.0306,W=0.994),Translation=(X=4,Y=0,Z=2),Scale3D=(X=1,Y=1,Z=1))),(Item=(Type=Bone,Name=\"base_skirt_01_front_r\"),Transform=(Rotation=(X=-0.01,Y=-0.1426,Z=0.0019,W=0.9897), ...(truncated)"
    Weight=@n3.Value
   )
   Delta Rotation Z = RigVMFunctionReferenceNode(Child=thigh_l, Parent=thigh_l_null, Source Maximum=-150, Source Minimum=-100, ExecuteContext=@RigVMFunction_Sequence_1.F)
   > &RigUnit_ModifyTransforms(
    ItemToModify="((Item=(Type=Bone,Name=\"dyn_skirt_01_fr\"),Transform=(Rotation=(X=-0.0062,Y=-0.1618,Z=0.0378,W=0.9861),Translation=(X=0,Y=0,Z=0),Scale3D=(X=1,Y=1,Z=1))),(Item=(Type=Bone,Name=\"dyn_skirt_01_front_l\"),Transform=(Rotation=(X=0.0583,Y=-0.1473,Z=-0.1469,W=0.9764 ...(truncated)"
    WeightMinimum=0
    Weight=&RigVMFunction_MathDoubleMul(A=@"Delta Rotation Z".Delta Rotation, B=RigVMVariableNode() {variable=isCrouching}.Value).Result
   )
   > Delta Rotation Z_1 = RigVMFunctionReferenceNode(Child=thigh_l, Parent=thigh_l_null, Source Maximum=-150, Source Minimum=-80)
   > &RigUnit_ModifyTransforms(
    ItemToModify="((Item=(Type=Bone,Name=\"dyn_charm_01_a\"),Transform=(Rotation=(X=0.6756,Y=-0.3017,Z=0.0122,W=0.6726),Translation=(X=0,Y=0,Z=0),Scale3D=(X=1,Y=1,Z=1))),(Item=(Type=Bone,Name=\"dyn_charm_01_b\"),Transform=(Rotation=(X=0.5558,Y=-0.3396,Z=0.0074,W=0.7587),Transla ...(truncated)"
    WeightMinimum=0
    Weight=&RigVMFunction_MathDoubleMul(A=@"Delta Rotation Z_1".Delta Rotation, B=RigVMVariableNode() {variable=isCrouching}.Value).Result
   )
   RigVMFunctionReturnNode(ExecuteContext=@RigVMFunction_Sequence_1.G)
   &RigUnit_SphericalPoseReader(
    ActiveRegionScaleFactors=(PositiveWidth=3.5,NegativeWidth=0.1,PositiveHeight=3,NegativeHeight=1)
    ActiveRegionSize=0.2
    Debug=(bDrawDebug=false,DebugScale=25,DebugSegments=20,DebugThickness=0.25)
    DriverAxis=(X=-1,Y=0,Z=0)
    DriverItem=(Type=Bone,Name="thigh_l")
    FalloffRegionScaleFactors=(PositiveWidth=1,NegativeWidth=1,PositiveHeight=1,NegativeHeight=10)
    FalloffSize=0.25
    FlipWidthScaling=false
    OptionalParentItem=(Type=Bone,Name="None")
    RotationOffset=(X=150,Y=100,Z=0)
    ExecutePin=@RigVMFunction_Sequence_1.C
   )
   n5 = RigVMDispatchNode(BoneName=root, Default=1, Name=deformRig_l_thigh_up_alpha) {template="DISPATCH_RigDispatch_GetAnimAttribute(in Name,in BoneName,in Default,out Value,out Found)"}
   > &RigUnit_ParentConstraint(
    AdvancedSettings=(RotationOrderForFilter=XZY)
    Child=(Type=Bone,Name="base_skirt_01_fr")
    Filter=(TranslationFilter=(bX=false,bY=false,bZ=false),RotationFilter=(bX=True,bY=True,bZ=True),ScaleFilter=(bX=True,bY=True,bZ=True))
    Parents=((Item=(Type=Bone,Name="thigh_twist_01_l"),Weight=0.7),(Item=(Type=Bone,Name="pelvis"),Weight=0.3))
    Weight=@n5.Value
   )
   > &RigUnit_ModifyTransforms(
    ItemToModify=((Item=(Type=Bone,Name="dyn_skirt_01_fr"),Transform=(Rotation=(X=0.0189,Y=-0.2156,Z=0.0851,W=0.9726),Translation=(X=0,Y=0,Z=0),Scale3D=(X=1,Y=1,Z=1))))
    Weight=@n5.Value
   )
   > &RigUnit_ModifyTransforms(
    ItemToModify="((Item=(Type=Bone,Name=\"dyn_skirt_01_fr\"),Transform=(Rotation=(X=0,Y=-0.0652,Z=0,W=0.9979),Translation=(X=0,Y=0,Z=0),Scale3D=(X=1,Y=1,Z=1))),(Item=(Type=Bone,Name=\"dyn_skirt_01_front_r\"),Transform=(Rotation=(X=-0.1604,Y=-0.1232,Z=0.034,W=0.9787),Translatio ...(truncated)"
    Weight=RigVMVariableNode() {variable=Thigh_L_In_SideRotation}.Value
   )
   &RigUnit_SphericalPoseReader(
    ActiveRegionScaleFactors=(PositiveWidth=3.5,NegativeWidth=0.1,PositiveHeight=3,NegativeHeight=1)
    ActiveRegionSize=0.2
    Debug=(bDrawDebug=false,DebugScale=25,DebugSegments=20,DebugThickness=0.25)
    DriverAxis=(X=1,Y=0,Z=0)
    DriverItem=(Type=Bone,Name="thigh_r")
    FalloffRegionScaleFactors=(PositiveWidth=1,NegativeWidth=1,PositiveHeight=1,NegativeHeight=10)
    FalloffSize=0.25
    FlipHeightScaling=false
    FlipWidthScaling=true
    OptionalParentItem=(Type=Bone,Name="None")
    RotationOffset=(X=-150,Y=85,Z=-120)
    ExecutePin=@RigVMFunction_Sequence_1.D
   )
   n6 = RigVMDispatchNode(BoneName=root, Default=1, Name=deformRig_r_thigh_up_alpha) {template="DISPATCH_RigDispatch_GetAnimAttribute(in Name,in BoneName,in Default,out Value,out Found)"}
   > &RigUnit_ParentConstraint(
    AdvancedSettings=(RotationOrderForFilter=XZY)
    Child=(Type=Bone,Name="base_skirt_01_fr")
    Filter=(TranslationFilter=(bX=false,bY=false,bZ=false),RotationFilter=(bX=True,bY=True,bZ=True),ScaleFilter=(bX=True,bY=True,bZ=True))
    Parents=((Item=(Type=Bone,Name="thigh_twist_01_r"),Weight=0.7),(Item=(Type=Bone,Name="pelvis"),Weight=0.3))
    Weight=@n6.Value
   )
   > &RigUnit_ModifyTransforms(
    ItemToModify=((Item=(Type=Bone,Name="base_skirt_01_fr"),Transform=(Rotation=(X=-0.0076,Y=-0.1735,Z=-0.043,W=0.9839),Translation=(X=0,Y=0,Z=0),Scale3D=(X=1,Y=1,Z=1))))
    Weight=@n6.Value
   )
   &RigUnit_ModifyTransforms(
    ItemToModify="((Item=(Type=Bone,Name=\"dyn_skirt_01_bck\"),Transform=(Rotation=(X=0,Y=-0.1459,Z=0,W=0.9893),Translation=(X=0,Y=0,Z=0),Scale3D=(X=1,Y=1,Z=1))),(Item=(Type=Bone,Name=\"dyn_skirt_01_bck_l\"),Transform=(Rotation=(X=0,Y=-0.0659,Z=0,W=0.9978),Translation=(X=0,Y=0, ...(truncated)"
    WeightMinimum=0
    Weight=RigVMVariableNode() {variable=PelvisToGround}.Value
    ExecutePin=@RigVMFunction_Sequence_1.E
   )
   > RigVMFunction_Sequence_2 = RigVMAggregateNode()
   &RigUnit_ModifyTransforms(
    ItemToModify="((Item=(Type=Bone,Name=\"dyn_skirt_01_bck\"),Transform=(Rotation=(X=0.1401,Y=-0.3003,Z=-0.0025,W=0.9435),Translation=(X=0,Y=0,Z=0),Scale3D=(X=1,Y=1,Z=1))),(Item=(Type=Bone,Name=\"dyn_skirt_01_bck_l\"),Transform=(Rotation=(X=0.0667,Y=-0.4634,Z=-0.1699,W=0.8671) ...(truncated)"
    WeightMinimum=0
    ExecutePin=@RigVMFunction_Sequence_2.A
    Weight=RigVMVariableNode() {variable=Thigh_L_BackRotation}.Value
   )
   Multiply = &RigVMFunction_MathDoubleMul(
    B=&RigVMFunction_MathDoubleRemap(
     SourceMaximum=1
     SourceMinimum=0.2
     TargetMaximum=0
     TargetMinimum=1
     bClamp=False
     Value=RigVMDispatchNode(BoneName=root, Default=1, Name=deformRig_l_thigh_up_alpha) {template="DISPATCH_RigDispatch_GetAnimAttribute(in Name,in BoneName,in Default,out Value,out Found)"}.Value
    ).Result
    A=RigVMVariableNode() {variable=Calf_L_Rotation}.Value
   )
   > &RigUnit_ModifyTransforms(
    ItemToModify="((Item=(Type=Bone,Name=\"dyn_skirt_02_bck\"),Transform=(Rotation=(X=-0.0167,Y=-0.3823,Z=-0.0403,W=0.923),Translation=(X=0,Y=0,Z=0),Scale3D=(X=1,Y=1,Z=1))),(Item=(Type=Bone,Name=\"dyn_skirt_02_bck_l\"),Transform=(Rotation=(X=0.0763,Y=-0.5722,Z=-0.1806,W=0.7963) ...(truncated)"
    WeightMinimum=0
    Weight=@Multiply.Result
   )
   &RigUnit_ModifyTransforms(
    ItemToModify="((Item=(Type=Bone,Name=\"dyn_skirt_01_bck\"),Transform=(Rotation=(X=-0.1445,Y=-0.2982,Z=0.0165,W=0.9434),Translation=(X=0,Y=0,Z=0),Scale3D=(X=1,Y=1,Z=1))),(Item=(Type=Bone,Name=\"dyn_skirt_01_bck_l\"),Transform=(Rotation=(X=0,Y=0,Z=0,W=1),Translation=(X=0,Y=0, ...(truncated)"
    WeightMinimum=0
    Weight=RigVMVariableNode() {variable=Thigh_R_BackRotation}.Value
    ExecutePin=@RigVMFunction_Sequence_2.B
   )
   n7 = &RigVMFunction_MathDoubleMul(
    B=&RigVMFunction_MathDoubleRemap(
     SourceMaximum=1
     SourceMinimum=0.2
     TargetMaximum=0
     TargetMinimum=1
     bClamp=False
     Value=RigVMDispatchNode(BoneName=root, Default=1, Name=deformRig_r_thigh_up_alpha) {template="DISPATCH_RigDispatch_GetAnimAttribute(in Name,in BoneName,in Default,out Value,out Found)"}.Value
    ).Result
    A=RigVMVariableNode() {variable=Calf_R_Rotation}.Value
   )
   > &RigUnit_ModifyTransforms(
    ItemToModify="((Item=(Type=Bone,Name=\"dyn_skirt_02_bck\"),Transform=(Rotation=(X=0.0184,Y=-0.4213,Z=0.0396,W=0.9059),Translation=(X=0,Y=0,Z=0),Scale3D=(X=1,Y=1,Z=1))),(Item=(Type=Bone,Name=\"dyn_skirt_02_bck_l\"),Transform=(Rotation=(X=-0.0045,Y=-0.1736,Z=-0.0258,W=0.9845) ...(truncated)"
    WeightMinimum=0
    Weight=@n7.Result
   )
   > &RigUnit_ModifyTransforms(
    ItemToModify=((Item=(Type=Bone,Name="base_katana_r"),Transform=(Rotation=(X=0,Y=0,Z=-0.0623,W=0.9981),Translation=(X=0.0634,Y=0.5053,Z=2.9586),Scale3D=(X=1,Y=1,Z=1))))
    WeightMinimum=0
    Weight=&RigVMFunction_MathDoubleMul(B=RigVMVariableNode() {variable=KatanaTransform}.Value, A=@n7.Result).Result
   )
   > &RigUnit_ModifyTransforms(
    ItemToModify="((Item=(Type=Bone,Name=\"dyn_skirt_02_bck\"),Transform=(Rotation=(X=0,Y=0.2222,Z=0,W=0.975),Translation=(X=0,Y=0,Z=0),Scale3D=(X=1,Y=1,Z=1))),(Item=(Type=Bone,Name=\"dyn_skirt_02_bck_l\"),Transform=(Rotation=(X=0,Y=0.2691,Z=0,W=0.9631),Translation=(X=0,Y=0,Z=0 ...(truncated)"
    WeightMinimum=0
    Weight=RigVMVariableNode() {variable=inFloating}.Value
   )
   &RigUnit_ModifyTransforms(
    ItemToModify="((Item=(Type=Bone,Name=\"dyn_skirt_01_fr\"),Transform=(Rotation=(X=0,Y=-0.0349,Z=0,W=0.9994),Translation=(X=0,Y=0,Z=0),Scale3D=(X=1,Y=1,Z=1))),(Item=(Type=Bone,Name=\"dyn_skirt_01_front_l\"),Transform=(Rotation=(X=-0.0002,Y=-0.0886,Z=-0.0061,W=0.996),Translati ...(truncated)"
    WeightMinimum=0
    Weight=RigVMVariableNode() {variable=SpineRotRight}.Value
    ExecutePin=@RigVMFunction_Sequence_2.C
   )
   > &RigUnit_ModifyTransforms(
    ItemToModify="((Item=(Type=Bone,Name=\"dyn_skirt_03_bck_l\"),Transform=(Rotation=(X=0,Y=-0.5011,Z=0,W=0.8654),Translation=(X=0,Y=0,Z=0),Scale3D=(X=1,Y=1,Z=1))),(Item=(Type=Bone,Name=\"dyn_skirt_03_bck\"),Transform=(Rotation=(X=0,Y=-0.4245,Z=0,W=0.9054),Translation=(X=0,Y=0, ...(truncated)"
    WeightMinimum=0
    Weight=&RigVMFunction_MathDoubleMul(B=RigVMVariableNode() {variable=skydiveThigBackL}.Value, A=RigVMVariableNode() {variable=isSkydiving}.Value).Result
   )
   > &RigUnit_ModifyTransforms(
    ItemToModify="((Item=(Type=Bone,Name=\"dyn_skirt_03_r\"),Transform=(Rotation=(X=0,Y=-0.0561,Z=0,W=0.9984),Translation=(X=0,Y=0,Z=0),Scale3D=(X=1,Y=1,Z=1))),(Item=(Type=Bone,Name=\"dyn_skirt_01_front_r\"),Transform=(Rotation=(X=0,Y=-0.3124,Z=0,W=0.95),Translation=(X=0,Y=0,Z= ...(truncated)"
    WeightMinimum=0
    Weight=&RigVMFunction_MathDoubleMul(A=RigVMVariableNode() {variable=isSkydiving}.Value, B=RigVMVariableNode() {variable=skydiveThigBackR}.Value).Result
   )
   Multiply_3 = RigVMAggregateNode(
    C=RigVMVariableNode() {variable=openParachute}.Value
    B=RigVMVariableNode() {variable=spineRotBack}.Value
    A=RigVMVariableNode() {variable=isSkydiving}.Value
   )
   > &RigUnit_ModifyTransforms(
    ItemToModify="((Item=(Type=Bone,Name=\"dyn_skirt_02_bck_l\"),Transform=(Rotation=(X=0,Y=0.3866,Z=0,W=0.9222),Translation=(X=0,Y=0,Z=0),Scale3D=(X=1,Y=1,Z=1))),(Item=(Type=Bone,Name=\"dyn_skirt_02_bck\"),Transform=(Rotation=(X=0,Y=0.4846,Z=0,W=0.8747),Translation=(X=0,Y=0,Z= ...(truncated)"
    WeightMinimum=0
    Weight=@Multiply_3.Result
   )
   RigVMFunction_MathBoolToFloat = &RigVMFunction_MathBoolToFloat(Value=RigVMVariableNode() {variable=IsInAir}.Value)
   > &RigUnit_ModifyTransforms(
    ItemToModify="((Item=(Type=Bone,Name=\"dyn_skirt_01_front_l\"),Transform=(Rotation=(X=0,Y=-0.0283,Z=0,W=0.9996),Translation=(X=0,Y=0,Z=0),Scale3D=(X=1,Y=1,Z=1))),(Item=(Type=Bone,Name=\"dyn_skirt_01_l\"),Transform=(Rotation=(X=0,Y=-0.0619,Z=0,W=0.9981),Translation=(X=0,Y=0, ...(truncated)"
    WeightMinimum=0
    Weight=@RigVMFunction_MathBoolToFloat.Result
   )
   &RigVMFunction_MathQuaternionFromEuler(Euler=(X=-20,Y=48,Z=-32))
   &RigVMFunction_MathQuaternionFromEuler(Euler=(X=-20,Y=20,Z=20))
   &RigVMFunction_MathQuaternionFromEuler(Euler=(X=-30,Y=-30,Z=30))
   &RigVMFunction_MathQuaternionFromEuler(Euler=(X=0,Y=20,Z=0))
   &RigVMFunction_MathQuaternionFromEuler(Euler=(X=0,Y=20,Z=10))
   &RigVMFunction_MathQuaternionFromEuler(Euler=(X=20,Y=20,Z=20))
   &RigVMFunction_MathQuaternionFromEuler(Euler=(X=0,Y=5,Z=5))
   &RigVMFunction_MathQuaternionFromEuler(Euler=(X=10,Y=-10,Z=25))
   &RigVMFunction_MathQuaternionFromEuler(Euler=(X=12,Y=-20,Z=65))
   &RigVMFunction_MathQuaternionFromEuler(Euler=(X=0,Y=5,Z=15))
   &RigVMFunction_MathQuaternionFromEuler(Euler=(X=0,Y=-10,Z=30))
   &RigUnit_SphericalPoseReader(
    ActiveRegionScaleFactors=(PositiveWidth=3,NegativeWidth=0.5,PositiveHeight=3,NegativeHeight=0.1)
    ActiveRegionSize=0.1
    Debug=(bDrawDebug=false,DebugScale=25,DebugSegments=20,DebugThickness=0.25)
    DriverAxis=(X=-1,Y=0,Z=0)
    DriverItem=(Type=Bone,Name="thigh_twist_01_l")
    ExecutePin=()
    FalloffRegionScaleFactors=(PositiveWidth=1,NegativeWidth=2,PositiveHeight=5,NegativeHeight=1)
    FalloffSize=0.38
    OptionalParentItem=(Type=Bone,Name="pelvis")
    RotationOffset=(X=90,Y=0,Z=-35)
   )
   &RigUnit_SphericalPoseReader(
    ActiveRegionScaleFactors=(PositiveWidth=3,NegativeWidth=0.5,PositiveHeight=3,NegativeHeight=0.1)
    ActiveRegionSize=0.1
    Debug=(bDrawDebug=false,DebugScale=25,DebugSegments=20,DebugThickness=0.25)
    DriverAxis=(X=1,Y=0,Z=0)
    DriverItem=(Type=Bone,Name="thigh_twist_01_r")
    ExecutePin=()
    FalloffRegionScaleFactors=(PositiveWidth=1,NegativeWidth=2,PositiveHeight=5,NegativeHeight=1)
    FalloffSize=0.38
    FlipHeightScaling=false
    FlipWidthScaling=true
    OptionalParentItem=(Type=Bone,Name="pelvis")
    RotationOffset=(X=-90,Y=0,Z=-35)
   )
   &RigUnit_SphericalPoseReader(
    ActiveRegionScaleFactors=(PositiveWidth=5.5,NegativeWidth=4.5,PositiveHeight=2,NegativeHeight=2)
    ActiveRegionSize=0.1
    Debug=(bDrawDebug=false,DebugScale=25,DebugSegments=20,DebugThickness=0.25)
    DriverAxis=(X=-1,Y=0,Z=0)
    DriverItem=(Type=Bone,Name="thigh_l")
    ExecutePin=()
    FalloffRegionScaleFactors=(PositiveWidth=1,NegativeWidth=1.5,PositiveHeight=1,NegativeHeight=1)
    FalloffSize=0.7
    FlipWidthScaling=false
    OptionalParentItem=(Type=Bone)
    RotationOffset=(X=180,Y=50,Z=0)
   )
  ]}
 New Function_ContainedGraph {nodes [
   RigVMFunctionEntryNode()
   > RigVMFunction_Sequence_3 = RigVMAggregateNode()
   n1 = RigVMDispatchNode(BoneName=root, Default=1, Name=deformRig_r_thigh_up_alpha) {template="DISPATCH_RigDispatch_GetAnimAttribute(in Name,in BoneName,in Default,out Value,out Found)"}
   &RigUnit_ModifyTransforms(
    ItemToModify="((Item=(Type=Bone,Name=\"base_belt_r\"),Transform=(Rotation=(X=0,Y=0,Z=-0.1418,W=0.9899),Translation=(X=-7.6105,Y=-1.6902,Z=0),Scale3D=(X=1,Y=1,Z=1))),(Item=(Type=Bone,Name=\"base_beltLower_r\"),Transform=(Rotation=(X=0.0075,Y=0.1694,Z=-0.0439,W=0.9845),Transl ...(truncated)"
    WeightMinimum=0
    ExecutePin=@RigVMFunction_Sequence_3.A
    Weight=@n1.Value
   )
   Multiply = &RigVMFunction_MathDoubleMul(A=RigVMVariableNode() {variable=KatanaTransform}.Value, B=@n1.Value)
   > &RigUnit_ModifyTransforms(
    ItemToModify=((Item=(Type=Bone,Name="base_katana_r"),Transform=(Rotation=(X=-0.1006,Y=-0.2397,Z=-0.0102,W=0.9656),Translation=(X=-0.2065,Y=0.204,Z=0.6789),Scale3D=(X=1,Y=1,Z=1))))
    WeightMinimum=0
    Weight=@Multiply.Result
   )
   n2 = RigVMVariableNode() {variable=Thigh_R_SideRotation}
   > &RigUnit_ModifyTransforms(
    ItemToModify="((Item=(Type=Bone,Name=\"base_belt_r\"),Transform=(Rotation=(X=0,Y=0,Z=-0.1728,W=0.985),Translation=(X=-5.3174,Y=-4.9192,Z=0),Scale3D=(X=1,Y=1,Z=1))),(Item=(Type=Bone,Name=\"base_beltLower_r\"),Transform=(Rotation=(X=0.0174,Y=0.0428,Z=-0.2958,W=0.9541),Transla ...(truncated)"
    WeightMinimum=0
    Weight=@n2.Value
   )
   > &RigUnit_ModifyTransforms(
    ItemToModify=((Item=(Type=Bone,Name="base_katana_r"),Transform=(Rotation=(X=-0.1844,Y=-0.2079,Z=-0.0836,W=0.957),Translation=(X=0,Y=0,Z=0),Scale3D=(X=1,Y=1,Z=1))))
    WeightMinimum=0
    Weight=&RigVMFunction_MathDoubleMul(A=RigVMVariableNode() {variable=KatanaTransform}.Value, B=@n2.Value).Result
   )
   > &RigUnit_ParentConstraint(
    AdvancedSettings=(InterpolationType=Average,RotationOrderForFilter=XZY)
    Child=(Type=Bone,Name="base_mask_a_r")
    Parents=((Item=(Type=Bone,Name="dyn_skirt_01_r"),Weight=1))
    Weight=1
   )
   > &RigUnit_ParentConstraint(
    AdvancedSettings=(InterpolationType=Average,RotationOrderForFilter=XZY)
    Child=(Type=Bone,Name="base_mask_b_r")
    Parents=((Item=(Type=Bone,Name="dyn_skirt_01_r"),Weight=1))
    Weight=1
   )
   > &RigUnit_ModifyTransforms(
    ItemToModify="((Item=(Type=Bone,Name=\"base_mask_a_r\"),Transform=(Rotation=(X=0,Y=-0.1114,Z=0,W=0.9938),Translation=(X=-1.2996,Y=0,Z=-0.045),Scale3D=(X=1,Y=1,Z=1))),(Item=(Type=Bone,Name=\"base_mask_b_r\"),Transform=(Rotation=(X=0,Y=-0.0586,Z=0,W=0.9983),Translation=(X=-3. ...(truncated)"
    WeightMinimum=0
    Weight=RigVMVariableNode() {variable=Thigh_R_SideRotation}.Value
   )
   RigVMFunctionReturnNode(ExecuteContext=@RigVMFunction_Sequence_3.D)
   &RigUnit_ParentConstraint(
    AdvancedSettings=(InterpolationType=Average,RotationOrderForFilter=XZY)
    Child=(Type=Bone,Name="base_bow01_l")
    Filter=(TranslationFilter=(bX=false,bY=false,bZ=false),RotationFilter=(bX=True,bY=True,bZ=True),ScaleFilter=(bX=True,bY=True,bZ=True))
    Parents=((Item=(Type=Bone,Name="dyn_skirt_02_bck"),Weight=1),(Item=(Type=Bone,Name="dyn_skirt_01_bck_l"),Weight=1))
    Weight=1
    ExecutePin=@RigVMFunction_Sequence_3.C
   )
   > &RigUnit_ParentConstraint(
    AdvancedSettings=(InterpolationType=Average,RotationOrderForFilter=XZY)
    Child=(Type=Bone,Name="base_bow01_r")
    Filter=(TranslationFilter=(bX=false,bY=false,bZ=false),RotationFilter=(bX=True,bY=True,bZ=True),ScaleFilter=(bX=True,bY=True,bZ=True))
    Parents=((Item=(Type=Bone,Name="dyn_skirt_02_bck"),Weight=1),(Item=(Type=Bone,Name="dyn_skirt_01_bck_r"),Weight=1))
    Weight=1
   )
   &RigUnit_ModifyTransforms(
    ItemToModify="((Item=(Type=Bone,Name=\"base_belt_l\"),Transform=(Rotation=(X=0,Y=0,Z=0.0152,W=0.9999),Translation=(X=-4.4116,Y=-2.9871,Z=0),Scale3D=(X=1,Y=1,Z=1))),(Item=(Type=Bone,Name=\"dyn_charm_01_a\"),Transform=(Rotation=(X=-0.1343,Y=0,Z=0,W=0.9909),Translation=(X=0,Y= ...(truncated)"
    WeightMinimum=0
    Weight=RigVMDispatchNode(BoneName=root, Default=1, Name=deformRig_l_thigh_up_alpha) {template="DISPATCH_RigDispatch_GetAnimAttribute(in Name,in BoneName,in Default,out Value,out Found)"}.Value
    ExecutePin=@RigVMFunction_Sequence_3.B
   )
   > &RigUnit_ModifyTransforms(
    ItemToModify="((Item=(Type=Bone,Name=\"base_belt_l\"),Transform=(Rotation=(X=0,Y=0,Z=-0.2271,W=0.9739),Translation=(X=-2.8792,Y=-0.306,Z=0),Scale3D=(X=1,Y=1,Z=1))),(Item=(Type=Bone,Name=\"dyn_charm_01_a\"),Transform=(Rotation=(X=0,Y=0,Z=0,W=1),Translation=(X=0,Y=0,Z=0),Scal ...(truncated)"
    WeightMinimum=0
    Weight=RigVMVariableNode() {variable=Thigh_L_SideRotation}.Value
   )
   > &RigUnit_ParentConstraint(
    AdvancedSettings=(RotationOrderForFilter=XZY)
    Child=(Type=Bone,Name="base_charm_01_a")
    Parents="((Item=(Type=Bone,Name=\"spine_01\"),Weight=0.0001),(Item=(Type=Bone,Name=\"thigh_twist_01_l\"),Weight=0.0181),(Item=(Type=Bone,Name=\"deform_groin_l\"),Weight=0.0358),(Item=(Type=Bone,Name=\"dyn_belt_l\"),Weight=0.3997),(Item=(Type=Bone,Name=\"pelvis\"),Weight=0.54 ...(truncated)"
    Weight=1
   )
   > &RigUnit_ParentConstraint(
    AdvancedSettings=(RotationOrderForFilter=XZY)
    Child=(Type=Bone,Name="base_charm_01_b")
    Parents="((Item=(Type=Bone,Name=\"deform_glute_l\"),Weight=0.0003),(Item=(Type=Bone,Name=\"deform_groin_l\"),Weight=0.0412),(Item=(Type=Bone,Name=\"thigh_twist_01_l\"),Weight=0.0461),(Item=(Type=Bone,Name=\"dyn_belt_l\"),Weight=0.3961),(Item=(Type=Bone,Name=\"pelvis\"),Weigh ...(truncated)"
    Weight=1
   )
   > &RigUnit_ParentConstraint(
    AdvancedSettings=(RotationOrderForFilter=XZY)
    Child=(Type=Bone,Name="base_charm_01_b")
    Filter=(TranslationFilter=(bX=false,bY=false,bZ=false),RotationFilter=(bX=true,bY=true,bZ=true),ScaleFilter=(bX=True,bY=True,bZ=True))
    Parents=((Item=(Type=Bone,Name="dyn_skirt_01_front_l"),Weight=1))
    Weight=1
   )
   > &RigUnit_ParentConstraint(
    AdvancedSettings=(RotationOrderForFilter=XZY)
    Child=(Type=Bone,Name="base_charm_01_a")
    Filter=(TranslationFilter=(bX=false,bY=false,bZ=false),RotationFilter=(bX=true,bY=true,bZ=true),ScaleFilter=(bX=True,bY=True,bZ=True))
    Parents=((Item=(Type=Bone,Name="dyn_skirt_01_front_l"),Weight=1))
    Weight=1
   )
  ]}
 New Function_ContainedGraph {nodes [
   n1 = RigVMFunctionEntryNode()
   RigVMFunction_Sequence = &RigVMFunction_Sequence(ExecuteContext=@n1.ExecuteContext)
   n2 = &RigUnit_SphericalPoseReader(
    ActiveRegionScaleFactors=(PositiveWidth=1,NegativeWidth=1,PositiveHeight=1,NegativeHeight=0.8)
    ActiveRegionSize=0.1
    Debug=(bDrawDebug=false,bDraw2D=False,bDrawLocalAxes=False,DebugScale=25,DebugSegments=20,DebugThickness=0.25)
    DriverAxis=(X=1,Y=0,Z=0)
    DriverItem=(Type=Bone,Name="upperarm_r")
    FalloffRegionScaleFactors=(PositiveWidth=1,NegativeWidth=1,PositiveHeight=1.7,NegativeHeight=0.5)
    FalloffSize=0.2
    FlipHeightScaling=False
    FlipWidthScaling=False
    OptionalParentItem=(Type=Bone,Name="None")
    RotationOffset=(X=51.9395,Y=8.7023,Z=22.4256)
    ExecutePin=@RigVMFunction_Sequence.A
   )
   Multiply = &RigVMFunction_MathDoubleMul(
    B=RigVMDispatchNode(False=1, True=0, Condition=RigVMVariableNode() {variable=IsInAir}.Value) {template="DISPATCH_RigVMDispatch_If(in Condition,in True,in False,out Result)"}.Result
    A=@n2.OutputParam
   )
   > &RigUnit_ModifyTransforms(
    ItemToModify="((Item=(Type=Bone,Name=\"base_shoulderPad_l\"),Transform=(Rotation=(X=0.006,Y=0.0142,Z=-0.0377,W=0.9992),Translation=(X=1.2431,Y=-0.5782,Z=0.2115),Scale3D=(X=1,Y=1,Z=1))),(Item=(Type=Bone,Name=\"base_shoulderPad_r\"),Transform=(Rotation=(X=0.3525,Y=0.0132,Z=-0 ...(truncated)"
    WeightMinimum=0
    Weight=@Multiply.Result
   )
   > &RigUnit_ModifyTransforms(
    ItemToModify="((Item=(Type=Bone,Name=\"base_shoulderPad_l\"),Transform=(Rotation=(X=0.3838,Y=0.057,Z=0.0238,W=0.9213),Translation=(X=3.7066,Y=-3.8784,Z=-3.1543),Scale3D=(X=1,Y=1,Z=1))),(Item=(Type=Bone,Name=\"base_shoulderPad_r\"),Transform=(Rotation=(X=0,Y=0,Z=0,W=1),Trans ...(truncated)"
    WeightMinimum=0
    Weight=RigVMDispatchNode(BoneName=root, Default=1, Name=deformRig_l_upperarm_up_alpha) {template="DISPATCH_RigDispatch_GetAnimAttribute(in Name,in BoneName,in Default,out Value,out Found)"}.Value
   )
   > n3 = &RigUnit_SphericalPoseReader(
    ActiveRegionScaleFactors=(PositiveWidth=1,NegativeWidth=1,PositiveHeight=1,NegativeHeight=2)
    ActiveRegionSize=0.15
    Debug=(bDrawDebug=false,bDraw2D=False,bDrawLocalAxes=False,DebugScale=25,DebugSegments=20,DebugThickness=0.25)
    DriverAxis=(X=-1,Y=0,Z=0)
    DriverItem=(Type=Bone,Name="upperarm_r")
    FalloffRegionScaleFactors=(PositiveWidth=1,NegativeWidth=1,PositiveHeight=2,NegativeHeight=1)
    FalloffSize=0.3
    FlipHeightScaling=False
    FlipWidthScaling=False
    OptionalParentItem=(Type=Bone,Name="upperarm_r_Hold")
    RotationOffset=(X=-90,Y=20,Z=15)
   )
   > &RigUnit_ModifyTransforms(
    ItemToModify="((Item=(Type=Bone,Name=\"base_collar_side_r\"),Transform=(Rotation=(X=0.1784,Y=-0.4207,Z=-0.051,W=0.888),Translation=(X=-2.0248,Y=0.076,Z=-0.6683),Scale3D=(X=1,Y=1,Z=1))),(Item=(Type=Bone,Name=\"base_collar_fr_r\"),Transform=(Rotation=(X=-0.0121,Y=-0.6266,Z=-0 ...(truncated)"
    WeightMinimum=0
    Weight=@n3.OutputParam
   )
   > &RigUnit_ModifyTransforms(
    ItemToModify=((Item=(Type=Bone,Name="base_collar_bck"),Transform=(Rotation=(X=0,Y=0.5736,Z=0,W=0.8192),Translation=(X=0.5,Y=0,Z=-1),Scale3D=(X=1,Y=1,Z=1))))
    WeightMinimum=0
    Weight=RigVMVariableNode() {variable=headLookUp}.Value
   )
   > &RigUnit_ModifyTransforms(
    ItemToModify=((Item=(Type=Bone,Name="base_shoulderPad_l"),Transform=(Rotation=(X=0.1753,Y=0,Z=0,W=0.9845),Translation=(X=2.0613,Y=-1.7388,Z=0.2315),Scale3D=(X=1,Y=1,Z=1))))
    WeightMinimum=0
    Weight=RigVMVariableNode() {variable=ArmUpSkydiveL}.Value
   )
   > &RigUnit_ModifyTransforms(
    ItemToModify=((Item=(Type=Bone,Name="base_shoulderPad_r"),Transform=(Rotation=(X=0.2208,Y=0.0071,Z=-0.0455,W=0.9742),Translation=(X=-4.8255,Y=3.9213,Z=0.6901),Scale3D=(X=1,Y=1,Z=1))))
    WeightMinimum=0
    Weight=RigVMVariableNode() {variable=ArmUpSkydiveR}.Value
   )
   > &RigUnit_ModifyTransforms(
    ItemToModify=((Item=(Type=Bone,Name="base_foxMask_l"),Transform=(Rotation=(X=0,Y=0,Z=0.0775,W=0.997),Translation=(X=-0.9266,Y=0.9494,Z=-1.1799),Scale3D=(X=1,Y=1,Z=1))))
    WeightMinimum=0
    Weight=RigVMDispatchNode(BoneName=root, Default=1, Name=deformRig_l_wrist_up_alpha) {template="DISPATCH_RigDispatch_GetAnimAttribute(in Name,in BoneName,in Default,out Value,out Found)"}.Value
   )
   RigVMFunctionReturnNode(ExecuteContext=@RigVMFunction_Sequence.B)
  ]}
 New Function_ContainedGraph {nodes [
   RigVMFunctionEntryNode()
   > RigVMFunction_Sequence = RigVMAggregateNode()
   Delta Rotation Y = RigVMFunctionReferenceNode(Child=thigh_l, Parent=thigh_l_null, Source Maximum=-75, Source Minimum=0, ExecuteContext=@RigVMFunction_Sequence.A)
   > RigVMVariableNode(Value=@"Delta Rotation Y".Delta Rotation) {variable=Thigh_L_SideRotation}
   > Delta Rotation Y_1 = RigVMFunctionReferenceNode(Child=thigh_r, Parent=thigh_r_null, Source Maximum=-75, Source Minimum=0)
   > RigVMVariableNode(Value=@"Delta Rotation Y_1".Delta Rotation) {variable=Thigh_R_SideRotation}
   > Delta Rotation Y_2 = RigVMFunctionReferenceNode(Child=thigh_l, Parent=thigh_l_null, Source Maximum=15, Source Minimum=0)
   > RigVMVariableNode(Value=@"Delta Rotation Y_2".Delta Rotation) {variable=Thigh_L_In_SideRotation}
   > RigVMVariableNode(Value=&RigVMFunction_MathDoubleRemap(
     SourceMaximum=50
     SourceMinimum=70
     TargetMaximum=1
     TargetMinimum=0
     Value=&RigUnit_GetRelativeTransformForItem(Child=(Type=Bone,Name="spine_01"), Parent=(Type=Bone,Name="root"), bChildInitial=False, bParentInitial=False).RelativeTransform.Translation.Z
    ).Result) {variable=PelvisToGround}
   > Delta Rotation Z = RigVMFunctionReferenceNode(Child=thigh_l, Parent=thigh_l_null, Source Maximum=50, Source Minimum=5)
   > RigVMVariableNode(Value=@"Delta Rotation Z".Delta Rotation) {variable=Thigh_L_BackRotation}
   > Delta Rotation Z_1 = RigVMFunctionReferenceNode(Child=thigh_r, Parent=thigh_r_null, Source Maximum=50, Source Minimum=5)
   > RigVMVariableNode(Value=@"Delta Rotation Z_1".Delta Rotation) {variable=Thigh_R_BackRotation}
   > Delta Rotation Z_1_1 = RigVMFunctionReferenceNode(Child=calf_l, Parent=calf_l_null, Source Maximum=90, Source Minimum=15)
   > RigVMVariableNode(Value=@"Delta Rotation Z_1_1".Delta Rotation) {variable=Calf_L_Rotation}
   > Delta Rotation Z_1_1_1 = RigVMFunctionReferenceNode(Child=calf_r, Parent=calf_r_null, Source Maximum=90, Source Minimum=15)
   > RigVMVariableNode(Value=@"Delta Rotation Z_1_1_1".Delta Rotation) {variable=Calf_R_Rotation}
   n1 = RigVMVariableNode() {variable=IsInAir}
   > RigVMVariableNode(Value=&RigVMFunction_MathDoubleRemap(
     SourceMaximum=1
     SourceMinimum=0
     TargetMaximum=1
     TargetMinimum=0
     Value=RigVMDispatchNode(False=0, True=1, Condition=@n1.Value) {template="DISPATCH_RigVMDispatch_If(in Condition,in True,in False,out Result)"}.Result
    ).Result) {variable=isSkydiving}
   > RigVMVariableNode(Value=&RigVMFunction_MathDoubleRemap(SourceMaximum=1, SourceMinimum=0, TargetMaximum=0, TargetMinimum=1, Value=RigVMVariableNode() {variable=isSkydiving}.Value).Result) {variable=NotInAir}
   > Delta Rotation Z_2 = RigVMFunctionReferenceNode(Child=head, Parent=neck_01_null, Source Maximum=-63, Source Minimum=-21)
   > RigVMVariableNode(Value=@"Delta Rotation Z_2".Delta Rotation) {variable=headLookUp}
   Delta Rotation X = RigVMFunctionReferenceNode(Child=pelvis, Parent=spine_05_null, Source Maximum=-38, Source Minimum=-15, ExecuteContext=@RigVMFunction_Sequence.B)
   > RigVMVariableNode(Value=@"Delta Rotation X".Delta Rotation) {variable=SpineRotRight}
   > Delta Rotation Z_3 = RigVMFunctionReferenceNode(Child=thigh_l, Parent=thigh_l_null, Source Maximum=22, Source Minimum=0)
   > RigVMVariableNode(Value=@"Delta Rotation Z_3".Delta Rotation) {variable=skydiveThigBackL}
   > Delta Rotation Z_3_1 = RigVMFunctionReferenceNode(Child=thigh_r, Parent=thigh_r_null, Source Maximum=22, Source Minimum=0)
   > RigVMVariableNode(Value=@"Delta Rotation Z_3_1".Delta Rotation) {variable=skydiveThigBackR}
   > Delta Rotation Z_3_2 = RigVMFunctionReferenceNode(Child=pelvis, Parent=spine_05_null, Source Maximum=58, Source Minimum=20)
   > RigVMVariableNode(Value=@"Delta Rotation Z_3_2".Delta Rotation) {variable=spineRotBack}
   n2 = RigVMVariableNode() {variable=IsInAir}
   Remap = &RigVMFunction_MathDoubleRemap(
    SourceMaximum=1
    SourceMinimum=0
    TargetMaximum=1
    TargetMinimum=0
    bClamp=False
    Value=RigVMDispatchNode(False=0, True=1, Condition=@n2.Value) {template="DISPATCH_RigVMDispatch_If(in Condition,in True,in False,out Result)"}.Result
   )
   Multiply = &RigVMFunction_MathDoubleMul(
    B=&RigVMFunction_MathDoubleRemap(
     SourceMaximum=0.2
     SourceMinimum=0
     TargetMaximum=1
     TargetMinimum=0
     bClamp=False
     Value=RigVMDispatchNode(BoneName=root, Default=1, Name=deformRig_l_upperarm_up_alpha) {template="DISPATCH_RigDispatch_GetAnimAttribute(in Name,in BoneName,in Default,out Value,out Found)"}.Value
    ).Result
    A=@Remap.Result
   )
   > RigVMVariableNode(Value=@Multiply.Result) {variable=ArmUpSkydiveL}
   n3 = RigVMDispatchNode(BoneName=root, Default=1, Name=deformRig_r_upperarm_up_alpha) {template="DISPATCH_RigDispatch_GetAnimAttribute(in Name,in BoneName,in Default,out Value,out Found)"}
   > RigVMVariableNode(Value=&RigVMFunction_MathDoubleMul(
     B=&RigVMFunction_MathDoubleRemap(SourceMaximum=0.2, SourceMinimum=0, TargetMaximum=1, TargetMinimum=0, bClamp=False, Value=@n3.Value).Result
     A=@Remap.Result
    ).Result) {variable=ArmUpSkydiveR}
   RigVMFunctionReturnNode(ExecuteContext=@RigVMFunction_Sequence.C)
   RigVMFunction_MathBoolOr = &RigVMFunction_MathBoolOr(A=@n2.Value, B=RigVMVariableNode() {variable=IsEmoting}.Value)
  ]}
 New Function_ContainedGraph {nodes [
   n1 = RigVMFunctionEntryNode()
   RigVMFunction_MathQuaternionToEuler = &RigVMFunction_MathQuaternionToEuler(
    RotationOrder=XYZ
    Value=&RigUnit_GetRelativeTransformForItem(
     Child=(Type=Bone,Name="hand_l")
     Parent=(Type=Null,Name="lowerarm_l")
     bChildInitial=False
     bParentInitial=False
     Child.Name=@n1.Child
     Parent.Name=@n1.Parent
    ).RelativeTransform.Rotation
   )
   Remap = &RigVMFunction_MathDoubleRemap(
    TargetMaximum=1
    TargetMinimum=0
    Value=@RigVMFunction_MathQuaternionToEuler.Result.X
    SourceMaximum=@n1.Source Maximum
    SourceMinimum=@n1.Source Minimum
   )
   > RigVMFunctionReturnNode(Delta Rotation=@Remap.Result)
  ]}
 RigVMFunctionLibrary {nodes [
   Belt Accessories Setup = RigVMCollapseNode(ExecuteContext=())
   Control Setup = RigVMCollapseNode(Debug Controls=False, ExecuteContext=())
   Delta Rotation X = RigVMCollapseNode(ExecuteContext=())
   Delta Rotation Y = RigVMCollapseNode(ExecuteContext=())
   Delta Rotation Z = RigVMCollapseNode(ExecuteContext=())
   Jacket Setup = RigVMCollapseNode(ExecuteContext=())
   Skirt Setup = RigVMCollapseNode(ExecuteContext=())
   Variables = RigVMCollapseNode(ExecuteContext=())
   rotationTransform = RigVMCollapseNode(ExecuteContext=(), Item=(Type=None,Name="None"))
  ]}
 RigVMFunction_Sequence_1_1_ContainedGraph {nodes [
   n1 = RigVMFunctionEntryNode()
   n2 = &RigVMFunction_Sequence(ExecuteContext=@n1.ExecuteContext)
   n3 = &RigVMFunction_Sequence(ExecuteContext=@n2.B)
   n4 = &RigVMFunction_Sequence(ExecuteContext=@n3.B)
   n5 = &RigVMFunction_Sequence(ExecuteContext=@n4.B)
   n6 = &RigVMFunction_Sequence(ExecuteContext=@n5.B)
   n7 = &RigVMFunction_Sequence(ExecuteContext=@n6.B)
   RigVMFunctionReturnNode(A=@n2.A, B=@n3.A, C=@n4.A, D=@n5.A, E=@n6.A, F=@n7.A, G=@n7.B)
  ]}
 RigVMFunction_Sequence_1_ContainedGraph {nodes [
   n1 = RigVMFunctionEntryNode()
   RigVMFunction_Sequence = &RigVMFunction_Sequence(ExecuteContext=@n1.ExecuteContext)
   n2 = &RigVMFunction_Sequence(ExecuteContext=@RigVMFunction_Sequence.B)
   RigVMFunctionReturnNode(A=@RigVMFunction_Sequence.A, B=@n2.A, C=@n2.B)
  ]}
 RigVMFunction_Sequence_2_1_ContainedGraph {nodes [
   n1 = RigVMFunctionEntryNode()
   n2 = &RigVMFunction_Sequence(ExecuteContext=@n1.ExecuteContext)
   n3 = &RigVMFunction_Sequence(ExecuteContext=@n2.B)
   RigVMFunctionReturnNode(A=@n2.A, B=@n3.A, C=@n3.B)
  ]}
 RigVMFunction_Sequence_3_1_ContainedGraph {nodes [
   n1 = RigVMFunctionEntryNode()
   n2 = &RigVMFunction_Sequence(ExecuteContext=@n1.ExecuteContext)
   n3 = &RigVMFunction_Sequence(ExecuteContext=@n2.B)
   n4 = &RigVMFunction_Sequence(ExecuteContext=@n3.B)
   RigVMFunctionReturnNode(A=@n2.A, B=@n3.A, C=@n4.A, D=@n4.B)
  ]}
 RigVMModel {nodes [
   RigUnit_BeginExecution = &RigUnit_BeginExecution()
   RigVMFunction_Sequence = &RigVMFunction_Sequence(ExecuteContext=@RigUnit_BeginExecution.ExecutePin)
   Control Setup = RigVMFunctionReferenceNode(Debug Controls=false, ExecuteContext=@RigVMFunction_Sequence.A)
   FN_DeformRig_FNC = RigVMFunctionReferenceNode(
    DeformRig_ Curve="(deformRig_l_upperarm_up_alpha_72_#1=True,deformRig_r_upperarm_up_alpha_73_#2=True,deformRig_l_upperarm_fr_alpha_74_#3=False,deformRig_r_upperarm_fr_alpha_75_CB51B80C ...(truncated)"
    DeformRig_Bool="(upperarm_1_#4=true,elbow_3_#5=True,wrist_7_#6=True,groin_8_#7=true,glute_17_#8=false,knee_9_40C408324CA26 ...(truncated)"
    DeformRig_Debug=(upperarm_debug_18_#4=false,wrist_debug_20_#6=false,lat_debug_26_#9=false,groinGlute_debug_27_#7=false)
    DeformRig_Value="(upperarm_up_18_#10=(Rotation=(X=0.0,Y=-0.0,Z=0.0,W=1.0),Translation=(X=0.0,Y=0.0,Z=8.0),Scale3D=(X=0.70,Y=1.0,Z=1.0)),upperarm_fwd_fr_19_#11=(Rotation=(X=0.0,Y=-0.0,Z=0.0,W=1.0),Translation=(X=0.0, ...(truncated)"
    ExecuteContext=@RigVMFunction_Sequence.B
   )
   > Variables = RigVMFunctionReferenceNode()
   > Skirt Setup = RigVMFunctionReferenceNode()
   > Belt Accessories Setup = RigVMFunctionReferenceNode()
   > Jacket Setup = RigVMFunctionReferenceNode()
  ]}
]
hierarchy [
 root {loc=<0 0 0>, rot=<0 0 0 1>}
 pelvis {parent=root, loc=<0 -0.5301 98.0571>, rot=<0.0127 0.707 -0.0127 -0.707>}
 spine_01 {parent=pelvis, loc=<6.1208 -0.9328 0>, rot=<0 0 0.034 -0.9994>}
 spine_02 {parent=spine_01, loc=<6.5996 -0.2095 0>, rot=<0 0 0.0238 -0.9997>}
 spine_03 {parent=spine_02, loc=<6.4914 -0.1237 0>, rot=<0 0 0.0149 -0.9999>}
 spine_04 {parent=spine_03, loc=<6.9598 0.0794 0>, rot=<0 0 -0.0132 -0.9999>}
 spine_05 {parent=spine_04, loc=<6.9987 0 0>, rot=<0 0 -0.0015 -1>}
 attach_cape {parent=spine_05, loc=<9.0178 -11.714 0>, rot=<0.0616 -0.7044 -0.0616 -0.7044>}
 neck_01 {parent=spine_05, loc=<11.3138 0.1529 0>, rot=<0 0 -0.2003 -0.9797>}
 neck_02 {parent=neck_01, loc=<4.5627 0 0>, rot=<0 0 0.0504 -0.9987>}
 head {parent=neck_02, loc=<5.3914 0 0>, rot=<0 0 0.115 -0.9934>}
 attach_backpack {parent=spine_05, loc=<-0.1221 -13.473 0>, rot=<-0.0093 -0.707 0.0093 -0.707>}
 clavicle_l {parent=spine_05, loc=<5.4831 -0.1053 -3.2946>, rot=<0.0465 -0.7289 -0.009 -0.6829>}
 upperarm_l {parent=clavicle_l, loc=<9.8658 -1.2136 -0.4105>, rot=<0.0164 -0.3245 -0.0269 -0.9454>}
 lowerarm_l {parent=upperarm_l, loc=<26.465 -0.8345 0.259>, rot=<0.0036 0.0161 -0.2298 -0.9731>}
 hand_l {parent=lowerarm_l, loc=<22.7355 -0.3203 -0.6105>, rot=<-0.6784 -0.0223 0.0735 -0.7307>}
 thumb_01_l {parent=hand_l, loc=<1.3118 -0.3191 -2.459>, rot=<0.6546 -0.4042 0.0664 -0.6354>}
 thumb_02_l {parent=thumb_01_l, loc=<5.9336 -0.0448 0.0401>, rot=<-0.0037 0.0213 0.298 -0.9543>}
 thumb_03_l {parent=thumb_02_l, loc=<2.2223 0.3682 -0.0907>, rot=<0.0118 -0.013 0.0066 -0.9998>}
 index_metacarpal_l {parent=hand_l, loc=<1.6535 0.0183 -2.96>, rot=<0.0158 -0.0124 -0.0173 -0.9996>}
 index_01_l {parent=index_metacarpal_l, loc=<8.3008 0.0357 -0.0524>, rot=<-0.0044 0.0485 0.0924 -0.9945>}
 index_02_l {parent=index_01_l, loc=<3.3587 -0.3004 0.0452>, rot=<-0.002 0.0027 0.3225 -0.9466>}
 index_03_l {parent=index_02_l, loc=<2.5578 -0.123 -0.0252>, rot=<-0.002 0.0067 0.2558 -0.9667>}
 middle_metacarpal_l {parent=hand_l, loc=<2.1856 -0.1725 -0.5984>, rot=<0.006 0.0168 -0.0329 -0.9993>}
 middle_01_l {parent=middle_metacarpal_l, loc=<7.3275 -0.2447 0.0129>, rot=<-0.0208 0.0042 0.0858 -0.9961>}
 middle_02_l {parent=middle_01_l, loc=<4.0123 0.1401 0.1175>, rot=<-0.0218 0.0258 0.3512 -0.9357>}
 middle_03_l {parent=middle_02_l, loc=<2.6077 0.0534 0.1373>, rot=<0.0082 0.0188 0.1708 -0.9851>}
 ring_metacarpal_l {parent=hand_l, loc=<2.2074 -0.3625 1.3638>, rot=<-0.0284 0.0722 -0.0458 -0.9959>}
 ring_01_l {parent=ring_metacarpal_l, loc=<7.0295 -0.6197 -0.1446>, rot=<-0.0489 -0.0252 0.1892 -0.9804>}
 ring_02_l {parent=ring_01_l, loc=<3.6174 0.4647 -0.1027>, rot=<0 0 0.2806 -0.9598>}
 ring_03_l {parent=ring_02_l, loc=<2.4152 0.114 0.0199>, rot=<0 0 0.1189 -0.9929>}
 pinky_metacarpal_l {parent=hand_l, loc=<1.7616 -0.9753 3.3814>, rot=<-0.0397 0.0814 -0.069 -0.9935>}
 pinky_01_l {parent=pinky_metacarpal_l, loc=<7.1709 -0.4831 -0.1094>, rot=<-0.0654 -0.0035 0.2242 -0.9723>}
 pinky_02_l {parent=pinky_01_l, loc=<2.8274 0.0893 -0.1043>, rot=<0.0037 0.0011 0.2854 -0.9584>}
 pinky_03_l {parent=pinky_02_l, loc=<2.2046 0.0307 -0.0078>, rot=<-0.0002 0.0098 0.1445 -0.9895>}
 deform_wrist_dn_base_l {parent=hand_l, loc=<0 -1.5326 0>, rot=<0 0 0 -1>}
 deform_wrist_dn_l {parent=deform_wrist_dn_base_l, loc=<0 0 0>, rot=<0 0 0 1>}
 deform_wrist_up_base_l {parent=hand_l, loc=<0 2.3009 0>, rot=<0 0 0 -1>}
 deform_wrist_up_l {parent=deform_wrist_up_base_l, loc=<0 0 0>, rot=<0 0 0 1>}
 base_foxMask_l {parent=deform_wrist_up_l, loc=<0 0 0>, rot=<0 0 0 1>}
 dyn_foxMask_l {parent=base_foxMask_l, loc=<0 0 0>, rot=<0 0 0 1>}
 lowerarm_twist_02_l {parent=lowerarm_l, loc=<10.6707 0 0>, rot=<0 0 0 1>}
 lowerarm_twist_01_l {parent=lowerarm_l, loc=<18.316 0 0>, rot=<0 0 0 1>}
 deform_elbow_in_base_l {parent=lowerarm_l, loc=<0.1181 2.321 -0.2572>, rot=<0 0 0 -1>}
 deform_elbow_in_l {parent=deform_elbow_in_base_l, loc=<0 0 0>, rot=<0 0 0 1>}
 deform_elbow_out_base_l {parent=lowerarm_l, loc=<0 -2.9138 0>, rot=<0 0 0 1>}
 deform_elbow_out_l {parent=deform_elbow_out_base_l, loc=<0 0 0>, rot=<0 0 0 1>}
 upperarm_twist_01_l {parent=upperarm_l, loc=<7.2135 -0.1134 0>, rot=<0 0 0 1>}
 deform_upperarm_bk_base_l {parent=upperarm_twist_01_l, loc=<-6.4053 -4.3746 0.314>, rot=<0 0 0 -1>}
 deform_upperarm_bk_l {parent=deform_upperarm_bk_base_l, loc=<0 0 0>, rot=<0 0 0 1>}
 deform_upperarm_fr_base_l {parent=upperarm_twist_01_l, loc=<-5.6312 5.092 0.3354>, rot=<0 0 0 -1>}
 deform_upperarm_fr_l {parent=deform_upperarm_fr_base_l, loc=<0 0 0>, rot=<0 0 0 1>}
 deform_upperarm_up_base_l {parent=upperarm_twist_01_l, loc=<-8.891 0.4825 3.9517>, rot=<0 0 0 -1>}
 deform_upperarm_up_l {parent=deform_upperarm_up_base_l, loc=<0 0 0>, rot=<0 0 0 1>}
 base_shoulderPad_l {parent=upperarm_twist_01_l, loc=<-3.7746 -0.9552 6.5872>, rot=<0 0.107 0 -0.9943>}
 dyn_shoulderPad_l {parent=base_shoulderPad_l, loc=<0 0 0>, rot=<0 0 0 1>}
 upperarm_twist_02_l {parent=upperarm_l, loc=<19.7998 -0.1134 0>, rot=<0 0 0 1>}
 upperarm_l_null {parent=clavicle_l, type=Null, loc=<9.8658 -1.2136 -0.4105>, rot=<0.0164 -0.3245 -0.0269 -0.9454>}
 upperarm_l_Hold {parent=clavicle_l, loc=<9.8658 -1.2136 -0.4105>, rot=<0.0164 -0.3245 -0.0269 -0.9454>}
 clavicle_r {parent=spine_05, loc=<5.4831 -0.1053 3.2946>, rot=<0.7289 0.0465 -0.6829 0.009>}
 upperarm_r {parent=clavicle_r, loc=<-9.8658 1.2136 0.4105>, rot=<0.0164 -0.3245 -0.0269 -0.9454>}
 lowerarm_r {parent=upperarm_r, loc=<-26.465 0.8345 -0.259>, rot=<0.0036 0.0161 -0.2298 -0.9731>}
 hand_r {parent=lowerarm_r, loc=<-22.7355 0.3203 0.6106>, rot=<-0.6784 -0.0223 0.0735 -0.7307>}
 thumb_01_r {parent=hand_r, loc=<-1.3118 0.3191 2.459>, rot=<0.6546 -0.4042 0.0664 -0.6354>}
 thumb_02_r {parent=thumb_01_r, loc=<-5.9336 0.0448 -0.0401>, rot=<-0.0037 0.0213 0.298 -0.9543>}
 thumb_03_r {parent=thumb_02_r, loc=<-2.2223 -0.3682 0.0907>, rot=<0.0118 -0.013 0.0066 -0.9998>}
 index_metacarpal_r {parent=hand_r, loc=<-1.6535 -0.0183 2.96>, rot=<0.0158 -0.0124 -0.0173 -0.9996>}
 index_01_r {parent=index_metacarpal_r, loc=<-8.3008 -0.0357 0.0524>, rot=<-0.0044 0.0485 0.0924 -0.9945>}
 index_02_r {parent=index_01_r, loc=<-3.3587 0.3004 -0.0452>, rot=<-0.002 0.0027 0.3225 -0.9466>}
 index_03_r {parent=index_02_r, loc=<-2.5578 0.123 0.0252>, rot=<-0.002 0.0067 0.2558 -0.9667>}
 middle_metacarpal_r {parent=hand_r, loc=<-2.1856 0.1725 0.5984>, rot=<0.006 0.0168 -0.0329 -0.9993>}
 middle_01_r {parent=middle_metacarpal_r, loc=<-7.3275 0.2447 -0.0129>, rot=<-0.0208 0.0042 0.0858 -0.9961>}
 middle_02_r {parent=middle_01_r, loc=<-4.0123 -0.1401 -0.1175>, rot=<-0.0218 0.0258 0.3512 -0.9357>}
 middle_03_r {parent=middle_02_r, loc=<-2.6077 -0.0534 -0.1373>, rot=<0.0082 0.0188 0.1708 -0.9851>}
 ring_metacarpal_r {parent=hand_r, loc=<-2.2074 0.3625 -1.3638>, rot=<-0.0284 0.0722 -0.0458 -0.9959>}
 ring_01_r {parent=ring_metacarpal_r, loc=<-7.0295 0.6197 0.1446>, rot=<-0.0489 -0.0252 0.1892 -0.9804>}
 ring_02_r {parent=ring_01_r, loc=<-3.6174 -0.4647 0.1027>, rot=<0 0 0.2806 -0.9598>}
 ring_03_r {parent=ring_02_r, loc=<-2.4152 -0.114 -0.0199>, rot=<0 0 0.1189 -0.9929>}
 pinky_metacarpal_r {parent=hand_r, loc=<-1.7616 0.9753 -3.3814>, rot=<-0.0397 0.0814 -0.069 -0.9935>}
 pinky_01_r {parent=pinky_metacarpal_r, loc=<-7.1709 0.4831 0.1094>, rot=<-0.0654 -0.0035 0.2242 -0.9723>}
 pinky_02_r {parent=pinky_01_r, loc=<-2.8274 -0.0893 0.1043>, rot=<0.0037 0.0011 0.2854 -0.9584>}
 pinky_03_r {parent=pinky_02_r, loc=<-2.2046 -0.0307 0.0078>, rot=<-0.0002 0.0098 0.1445 -0.9895>}
 deform_wrist_dn_base_r {parent=hand_r, loc=<-0.0001 1.5328 0>, rot=<0 0 0 -1>}
 deform_wrist_dn_r {parent=deform_wrist_dn_base_r, loc=<0 0 0>, rot=<0 0 0 1>}
 deform_wrist_up_base_r {parent=hand_r, loc=<-0.0002 -2.3007 -0.0001>, rot=<0 0 0 -1>}
 deform_wrist_up_r {parent=deform_wrist_up_base_r, loc=<0 0 0>, rot=<0 0 0 1>}
 lowerarm_twist_02_r {parent=lowerarm_r, loc=<-10.6707 0 0>, rot=<0 0 0 -1>}
 lowerarm_twist_01_r {parent=lowerarm_r, loc=<-18.316 0 0>, rot=<0 0 0 -1>}
 deform_elbow_in_base_r {parent=lowerarm_r, loc=<-0.1834 -3.6062 0.2572>, rot=<0 0 0 -1>}
 deform_elbow_in_r {parent=deform_elbow_in_base_r, loc=<0 0 0>, rot=<0 0 0 1>}
 deform_elbow_out_base_r {parent=lowerarm_r, loc=<0 4 -0.0001>, rot=<0 0 0 -1>}
 deform_elbow_out_r {parent=deform_elbow_out_base_r, loc=<0 0 0>, rot=<0 0 0 1>}
 upperarm_twist_01_r {parent=upperarm_r, loc=<-7.2135 0.1134 0>, rot=<0 0 0 -1>}
 deform_upperarm_bk_base_r {parent=upperarm_twist_01_r, loc=<6.4053 4.3746 -0.314>, rot=<0 0 0 -1>}
 deform_upperarm_bk_r {parent=deform_upperarm_bk_base_r, loc=<0 0 0>, rot=<0 0 0 1>}
 deform_upperarm_fr_base_r {parent=upperarm_twist_01_r, loc=<5.6312 -5.092 -0.3354>, rot=<0 0 0 -1>}
 deform_upperarm_fr_r {parent=deform_upperarm_fr_base_r, loc=<0 0 0>, rot=<0 0 0 1>}
 deform_upperarm_up_base_r {parent=upperarm_twist_01_r, loc=<8.891 -0.4825 -3.9517>, rot=<0 0 0 -1>}
 deform_upperarm_up_r {parent=deform_upperarm_up_base_r, loc=<0 0 0>, rot=<0 0 0 1>}
 base_shoulderPad_r {parent=upperarm_twist_01_r, loc=<3.7743 0.9552 -6.5868>, rot=<0 0.107 0 -0.9943>}
 dyn_shoulderPad_r {parent=base_shoulderPad_r, loc=<0 0 0>, rot=<0 0 0 1>}
 upperarm_twist_02_r {parent=upperarm_r, loc=<-19.7998 0.1134 0>, rot=<0 0 0 -1>}
 upperarm_r_null {parent=clavicle_r, type=Null, loc=<-9.8658 1.2136 0.4105>, rot=<0.0164 -0.3245 -0.0269 -0.9454>}
 upperarm_r_Hold {parent=clavicle_r, loc=<-9.8658 1.2136 0.4105>, rot=<0.0164 -0.3245 -0.0269 -0.9454>}
 base_collar_bck {parent=spine_05, loc=<12.0115 -4.2544 0>, rot=<0.1953 -0.6796 0.6796 -0.1953>}
 dyn_collar_bck {parent=base_collar_bck, loc=<0 0 0>, rot=<0 0 0 1>}
 base_collar_side_l {parent=spine_05, loc=<9.8352 0.3233 -7.3207>, rot=<0.2772 -0.9398 0.1798 -0.0871>}
 dyn_collar_fr_l {parent=base_collar_side_l, loc=<0 0 0>, rot=<0 0 0 1>}
 base_collar_fr_l {parent=spine_05, loc=<5.0358 7.2168 -6.8516>, rot=<0.4115 -0.9107 -0.0337 -0.012>}
 dyn_collar_side_l {parent=base_collar_fr_l, loc=<0 0 0>, rot=<0 0 0 1>}
 base_collar_side_r {parent=spine_05, loc=<9.8347 0.3233 7.3207>, rot=<0.0871 -0.1798 0.9398 -0.2772>}
 dyn_collar_fr_r {parent=base_collar_side_r, loc=<0 0 0>, rot=<0 0 0 1>}
 base_collar_fr_r {parent=spine_05, loc=<5.0355 7.2168 6.8516>, rot=<0.012 0.0337 0.9107 -0.4115>}
 dyn_collar_side_r {parent=base_collar_fr_r, loc=<0 0 0>, rot=<0 0 0 1>}
 neck_01_null {parent=spine_05, type=Null, loc=<5.4237 1.0981 -2.7084>, rot=<-0.1326 0.003 0.1827 0.9742>}
 spine_05_null {parent=spine_05, type=Null, loc=<0 0 0>, rot=<0 0 0 1>}
 deform_pec_aim_base_l {parent=spine_04, loc=<3.2456 9.3309 -1.2237>, rot=<0.0603 -0.4033 0.3318 -0.8507>}
 deform_pec_aim_l {parent=deform_pec_aim_base_l, loc=<0 0 0>, rot=<0 0 0 1>}
 deform_pec_base_l {parent=deform_pec_aim_l, loc=<9.3665 0 0>, rot=<0 0 0 1>}
 deform_pec_l {parent=deform_pec_base_l, loc=<0 0 0>, rot=<0 0 0 1>}
 deform_pec_aim_base_r {parent=spine_04, loc=<3.2452 9.3309 1.2237>, rot=<-0.4033 -0.0603 0.8507 0.3318>}
 deform_pec_aim_r {parent=deform_pec_aim_base_r, loc=<0 0 0>, rot=<0 0 0 -1>}
 deform_pec_base_r {parent=deform_pec_aim_r, loc=<-9.3667 -0.0003 -0.0004>, rot=<0 0 0 1>}
 deform_pec_r {parent=deform_pec_base_r, loc=<0 0 0>, rot=<0 0 0 1>}
 base_jacket_fr_l {parent=spine_04, loc=<-2.389 7.6917 -10.1008>, rot=<0.04 -0.9609 -0.2738 0.0114>}
 dyn_jacket_fr_l {parent=base_jacket_fr_l, loc=<0 0 0>, rot=<0 0 0 1>}
 base_jacket_side_l {parent=spine_04, loc=<-1.4419 -4.4831 -10.5024>, rot=<0.0388 -0.9324 0.3591 -0.015>}
 dyn_jacket_side_l {parent=base_jacket_side_l, loc=<0 0 0>, rot=<0 0 0 1>}
 base_jacket_bck {parent=spine_04, loc=<-0.8516 -8.4268 0>, rot=<0.0294 -0.7065 0.7065 -0.0294>}
 dyn_jacket_bck {parent=base_jacket_bck, loc=<0 0 0>, rot=<0 0 0 1>}
 base_jacket_fr_r {parent=spine_04, loc=<-2.402 7.8474 10.3527>, rot=<0.9609 0.04 0.0114 0.2738>}
 dyn_jacket_fr_r {parent=base_jacket_fr_r, loc=<0 0 0>, rot=<0 0 0 -1>}
 base_jacket_side_r {parent=spine_04, loc=<-1.4254 -4.6813 10.7223>, rot=<0.9324 0.0388 -0.015 -0.3591>}
 dyn_jacket_side_r {parent=base_jacket_side_r, loc=<0 0 0>, rot=<0 0 0 1>}
 base_jacketTop_l {parent=spine_04, loc=<6.0897 11.1972 -6.5604>, rot=<-0.1979 -0.94 -0.272 -0.0573>}
 dyn_jacketTop_l {parent=base_jacketTop_l, loc=<0 0 0>, rot=<0 0 0 1>}
 base_jacketTop_r {parent=spine_04, loc=<6.0895 11.1972 6.5604>, rot=<0.272 -0.0573 0.1979 -0.94>}
 dyn_jacketTop_r {parent=base_jacketTop_r, loc=<0 0 0>, rot=<0 0 0 1>}
 thigh_l {parent=pelvis, loc=<-1.0855 -0.2109 -8.7053>, rot=<0.0009 0.0355 0.0155 -0.9993>}
 calf_l {parent=thigh_l, loc=<-42.5284 -0.0019 -0.7297>, rot=<0.0006 -0.0161 -0.0391 -0.9991>}
 foot_l {parent=calf_l, loc=<-46.7401 -2.5014 -0.592>, rot=<0.0427 -0.0238 0.0353 -0.9982>}
 ball_l {parent=foot_l, loc=<-3.5658 10.8066 -0.1544>, rot=<0.0001 -0.0001 -0.7063 -0.7079>}
 calf_twist_02_l {parent=calf_l, loc=<-19.1546 -1.7943 -0.2638>, rot=<0.0237 0.0095 -0.0035 -0.9997>}
 calf_twist_01_l {parent=calf_l, loc=<-35.8483 -1.9197 -0.5794>, rot=<0.0237 0.0095 -0.0035 -0.9997>}
 deform_knee_in_base_l {parent=calf_l, loc=<0.8031 -5.0231 0.2456>, rot=<0 0 0 -1>}
 deform_knee_in_l {parent=deform_knee_in_base_l, loc=<0 0 0>, rot=<0 0 0 1>}
 deform_knee_out_base_l {parent=calf_l, loc=<0 4 0>, rot=<0 0 0 -1>}
 deform_knee_out_l {parent=deform_knee_out_base_l, loc=<0 0 0>, rot=<0 0 0 1>}
 thigh_twist_01_l {parent=thigh_l, loc=<-27.952 0 0>, rot=<0.0013 0 0 -1>}
 deform_groin_base_l {parent=thigh_l, loc=<0.4801 5.6776 -0.2486>, rot=<0 0 0 -1>}
 deform_groin_l {parent=deform_groin_base_l, loc=<0 0 0>, rot=<0 0 0 1>}
 calf_l_null {parent=thigh_l, type=Null, loc=<-42.5284 -0.0019 -0.7297>, rot=<0.0006 -0.0161 -0.0391 -0.9991>}
 thigh_r {parent=pelvis, loc=<-1.0855 -0.2109 8.7053>, rot=<0.0355 -0.0009 0.9993 0.0155>}
 calf_r {parent=thigh_r, loc=<42.5284 0.0019 0.7297>, rot=<0.0006 -0.0161 -0.0391 -0.9991>}
 foot_r {parent=calf_r, loc=<46.7401 2.5014 0.592>, rot=<0.0427 -0.0238 0.0353 -0.9982>}
 ball_r {parent=foot_r, loc=<3.5658 -10.8066 0.1544>, rot=<0.0001 -0.0001 -0.7063 -0.7079>}
 calf_twist_02_r {parent=calf_r, loc=<19.1546 1.7943 0.2638>, rot=<0.0237 0.0095 -0.0035 -0.9997>}
 calf_twist_01_r {parent=calf_r, loc=<35.8483 1.9197 0.5794>, rot=<0.0237 0.0095 -0.0035 -0.9997>}
 deform_knee_in_base_r {parent=calf_r, loc=<-0.8031 5.0231 -0.2457>, rot=<0 0 0 -1>}
 deform_knee_in_r {parent=deform_knee_in_base_r, loc=<0 0 0>, rot=<0 0 0 1>}
 deform_knee_out_base_r {parent=calf_r, loc=<0 -4 0>, rot=<0 0 0 -1>}
 deform_knee_out_r {parent=deform_knee_out_base_r, loc=<0 0 0>, rot=<0 0 0 1>}
 thigh_twist_01_r {parent=thigh_r, loc=<27.952 0 0>, rot=<0.0013 0 0 -1>}
 deform_groin_base_r {parent=thigh_r, loc=<-0.4801 -5.6776 0.2486>, rot=<0 0 0 -1>}
 deform_groin_r {parent=deform_groin_base_r, loc=<0 0 0>, rot=<0 0 0 1>}
 calf_r_null {parent=thigh_r, type=Null, loc=<42.5284 0.0019 0.7297>, rot=<0.0006 -0.0161 -0.0391 -0.9991>}
 deform_glute_base_l {parent=pelvis, loc=<-7.3466 -9.0213 -9.1245>, rot=<0.0172 0.031 -0.4477 -0.8935>}
 deform_glute_l {parent=deform_glute_base_l, loc=<0 0 0>, rot=<0 0 0 1>}
 deform_glute_base_r {parent=pelvis, loc=<-7.3464 -9.0213 9.1245>, rot=<0.031 -0.0172 0.8935 -0.4477>}
 deform_glute_r {parent=deform_glute_base_r, loc=<0 0 0>, rot=<0 0 0 1>}
 base_belt_l {parent=pelvis, loc=<7.2262 -1.0082 -6.7601>, rot=<-0.0195 -0.7068 -0.7065 -0.0287>}
 dyn_belt_l {parent=base_belt_l, loc=<0 0 0>, rot=<0 0 0 1>}
 base_belt_r {parent=pelvis, loc=<7.2264 -1.0082 6.7601>, rot=<-0.0195 -0.7068 0.7065 0.0287>}
 dyn_belt_r {parent=base_belt_r, loc=<0 0 0>, rot=<0 0 0 1>}
 base_bag_l {parent=pelvis, loc=<2.4782 -1.9379 -13.9315>, rot=<-0.0179 -0.9998 0 0>}
 dyn_bag_l {parent=base_bag_l, loc=<0 0 0>, rot=<0 0 0 1>}
 base_charm_01_b {parent=pelvis, loc=<4.2726 5.8363 -11.4745>, rot=<-0.0657 -0.8718 -0.4723 -0.1118>}
 dyn_charm_01_b {parent=base_charm_01_b, loc=<0 0 0>, rot=<0 0 0 1>}
 dyn_charm_02_b {parent=dyn_charm_01_b, loc=<4.1649 0 0>, rot=<0 0 0 1>}
 dyn_charm_03_b {parent=dyn_charm_02_b, loc=<4.3622 0 0>, rot=<0 0 0 1>}
 base_charm_01_a {parent=pelvis, loc=<4.2803 7.2817 -7.1446>, rot=<-0.0934 -0.8687 -0.4749 -0.1055>}
 dyn_charm_01_a {parent=base_charm_01_a, loc=<0 0 0>, rot=<0 0 0 1>}
 dyn_charm_02_a {parent=dyn_charm_01_a, loc=<2.8925 0 0>, rot=<0 0 0 1>}
 dyn_charm_03_a {parent=dyn_charm_02_a, loc=<3.0898 0 0>, rot=<0 0 0 1>}
 base_beltLower_r {parent=pelvis, loc=<2.1866 -0.6646 8.6572>, rot=<-0.0195 -0.7068 0.7065 0.0287>}
 dyn_beltLower_r {parent=base_beltLower_r, loc=<0 0 0>, rot=<0 0 0 1>}
 base_katana_r {parent=dyn_beltLower_r, loc=<5.673 -7.3286 1.8088>, rot=<0.007 0.4722 0.0679 -0.8789>}
 dyn_katana_r {parent=base_katana_r, loc=<0 0 0>, rot=<0 0 0 1>}
 base_mask_a_r {parent=dyn_beltLower_r, loc=<5.5543 -9.9754 1.5888>, rot=<-0.7408 0.0666 -0.0496 -0.6666>}
 dyn_mask_a_r {parent=base_mask_a_r, loc=<0 0 0>, rot=<0 0 0 1>}
 base_mask_b_r {parent=dyn_beltLower_r, loc=<7.2131 -8.1846 0.0426>, rot=<-0.8054 0.0614 0.0386 -0.5883>}
 dyn_mask_b01_r {parent=base_mask_b_r, loc=<0 0 0>, rot=<0 0 0 1>}
 dyn_mask_b02_r {parent=dyn_mask_b01_r, loc=<5.6168 0 0>, rot=<0 0 0 1>}
 dyn_mask_b03_r {parent=dyn_mask_b02_r, loc=<7.2082 0 0>, rot=<0 0 0 1>}
 base_skirt_01_bck {parent=pelvis, loc=<4.6678 -9.7864 -0.7526>, rot=<0.0779 -0.7028 0.7028 -0.0779>}
 dyn_skirt_01_bck {parent=base_skirt_01_bck, loc=<0 0 0>, rot=<0 0 0 1>}
 dyn_skirt_02_bck {parent=dyn_skirt_01_bck, loc=<10.3902 0 0>, rot=<0 -0.0691 0 -0.9976>}
 dyn_skirt_03_bck {parent=dyn_skirt_02_bck, loc=<16.3897 0 0>, rot=<0 -0.0037 0 -1>}
 dyn_skirt_04_bck {parent=dyn_skirt_03_bck, loc=<16.7296 0 0>, rot=<0 -0.0143 0 -0.9999>}
 dyn_skirt_05_bck {parent=dyn_skirt_04_bck, loc=<10.8228 0 0>, rot=<0 0 0 1>}
 base_skirt_01_bck_l {parent=pelvis, loc=<4.8804 -7.3366 -8.9344>, rot=<0.0126 -0.8224 0.5547 -0.1258>}
 dyn_skirt_01_bck_l {parent=base_skirt_01_bck_l, loc=<0 0 0>, rot=<0 0 0 1>}
 dyn_skirt_02_bck_l {parent=dyn_skirt_01_bck_l, loc=<10.3902 0.0343 0.1943>, rot=<0 -0.0681 0.012 -0.9976>}
 dyn_skirt_03_bck_l {parent=dyn_skirt_02_bck_l, loc=<16.3897 0 0>, rot=<0 -0.0036 0.0006 -1>}
 dyn_skirt_04_bck_l {parent=dyn_skirt_03_bck_l, loc=<15.5728 0.0058 0.0326>, rot=<0 -0.0141 0.0025 -0.9999>}
 dyn_skirt_05_bck_l {parent=dyn_skirt_04_bck_l, loc=<11.271 0 0>, rot=<0 0 0 1>}
 base_skirt_01_bck_r {parent=pelvis, loc=<4.8801 -7.3366 8.9344>, rot=<0.1258 -0.5547 0.8224 -0.0126>}
 dyn_skirt_01_bck_r {parent=base_skirt_01_bck_r, loc=<0 0 0>, rot=<0 0 0 1>}
 dyn_skirt_02_bck_r {parent=dyn_skirt_01_bck_r, loc=<10.39 -0.0343 0.1943>, rot=<0 -0.0681 -0.012 -0.9976>}
 dyn_skirt_03_bck_r {parent=dyn_skirt_02_bck_r, loc=<16.3897 0 0.0001>, rot=<0 -0.0036 -0.0006 -1>}
 dyn_skirt_04_bck_r {parent=dyn_skirt_03_bck_r, loc=<15.5729 -0.0057 0.0326>, rot=<0 -0.0141 -0.0025 -0.9999>}
 dyn_skirt_05_bck_r {parent=dyn_skirt_04_bck_r, loc=<11.271 0 0>, rot=<0 0 0 1>}
 base_skirt_01_fr {parent=pelvis, loc=<4.2892 7.1212 -0.925>, rot=<0.0583 0.7047 0.7047 0.0583>}
 dyn_skirt_01_fr {parent=base_skirt_01_fr, loc=<0 0 0>, rot=<0 0 0 1>}
 dyn_skirt_02_fr {parent=dyn_skirt_01_fr, loc=<10.2935 0 0>, rot=<0 -0.0229 0 -0.9997>}
 dyn_skirt_03_fr {parent=dyn_skirt_02_fr, loc=<15.5632 0 0>, rot=<0 -0.0166 0 -0.9999>}
 dyn_skirt_04_fr {parent=dyn_skirt_03_fr, loc=<16.1401 0 0>, rot=<0 0.0011 0 -1>}
 dyn_skirt_05_fr {parent=dyn_skirt_04_fr, loc=<10.834 0 0>, rot=<0 0 0 1>}
 base_skirt_01_front_l {parent=pelvis, loc=<4.7392 4.3977 -8.7662>, rot=<-0.0329 -0.8447 -0.5259 -0.094>}
 dyn_skirt_01_front_l {parent=base_skirt_01_front_l, loc=<0 0 0>, rot=<0 0 0 1>}
 dyn_skirt_02_front_l {parent=dyn_skirt_01_front_l, loc=<10.2935 0 0>, rot=<0 -0.0229 0 -0.9997>}
 dyn_skirt_03_front_l {parent=dyn_skirt_02_front_l, loc=<15.5632 0 0>, rot=<0 -0.0166 0 -0.9999>}
 dyn_skirt_04_front_l {parent=dyn_skirt_03_front_l, loc=<16.1401 0 0>, rot=<0 0.0011 0 -1>}
 dyn_skirt_05_front_l {parent=dyn_skirt_04_front_l, loc=<10.834 0 0>, rot=<0 0 0 1>}
 base_skirt_01_front_r {parent=pelvis, loc=<4.7393 4.3977 8.7662>, rot=<0.094 0.5259 0.8447 0.0329>}
 dyn_skirt_01_front_r {parent=base_skirt_01_front_r, loc=<0 0 0>, rot=<0 0 0 1>}
 dyn_skirt_02_front_r {parent=dyn_skirt_01_front_r, loc=<10.2936 0 0>, rot=<0 -0.0229 0 -0.9997>}
 dyn_skirt_03_front_r {parent=dyn_skirt_02_front_r, loc=<15.5632 0 0>, rot=<0 -0.0166 0 -0.9999>}
 dyn_skirt_04_front_r {parent=dyn_skirt_03_front_r, loc=<16.1401 -0.0001 0>, rot=<0 0.0011 0 -1>}
 dyn_skirt_05_front_r {parent=dyn_skirt_04_front_r, loc=<10.834 0 0>, rot=<0 0 0 1>}
 base_skirt_01_l {parent=pelvis, loc=<5.2809 -1.7557 -11.5646>, rot=<-0.0177 -0.9888 0.0027 -0.1479>}
 dyn_skirt_01_l {parent=base_skirt_01_l, loc=<0 0 0>, rot=<0 0 0 1>}
 dyn_skirt_02_l {parent=dyn_skirt_01_l, loc=<10.2577 0 0>, rot=<0 -0.0432 0 -0.9991>}
 dyn_skirt_03_l {parent=dyn_skirt_02_l, loc=<16.0397 0 0>, rot=<0 -0.0044 0 -1>}
 dyn_skirt_04_l {parent=dyn_skirt_03_l, loc=<16.0994 0 0>, rot=<0 -0.0171 0 -0.9999>}
 dyn_skirt_05_l {parent=dyn_skirt_04_l, loc=<11.2774 0 0>, rot=<0 0 0 1>}
 base_skirt_01_r {parent=pelvis, loc=<5.2789 -1.7557 11.5633>, rot=<0.149 -0.0027 0.9887 0.0178>}
 dyn_skirt_01_r {parent=base_skirt_01_r, loc=<0 0 0>, rot=<0 0 0 1>}
 dyn_skirt_02_r {parent=dyn_skirt_01_r, loc=<10.3142 0 0>, rot=<0.0003 -0.0384 0.008 -0.9992>}
 dyn_skirt_03_r {parent=dyn_skirt_02_r, loc=<15.6217 0 0>, rot=<0.0001 -0.0048 0.0007 -1>}
 dyn_skirt_04_r {parent=dyn_skirt_03_r, loc=<15.9462 0 0>, rot=<0.0006 -0.0272 0.0031 -0.9996>}
 dyn_skirt_05_r {parent=dyn_skirt_04_r, loc=<11.2322 0 0>, rot=<0 0 0 1>}
 base_bow_m {parent=pelvis, loc=<6.7318 -12.6139 -0.4027>, rot=<-0.0127 -0.707 0.707 0.0127>}
 dyn_bow_m {parent=base_bow_m, loc=<0 0 0>, rot=<0 0 0 1>}
 base_bow01_r {parent=dyn_bow_m, loc=<2.3386 -3.6597 0.2295>, rot=<-0.1729 0.1821 0.0475 -0.9668>}
 dyn_bow01_r {parent=base_bow01_r, loc=<0 0 0>, rot=<0 0 0 1>}
 dyn_bow02_r {parent=dyn_bow01_r, loc=<6.5136 0 0>, rot=<0.0088 -0.0411 0.028 -0.9987>}
 dyn_bow03_r {parent=dyn_bow02_r, loc=<7.0028 0.2006 0>, rot=<0 0 0 1>}
 base_bow01_l {parent=dyn_bow_m, loc=<2.8205 3.3056 0.0419>, rot=<0.1584 0.2069 -0.0867 -0.9615>}
 dyn_bow01_l {parent=base_bow01_l, loc=<0 0 0>, rot=<0 0 0 1>}
 dyn_bow02_l {parent=dyn_bow01_l, loc=<6.5136 0 0>, rot=<0.0088 -0.0411 0.028 -0.9987>}
 dyn_bow03_l {parent=dyn_bow02_l, loc=<7.0028 0.2006 0>, rot=<0 0 0 1>}
 thigh_l_null {parent=pelvis, type=Null, loc=<-1.0855 -0.2109 -8.7053>, rot=<0.0009 0.0355 0.0155 -0.9993>}
 thigh_r_null {parent=pelvis, type=Null, loc=<-1.0855 -0.2109 8.7053>, rot=<0.0355 -0.0009 0.9993 0.0155>}
 base_floor_m {parent=root, loc=<0 0 -2.895>, rot=<0 0 0.0112 -0.9999>}
 thigh_l_ctrl {controlType=EULER_TRANSFORM, type=Control, loc=<0 0 0>, rot=<0 0 0 1>}
 calf_l_ctrl {controlType=EULER_TRANSFORM, parent=thigh_l_ctrl, type=Control, loc=<0 0 0>, rot=<0 0 0 1>}
 thigh_r_ctrl {controlType=EULER_TRANSFORM, type=Control, loc=<0 0 0>, rot=<0 0 0 1>}
 calf_r_ctrl {controlType=EULER_TRANSFORM, parent=thigh_r_ctrl, type=Control, loc=<0 0 0>, rot=<0 0 0 1>}
 upperarm_l_ctrl {controlType=EULER_TRANSFORM, type=Control, loc=<0 0 0>, rot=<0 0 0 1>}
 upperarm_r_ctrl {controlType=EULER_TRANSFORM, type=Control, loc=<0 0 0>, rot=<0 0 0 1>}
 neck_01_ctrl {controlType=EULER_TRANSFORM, type=Control, loc=<0 0 0>, rot=<0 0 0 1>}
 spine_05_ctrl {controlType=EULER_TRANSFORM, type=Control, loc=<0 0 0>, rot=<0 0 0 1>}
]
variables [
 ArmUpSkydiveL {type=double, default=0}
 ArmUpSkydiveR {type=double, default=0}
 Calf_L_Rotation {type=double, default=0}
 Calf_R_Rotation {type=double, default=0}
 IsEmoting {type=bool, default=False}
 IsInAir {type=bool, default=False}
 KatanaTransform {type=double, default=0}
 NotInAir {type=double, default=0}
 PelvisToGround {type=double, default=0}
 SpineRotRight {type=double, default=0}
 Thigh_L_BackRotation {type=double, default=0}
 Thigh_L_In_SideRotation {type=double, default=0}
 Thigh_L_SideRotation {type=double, default=0}
 Thigh_R_BackRotation {type=double, default=0}
 Thigh_R_SideRotation {type=double, default=0}
 headLookUp {type=double, default=0}
 inFloating {type=double, default=0}
 isCrouching {type=double, default=0}
 isMoving {type=double, default=0}
 isSkydiving {type=double, default=0}
 openParachute {type=double, default=0}
 skydiveThigBackL {type=double, default=0}
 skydiveThigBackR {type=double, default=0}
 spineRotBack {type=double, default=0}
]
