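# Physics asset for the masks mesh of the F_MED_AnodizedMetal female body cosmetic, with capsule bodies on right-side bones only.
# Two dynamic mask chains hang from tiny base capsules, one a single link and the other three links, using limited swing and zero-degree twist limits.
# Skirt, calf and foot capsules are also present; only foot-to-calf is constrained, fully locked, while the skirt bodies have no constraints.
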PhysicsAsset /BRCosmetics/Characters/Player/Female/Medium/Bodies/F_MED_AnodizedMetal/Meshes/F_MED_AnodizedMetal_Masks_Physics
bodies {
 base_mask_a_r [{shapeType=Capsule, length=0.505, radius=0.505, shapeName=base_mask_a_r_capsule, rotation=<rot 0 -90 0>}]
 base_mask_b_r [{shapeType=Capsule, length=0.505, radius=0.505, shapeName=base_mask_b_r_capsule, rotation=<rot 0 -90 0>}]
 calf_r [{shapeType=Capsule, length=21.5004, radius=7.7451, center=<21.2963 1.4862 0.6156>, rotation=<rot 0 85.5948 -88.0835>}]
 dyn_mask_a_r [{
   shapeType=Capsule
   length=18.8816
   radius=3.6935
   shapeName=dyn_mask_a_r_capsule
   center=<10.0437 -0.0777 5.0163>
   rotation=<rot 0 -6.3201 90>
  }]
 dyn_mask_b01_r [{shapeType=Capsule, length=3.849, radius=3.4642, shapeName=dyn_mask_b01_r_capsule, center=<4.0272 0 1.3334>, rotation=<rot 90 0 0>}]
 dyn_mask_b02_r [{shapeType=Capsule, length=3.849, radius=3.4642, shapeName=dyn_mask_b02_r_capsule, center=<5.4287 0 0>, rotation=<rot 90 0 0>}]
 dyn_mask_b03_r [{shapeType=Capsule, length=6.903, radius=4.0785, shapeName=dyn_mask_b03_r_capsule, center=<6.3179 0 2.2076>, rotation=<rot 90 0 0>}]
 dyn_skirt_02_bck_r [{shapeType=Capsule, length=7.3489, radius=6.301, shapeName=dyn_skirt_02_bck_r_capsule, center=<4.8686 0 1.6335>, rotation=<rot 90 0 0>}]
 dyn_skirt_02_r [{shapeType=Capsule, length=10, radius=5, shapeName=dyn_skirt_02_r_capsule, center=<0 0 -4.4195>, rotation=<rot 90 0 0>}]
 dyn_skirt_03_bck_r [{shapeType=Capsule, length=8.5241, radius=6.5865, shapeName=dyn_skirt_03_bck_r_capsule, center=<4.8686 0 1.6335>, rotation=<rot 90 0 0>}]
 dyn_skirt_03_r [{shapeType=Capsule, length=10, radius=5, shapeName=dyn_skirt_03_r_capsule, center=<2.1743 0 -4.4195>, rotation=<rot 90 0 0>}]
 dyn_skirt_04_r [{shapeType=Capsule, length=10, radius=5, shapeName=dyn_skirt_04_r_capsule, center=<0 0 -4.4195>, rotation=<rot 90 0 0>}]
 foot_r [{shapeType=Capsule, length=14.2812, radius=8.4687, center=<1.964 -6.1161 0.102>, rotation=<rot 0 -12.2434 -87.0442>}]
}
constraints [
 {
  bone1Name=foot_r
  bone2Name=calf_r
  swing1LimitDegrees=45
  swing1Motion=Locked
  swing2LimitDegrees=45
  swing2Motion=Locked
  twistLimitDegrees=45
  twistMotion=Locked
 }
 {bone1Name=dyn_mask_a_r, bone2Name=base_mask_a_r, swing1LimitDegrees=30, swing2LimitDegrees=40, twistLimitDegrees=0}
 {bone1Name=dyn_mask_b01_r, bone2Name=base_mask_b_r, swing1LimitDegrees=10, swing2LimitDegrees=30, twistLimitDegrees=0}
 {bone1Name=dyn_mask_b02_r, bone2Name=dyn_mask_b01_r, swing1LimitDegrees=10, swing2LimitDegrees=30, twistLimitDegrees=0}
 {bone1Name=dyn_mask_b03_r, bone2Name=dyn_mask_b02_r, swing1LimitDegrees=10, swing2LimitDegrees=30, twistLimitDegrees=0}
]
