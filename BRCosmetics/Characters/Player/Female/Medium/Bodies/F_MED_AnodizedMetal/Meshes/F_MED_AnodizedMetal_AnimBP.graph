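# Post-process style cosmetic AnimBP for the AnodizedMetal female body, adding secondary motion to skirt, masks, bows, charms and jacket.
# The parent mesh pose feeds a movement-state-driven Control Rig, then cached skirt setups blended between base, knee-slide and baseball-slide.
# Bone offsets, RigidBody and AnimDynamics refine the result.
# A final per-bone layer blends DefaultSlot onto the right katana branch during custom front-end idle.
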AnimBlueprint /BRCosmetics/Characters/Player/Female/Medium/Bodies/F_MED_AnodizedMetal/Meshes/F_MED_AnodizedMetal_AnimBP
target_skeleton=&/F_MED_AnodizedMetal_Skeleton
graphs [
 AnimGraph {nodes [
   Copy Pose From Mesh() {bUseAttachedParent=true}
   n1 = "Transform (Modify) Bone - Bone: base_mask_a_r"!(
    ComponentPose=Local To Component(LocalPose=Use cached pose 'CRCache_Skirt_Dynamics'())
    bAlphaBoolEnabled:bool=True
    Translation="0,0,0"
    Rotation="10,0,0"
    Scale="1,1,1"
   ) -> (Pose) {
    alphaInputType=Bool
    rotationMode=BMM_Additive
    rotationSpace=BCS_BoneSpace
    alphaBoolBlend {blendInTime=0.2, blendOption=Linear, blendOutTime=0.2}
    boneToModify {boneName=base_mask_a_r}
    rotation {pitch=10}
   }
   n2 = "Transform (Modify) Bone - Bone: dyn_mask_b01_r"!(ComponentPose=@n1, bAlphaBoolEnabled:bool=True, Translation="0,0,0", Rotation="5,0,0", Scale="1,1,1") -> (Pose) {
    alphaInputType=Bool
    rotationMode=BMM_Additive
    rotationSpace=BCS_BoneSpace
    alphaBoolBlend {blendInTime=0.2, blendOption=Linear, blendOutTime=0.2}
    boneToModify {boneName=dyn_mask_b01_r}
    rotation {pitch=5}
   }
   n3 = RigidBody!(ComponentPose=@n2, Alpha, OverrideWorldGravity:Vector="0,0,-980", ExternalForce) -> (Pose) {
    bOverrideWorldGravity=true
    overridePhysicsAsset=&/F_MED_AnodizedMetal_Masks_Physics
    simulationSpace=BaseBoneSpace
    alphaBoolBlend {blendOption=Linear}
    baseBoneRef {boneName=pelvis}
    componentAppliedLinearAccClamp=<3000 3000 3000>
    componentLinearAccScale=<0.3 0.3 0.3>
    componentLinearVelScale=<0.3 0.3 0.3>
    simSpaceSettings {maxAngularAcceleration=1000, maxAngularVelocity=3, maxLinearAcceleration=700, maxLinearVelocity=500, velocityScaleZ=0}
   }
   n4 = "Transform (Modify) Bone - Bone: dyn_shoulderPad_l"!(ComponentPose=@n3, bAlphaBoolEnabled:bool=True, Translation="2.812,-0.060,1.251", Rotation="0.000,-0.000,-2.730") -> (Pose) {
    alphaInputType=Bool
    rotationMode=BMM_Additive
    rotationSpace=BCS_BoneSpace
    scaleMode=BMM_Additive
    scaleSpace=BCS_BoneSpace
    translationMode=BMM_Additive
    translationSpace=BCS_BoneSpace
    alphaBoolBlend {blendInTime=0.2, blendOption=Linear, blendOutTime=0.2}
    boneToModify {boneName=dyn_shoulderPad_l}
    rotation=<rot 0 0 -2.73>
    translation=<2.812 -0.06 1.251>
   }
   n5 = "Transform (Modify) Bone - Bone: dyn_shoulderPad_r"!(ComponentPose=@n4, bAlphaBoolEnabled:bool=True, Translation="-4.256,1.874,0.272", Rotation="0.000,-0.000,-15.116") -> (Pose) {
    alphaInputType=Bool
    rotationMode=BMM_Additive
    rotationSpace=BCS_BoneSpace
    scaleMode=BMM_Additive
    scaleSpace=BCS_BoneSpace
    translationMode=BMM_Additive
    translationSpace=BCS_BoneSpace
    alphaBoolBlend {blendInTime=0.2, blendOption=Linear, blendOutTime=0.2}
    boneToModify {boneName=dyn_shoulderPad_r}
    rotation {roll=-15.116}
    translation=<-4.256 1.874 0.272>
   }
   n6 = Anim Dynamics - Bone: dyn_bow01_r!(ComponentPose=@n5, Alpha=1, ComponentLinearAccScale:Vector="1,1,1", ComponentLinearVelScale:Vector="1,1,1", ExternalForce="0,0,0") -> (Pose) {
    angularDampingOverride=1
    angularSpringConstant=15000
    bAngularSpring=true
    bOverrideAngularDamping=true
    bUseGravityOverride=true
    gravityScale=0.35
    alphaBoolBlend {blendOption=Linear}
    boundBone {boneName=dyn_bow01_r}
    componentAppliedLinearAccClamp=<9000 9000 9000>
    componentLinearAccScale=<1 1 1>
    componentLinearVelScale=<1 1 1>
    gravityOverride {z=-980}
    physicsBodyDefinitions [{
      boundBone {boneName=dyn_bow01_r}
      boxExtents=<30 30 30>
      constraintSetup {angularLimitsMax=<15 0 15>, angularLimitsMin=<-15 -50 -15>, angularTarget=<1 0 0>}
      localJointOffset=<4 0 0>
     }]
    relativeSpaceBone {boneName=pelvis}
   }
   n7 = Anim Dynamics - Bone: dyn_bow02_r!(ComponentPose=@n6, Alpha=1, ComponentLinearAccScale:Vector="1,1,1", ComponentLinearVelScale:Vector="1,1,1", ExternalForce="0,0,0") -> (Pose) {
    angularDampingOverride=1
    angularSpringConstant=15000
    bAngularSpring=true
    bOverrideAngularDamping=true
    bUseGravityOverride=true
    gravityScale=0.35
    alphaBoolBlend {blendOption=Linear}
    boundBone {boneName=dyn_bow02_r}
    componentAppliedLinearAccClamp=<9000 9000 9000>
    componentLinearAccScale=<1 1 1>
    componentLinearVelScale=<1 1 1>
    gravityOverride {z=-980}
    physicsBodyDefinitions [{
      boundBone {boneName=dyn_bow02_r}
      boxExtents=<30 30 30>
      constraintSetup {angularLimitsMax=<15 0 15>, angularLimitsMin=<-15 -30 -15>, angularTarget=<1 0 0>}
      localJointOffset=<4 0 0>
     }]
    relativeSpaceBone {boneName=pelvis}
   }
   n8 = Anim Dynamics - Bone: dyn_bow03_r!(ComponentPose=@n7, Alpha=1, ComponentLinearAccScale:Vector="1,1,1", ComponentLinearVelScale:Vector="1,1,1", ExternalForce="0,0,0") -> (Pose) {
    angularDampingOverride=1
    angularSpringConstant=15000
    bAngularSpring=true
    bOverrideAngularDamping=true
    bUseGravityOverride=true
    gravityScale=0.35
    alphaBoolBlend {blendOption=Linear}
    boundBone {boneName=dyn_bow03_r}
    componentAppliedLinearAccClamp=<9000 9000 9000>
    componentLinearAccScale=<1 1 1>
    componentLinearVelScale=<1 1 1>
    gravityOverride {z=-980}
    physicsBodyDefinitions [{
      boundBone {boneName=dyn_bow03_r}
      boxExtents=<30 30 30>
      constraintSetup {angularLimitsMax=<15 0 15>, angularLimitsMin=<-15 -30 -15>, angularTarget=<1 0 0>}
      localJointOffset=<4 0 0>
     }]
    relativeSpaceBone {boneName=pelvis}
   }
   n9 = Anim Dynamics - Bone: dyn_bow01_l!(ComponentPose=@n8, Alpha=1, ComponentLinearAccScale:Vector="1,1,1", ComponentLinearVelScale:Vector="1,1,1", ExternalForce="0,0,0") -> (Pose) {
    angularDampingOverride=1
    angularSpringConstant=10000
    bAngularSpring=true
    bOverrideAngularDamping=true
    bUseGravityOverride=true
    gravityScale=0.35
    alphaBoolBlend {blendOption=Linear}
    boundBone {boneName=dyn_bow01_l}
    componentAppliedLinearAccClamp=<9000 9000 9000>
    componentLinearAccScale=<1 1 1>
    componentLinearVelScale=<1 1 1>
    gravityOverride {z=-980}
    physicsBodyDefinitions [{
      boundBone {boneName=dyn_bow01_l}
      boxExtents=<30 30 30>
      constraintSetup {angularLimitsMax=<15 0 15>, angularLimitsMin=<-15 -50 -15>, angularTarget=<1 0 0>}
      localJointOffset=<4 0 0>
     }]
    relativeSpaceBone {boneName=pelvis}
   }
   n10 = Anim Dynamics - Bone: dyn_bow02_l!(ComponentPose=@n9, Alpha=1, ComponentLinearAccScale:Vector="1,1,1", ComponentLinearVelScale:Vector="1,1,1", ExternalForce="0,0,0") -> (Pose) {
    angularDampingOverride=1
    angularSpringConstant=10000
    bAngularSpring=true
    bOverrideAngularDamping=true
    bUseGravityOverride=true
    gravityScale=0.35
    alphaBoolBlend {blendOption=Linear}
    boundBone {boneName=dyn_bow02_l}
    componentAppliedLinearAccClamp=<9000 9000 9000>
    componentLinearAccScale=<1 1 1>
    componentLinearVelScale=<1 1 1>
    gravityOverride {z=-980}
    physicsBodyDefinitions [{
      boundBone {boneName=dyn_bow02_l}
      boxExtents=<30 30 30>
      constraintSetup {angularLimitsMax=<15 0 15>, angularLimitsMin=<-15 -30 -15>, angularTarget=<1 0 0>}
      localJointOffset=<4 0 0>
     }]
    relativeSpaceBone {boneName=pelvis}
   }
   n11 = Anim Dynamics - Bone: dyn_bow03_l!(ComponentPose=@n10, Alpha=1, ComponentLinearAccScale:Vector="1,1,1", ComponentLinearVelScale:Vector="1,1,1", ExternalForce="0,0,0") -> (Pose) {
    angularDampingOverride=1
    angularSpringConstant=10000
    bAngularSpring=true
    bOverrideAngularDamping=true
    bUseGravityOverride=true
    gravityScale=0.35
    alphaBoolBlend {blendOption=Linear}
    boundBone {boneName=dyn_bow03_l}
    componentAppliedLinearAccClamp=<9000 9000 9000>
    componentLinearAccScale=<1 1 1>
    componentLinearVelScale=<1 1 1>
    gravityOverride {z=-980}
    physicsBodyDefinitions [{
      boundBone {boneName=dyn_bow03_l}
      boxExtents=<30 30 30>
      constraintSetup {angularLimitsMax=<15 0 15>, angularLimitsMin=<-15 -30 -15>, angularTarget=<1 0 0>}
      localJointOffset=<4 0 0>
     }]
    relativeSpaceBone {boneName=pelvis}
   }
   n12 = Anim Dynamics - Bone: dyn_charm_01_a(ComponentPose=@n11, Alpha=0, ExternalForce="0,0,0") {
    angularDampingOverride=1
    angularSpringConstant=20000
    bAngularSpring=true
    bOverrideAngularDamping=true
    bUseGravityOverride=true
    gravityScale=0.35
    simulationSpace=World
    alphaBoolBlend {blendInTime=0.2, blendOption=Linear, blendOutTime=0.2}
    alphaScaleBiasClamp {bInterpResult=true}
    boundBone {boneName=dyn_charm_01_a}
    componentAppliedLinearAccClamp=<8000 8000 8000>
    componentLinearAccScale=<0.5 0.5 0.5>
    componentLinearVelScale=<0.5 0.5 0.5>
    gravityOverride {z=-980}
    physicsBodyDefinitions [{
      boundBone {boneName=dyn_charm_01_a}
      boxExtents=<30 30 30>
      constraintSetup {angularLimitsMax=<0 0 15>, angularLimitsMin=<0 -40 -30>, angularTarget=<1 0 0>}
      localJointOffset=<4 0 0>
     }]
   }
   n13 = Anim Dynamics - Bone: dyn_charm_02_a(ComponentPose=@n12, Alpha=0, ExternalForce="0,0,0") {
    angularDampingOverride=1
    angularSpringConstant=20000
    bAngularSpring=true
    bOverrideAngularDamping=true
    bUseGravityOverride=true
    gravityScale=0.35
    simulationSpace=World
    alphaBoolBlend {blendInTime=0.2, blendOption=Linear, blendOutTime=0.2}
    alphaScaleBiasClamp {bInterpResult=true}
    boundBone {boneName=dyn_charm_02_a}
    componentAppliedLinearAccClamp=<8000 8000 8000>
    componentLinearAccScale=<0.5 0.5 0.5>
    componentLinearVelScale=<0.5 0.5 0.5>
    gravityOverride {z=-980}
    physicsBodyDefinitions [{
      boundBone {boneName=dyn_charm_02_a}
      boxExtents=<30 30 30>
      constraintSetup {angularLimitsMax=<0 0 15>, angularLimitsMin=<0 -40 -30>, angularTarget=<1 0 0>}
      localJointOffset=<4 0 0>
     }]
   }
   n14 = Anim Dynamics - Bone: dyn_charm_03_a(ComponentPose=@n13, Alpha=0, ExternalForce="0,0,0") {
    angularDampingOverride=1
    angularSpringConstant=20000
    bAngularSpring=true
    bOverrideAngularDamping=true
    bUseGravityOverride=true
    gravityScale=0.35
    simulationSpace=World
    alphaBoolBlend {blendInTime=0.2, blendOption=Linear, blendOutTime=0.2}
    alphaScaleBiasClamp {bInterpResult=true}
    boundBone {boneName=dyn_charm_03_a}
    componentAppliedLinearAccClamp=<8000 8000 8000>
    componentLinearAccScale=<0.5 0.5 0.5>
    componentLinearVelScale=<0.5 0.5 0.5>
    gravityOverride {z=-980}
    physicsBodyDefinitions [{
      boundBone {boneName=dyn_charm_03_a}
      boxExtents=<30 30 30>
      constraintSetup {angularLimitsMax=<0 0 15>, angularLimitsMin=<0 -40 -30>, angularTarget=<1 0 0>}
      localJointOffset=<4 0 0>
     }]
   }
   n15 = Anim Dynamics - Bone: dyn_charm_01_b(ComponentPose=@n14, Alpha=0, ExternalForce="0,0,0") {
    angularDampingOverride=1
    angularSpringConstant=20000
    bAngularSpring=true
    bOverrideAngularDamping=true
    bUseGravityOverride=true
    gravityScale=0.35
    simulationSpace=World
    alphaBoolBlend {blendInTime=0.2, blendOption=Linear, blendOutTime=0.2}
    alphaScaleBiasClamp {bInterpResult=true}
    boundBone {boneName=dyn_charm_01_b}
    componentAppliedLinearAccClamp=<8000 8000 8000>
    componentLinearAccScale=<0.5 0.5 0.5>
    componentLinearVelScale=<0.5 0.5 0.5>
    gravityOverride {z=-980}
    physicsBodyDefinitions [{
      boundBone {boneName=dyn_charm_01_b}
      boxExtents=<30 30 30>
      constraintSetup {angularLimitsMax=<0 0 15>, angularLimitsMin=<0 -40 -30>, angularTarget=<1 0 0>}
      localJointOffset=<4 0 0>
     }]
   }
   n16 = Anim Dynamics - Bone: dyn_charm_02_b(ComponentPose=@n15, Alpha=0, ExternalForce="0,0,0") {
    angularDampingOverride=1
    angularSpringConstant=20000
    bAngularSpring=true
    bOverrideAngularDamping=true
    bUseGravityOverride=true
    gravityScale=0.35
    simulationSpace=World
    alphaBoolBlend {blendInTime=0.2, blendOption=Linear, blendOutTime=0.2}
    alphaScaleBiasClamp {bInterpResult=true}
    boundBone {boneName=dyn_charm_02_b}
    componentAppliedLinearAccClamp=<8000 8000 8000>
    componentLinearAccScale=<0.5 0.5 0.5>
    componentLinearVelScale=<0.5 0.5 0.5>
    gravityOverride {z=-980}
    physicsBodyDefinitions [{
      boundBone {boneName=dyn_charm_02_b}
      boxExtents=<30 30 30>
      constraintSetup {angularLimitsMax=<0 0 15>, angularLimitsMin=<0 -40 -30>, angularTarget=<1 0 0>}
      localJointOffset=<4 0 0>
     }]
   }
   n17 = Anim Dynamics - Bone: dyn_charm_03_b(ComponentPose=@n16, Alpha=0, ExternalForce="0,0,0") {
    angularDampingOverride=1
    angularSpringConstant=20000
    bAngularSpring=true
    bOverrideAngularDamping=true
    bUseGravityOverride=true
    gravityScale=0.35
    simulationSpace=World
    alphaBoolBlend {blendInTime=0.2, blendOption=Linear, blendOutTime=0.2}
    alphaScaleBiasClamp {bInterpResult=true}
    boundBone {boneName=dyn_charm_03_b}
    componentAppliedLinearAccClamp=<8000 8000 8000>
    componentLinearAccScale=<0.5 0.5 0.5>
    componentLinearVelScale=<0.5 0.5 0.5>
    gravityOverride {z=-980}
    physicsBodyDefinitions [{
      boundBone {boneName=dyn_charm_03_b}
      boxExtents=<30 30 30>
      constraintSetup {angularLimitsMax=<0 0 15>, angularLimitsMin=<0 -40 -30>, angularTarget=<1 0 0>}
      localJointOffset=<4 0 0>
     }]
   }
   n18 = Anim Dynamics - Bone: dyn_jacket_fr_l!(
    ComponentPose=@n17
    Alpha=1
    ComponentLinearAccScale:Vector="0.5,0.5,0.5"
    ComponentLinearVelScale:Vector="0.5,0.5,0.5"
    ExternalForce="0,0,0"
   ) -> (Pose) {
    angularDampingOverride=1
    angularSpringConstant=15000
    bAngularSpring=true
    bOverrideAngularDamping=true
    bUseGravityOverride=true
    gravityScale=0.35
    alphaBoolBlend {blendOption=Linear}
    boundBone {boneName=dyn_jacket_fr_l}
    componentAppliedLinearAccClamp=<9000 9000 9000>
    componentLinearAccScale=<0.5 0.5 0.5>
    componentLinearVelScale=<0.5 0.5 0.5>
    gravityOverride {z=-980}
    physicsBodyDefinitions [{
      boundBone {boneName=dyn_jacket_fr_l}
      boxExtents=<30 30 30>
      constraintSetup {angularLimitsMax=<0 0 20>, angularLimitsMin=<0 -20 -20>, angularTarget=<1 0 0>}
      localJointOffset=<4 0 0>
     }]
    relativeSpaceBone {boneName=spine_04}
   }
   n19 = Anim Dynamics - Bone: dyn_jacket_side_l!(
    ComponentPose=@n18
    Alpha=1
    ComponentLinearAccScale:Vector="0.5,0.5,0.5"
    ComponentLinearVelScale:Vector="0.5,0.5,0.5"
    ExternalForce="0,0,0"
   ) -> (Pose) {
    angularDampingOverride=1
    angularSpringConstant=15000
    bAngularSpring=true
    bOverrideAngularDamping=true
    bUseGravityOverride=true
    gravityScale=0.35
    alphaBoolBlend {blendOption=Linear}
    boundBone {boneName=dyn_jacket_side_l}
    componentAppliedLinearAccClamp=<9000 9000 9000>
    componentLinearAccScale=<0.5 0.5 0.5>
    componentLinearVelScale=<0.5 0.5 0.5>
    gravityOverride {z=-980}
    physicsBodyDefinitions [{
      boundBone {boneName=dyn_jacket_side_l}
      boxExtents=<30 30 30>
      constraintSetup {angularLimitsMax=<0 0 20>, angularLimitsMin=<0 -20 -20>, angularTarget=<1 0 0>}
      localJointOffset=<4 0 0>
     }]
    relativeSpaceBone {boneName=spine_04}
   }
   n20 = Anim Dynamics - Bone: dyn_jacket_bck!(
    ComponentPose=@n19
    Alpha=1
    ComponentLinearAccScale:Vector="0.5,0.5,0.5"
    ComponentLinearVelScale:Vector="0.5,0.5,0.5"
    ExternalForce="0,0,0"
   ) -> (Pose) {
    angularDampingOverride=1
    angularSpringConstant=15000
    bAngularSpring=true
    bOverrideAngularDamping=true
    bUseGravityOverride=true
    gravityScale=0.35
    alphaBoolBlend {blendOption=Linear}
    boundBone {boneName=dyn_jacket_bck}
    componentAppliedLinearAccClamp=<9000 9000 9000>
    componentLinearAccScale=<0.5 0.5 0.5>
    componentLinearVelScale=<0.5 0.5 0.5>
    gravityOverride {z=-980}
    physicsBodyDefinitions [{
      boundBone {boneName=dyn_jacket_bck}
      boxExtents=<30 30 30>
      constraintSetup {angularLimitsMax=<0 0 0>, angularLimitsMin=<0 -20 0>, angularTarget=<1 0 0>}
      localJointOffset=<4 0 0>
     }]
    relativeSpaceBone {boneName=spine_04}
   }
   n21 = Anim Dynamics - Bone: dyn_jacket_fr_r!(
    ComponentPose=@n20
    Alpha=1
    ComponentLinearAccScale:Vector="0.5,0.5,0.5"
    ComponentLinearVelScale:Vector="0.5,0.5,0.5"
    ExternalForce="0,0,0"
   ) -> (Pose) {
    angularDampingOverride=1
    angularSpringConstant=15000
    bAngularSpring=true
    bOverrideAngularDamping=true
    bUseGravityOverride=true
    gravityScale=0.35
    alphaBoolBlend {blendOption=Linear}
    boundBone {boneName=dyn_jacket_fr_r}
    componentAppliedLinearAccClamp=<9000 9000 9000>
    componentLinearAccScale=<0.5 0.5 0.5>
    componentLinearVelScale=<0.5 0.5 0.5>
    gravityOverride {z=-980}
    physicsBodyDefinitions [{
      boundBone {boneName=dyn_jacket_fr_r}
      boxExtents=<30 30 30>
      constraintSetup {angularLimitsMax=<0 0 20>, angularLimitsMin=<0 -20 -20>, angularTarget=<1 0 0>}
      localJointOffset=<-4 0 0>
     }]
    relativeSpaceBone {boneName=spine_04}
   }
   n22 = Anim Dynamics - Bone: dyn_jacket_side_r!(
    ComponentPose=@n21
    Alpha=1
    ComponentLinearAccScale:Vector="0.5,0.5,0.5"
    ComponentLinearVelScale:Vector="0.5,0.5,0.5"
    ExternalForce="0,0,0"
   ) -> (Pose) {
    angularDampingOverride=1
    angularSpringConstant=15000
    bAngularSpring=true
    bOverrideAngularDamping=true
    bUseGravityOverride=true
    gravityScale=0.35
    alphaBoolBlend {blendOption=Linear}
    boundBone {boneName=dyn_jacket_side_r}
    componentAppliedLinearAccClamp=<9000 9000 9000>
    componentLinearAccScale=<0.5 0.5 0.5>
    componentLinearVelScale=<0.5 0.5 0.5>
    gravityOverride {z=-980}
    physicsBodyDefinitions [{
      boundBone {boneName=dyn_jacket_side_r}
      boxExtents=<30 30 30>
      constraintSetup {angularLimitsMax=<0 0 20>, angularLimitsMin=<0 -20 -20>, angularTarget=<1 0 0>}
      localJointOffset=<-4 0 0>
     }]
    relativeSpaceBone {boneName=spine_04}
   }
   Output Pose(Result=Layered blend per bone(
     BasePose=Component To Local(ComponentPose=@n22)
     BlendPoses_0=Slot 'DefaultSlot'()
     BlendWeights_0=Get Is Playing Custom FrontEnd Idle() -> (Is Playing Custom FrontEnd Idle:double)
    ) {layerSetup [{branchFilters [{boneName=dyn_katana_r}]}]}) {layerGroup=None}
   n23 = Control Rig!(
    Source=Copy Pose From Mesh() {bUseAttachedParent=true}
    Alpha
    isCrouching:double=Get isCrouch() -> (isCrouch:double)
    isMoving:double=Get Moving2d() -> (Moving2d:double)
    inFloating:double=Get isFloating() -> (isFloating:double)
    IsInAir:bool=Get bIsSkyDiving()
    openParachute:double=Get openParachute() -> (openParachute:double)
    IsEmoting:bool=Get bIsPlayingEmote()
    KatanaTransform:double=Get transformAlpha() -> (transformAlpha:double)
   ) -> (Pose) {
    alphaBoolBlend {blendOption=Linear}
    controlRigAssetReference {blueprintRigClass=&/F_MED_AnodizedMetal_CtrlRig.F_MED_AnodizedMetal_CtrlRig_C}
   }
   Save cached pose 'CRCache_InputPose'(Pose=@n23) {cacheName=CRCache_InputPose}
   n24 = "Transform (Modify) Bone - Bone: dyn_skirt_01_front_l"!(ComponentPose=Local To Component(LocalPose=Use cached pose 'CRCache_InputPose'()), bAlphaBoolEnabled:bool=True) -> (Pose) {
    alphaInputType=Bool
    rotationMode=BMM_Additive
    rotationSpace=BCS_BoneSpace
    translationMode=BMM_Additive
    translationSpace=BCS_BoneSpace
    alphaBoolBlend {blendInTime=0.2, blendOption=Linear, blendOutTime=0.2}
    boneToModify {boneName=dyn_skirt_01_front_l}
    rotation=<rot 6.1968 0.0238 -0.0128>
   }
   n25 = "Transform (Modify) Bone - Bone: dyn_skirt_01_l"!(ComponentPose=@n24, bAlphaBoolEnabled:bool=True) -> (Pose) {
    alphaInputType=Bool
    rotationMode=BMM_Additive
    rotationSpace=BCS_BoneSpace
    translationMode=BMM_Additive
    translationSpace=BCS_BoneSpace
    alphaBoolBlend {blendInTime=0.2, blendOption=Linear, blendOutTime=0.2}
    boneToModify {boneName=dyn_skirt_01_l}
    rotation=<rot 3.8881 0.0028 -0.0016>
   }
   n26 = "Transform (Modify) Bone - Bone: dyn_skirt_01_front_r"!(ComponentPose=@n25, bAlphaBoolEnabled:bool=True) -> (Pose) {
    alphaInputType=Bool
    rotationMode=BMM_Additive
    rotationSpace=BCS_BoneSpace
    translationMode=BMM_Additive
    translationSpace=BCS_BoneSpace
    alphaBoolBlend {blendInTime=0.2, blendOption=Linear, blendOutTime=0.2}
    boneToModify {boneName=dyn_skirt_01_front_r}
    rotation {pitch=5}
   }
   n27 = "Transform (Modify) Bone - Bone: dyn_skirt_01_r"!(ComponentPose=@n26, bAlphaBoolEnabled:bool=True) -> (Pose) {
    alphaInputType=Bool
    rotationMode=BMM_Additive
    rotationSpace=BCS_BoneSpace
    translationMode=BMM_Additive
    translationSpace=BCS_BoneSpace
    alphaBoolBlend {blendInTime=0.2, blendOption=Linear, blendOutTime=0.2}
    boneToModify {boneName=dyn_skirt_01_r}
    rotation {pitch=5}
   }
   n28 = "Transform (Modify) Bone - Bone: dyn_skirt_01_front_l"(ComponentPose=@n27, bAlphaBoolEnabled=True) {
    alphaInputType=Bool
    rotationMode=BMM_Additive
    rotationSpace=BCS_BoneSpace
    translationMode=BMM_Additive
    translationSpace=BCS_BoneSpace
    alphaBoolBlend {blendInTime=0.2, blendOption=Linear, blendOutTime=0.2}
    boneToModify {boneName=dyn_skirt_01_front_l}
    rotation=<rot 2.3124 0.0407 -0.0151>
   }
   n29 = "Transform (Modify) Bone - Bone: dyn_skirt_01_l"(ComponentPose=@n28, bAlphaBoolEnabled=True) {
    alphaInputType=Bool
    rotationMode=BMM_Additive
    rotationSpace=BCS_BoneSpace
    translationMode=BMM_Additive
    translationSpace=BCS_BoneSpace
    alphaBoolBlend {blendInTime=0.2, blendOption=Linear, blendOutTime=0.2}
    boneToModify {boneName=dyn_skirt_01_l}
    rotation=<rot 5.9284 0.0019 0.0011>
   }
   n30 = "Transform (Modify) Bone - Bone: dyn_skirt_01_front_l"(ComponentPose=@n29, bAlphaBoolEnabled=True) {
    alphaInputType=Bool
    rotationMode=BMM_Additive
    rotationSpace=BCS_BoneSpace
    translationMode=BMM_Additive
    translationSpace=BCS_BoneSpace
    alphaBoolBlend {blendInTime=0.2, blendOption=Linear, blendOutTime=0.2}
    boneToModify {boneName=dyn_skirt_01_front_l}
    rotation=<rot 6.1968 0.0238 -0.0128>
   }
   n31 = "Transform (Modify) Bone - Bone: dyn_skirt_01_l"(ComponentPose=@n30, bAlphaBoolEnabled=True) {
    alphaInputType=Bool
    rotationMode=BMM_Additive
    rotationSpace=BCS_BoneSpace
    translationMode=BMM_Additive
    translationSpace=BCS_BoneSpace
    alphaBoolBlend {blendInTime=0.2, blendOption=Linear, blendOutTime=0.2}
    boneToModify {boneName=dyn_skirt_01_l}
    rotation=<rot 3.8881 0.0028 -0.0016>
   }
   n32 = "Transform (Modify) Bone - Bone: dyn_skirt_01_front_r"(ComponentPose=@n31, bAlphaBoolEnabled=True) {
    alphaInputType=Bool
    rotationMode=BMM_Additive
    rotationSpace=BCS_BoneSpace
    translationMode=BMM_Additive
    translationSpace=BCS_BoneSpace
    alphaBoolBlend {blendInTime=0.2, blendOption=Linear, blendOutTime=0.2}
    boneToModify {boneName=dyn_skirt_01_front_r}
    rotation {pitch=5}
   }
   n33 = "Transform (Modify) Bone - Bone: dyn_skirt_01_r"(ComponentPose=@n32, bAlphaBoolEnabled=True) {
    alphaInputType=Bool
    rotationMode=BMM_Additive
    rotationSpace=BCS_BoneSpace
    translationMode=BMM_Additive
    translationSpace=BCS_BoneSpace
    alphaBoolBlend {blendInTime=0.2, blendOption=Linear, blendOutTime=0.2}
    boneToModify {boneName=dyn_skirt_01_r}
    rotation {pitch=5}
   }
   n34 = RigidBody(ComponentPose=@n33, OverrideWorldGravity="0,0,-980") {
    bOverrideWorldGravity=true
    overridePhysicsAsset=&/F_MED_AnodizedMetal_RightLeg_SkirtPhysics
    simulationSpace=BaseBoneSpace
    alphaBoolBlend {blendOption=Linear}
    baseBoneRef {boneName=thigh_r}
    componentAppliedLinearAccClamp=<3000 3000 3000>
    componentLinearAccScale=<0.3 0.3 0.3>
    componentLinearVelScale=<0.3 0.3 0.3>
    simSpaceSettings {maxAngularAcceleration=1000, maxAngularVelocity=3, maxLinearAcceleration=700, maxLinearVelocity=500, velocityScaleZ=0, worldAlpha=0.2}
   }
   n35 = RigidBody(ComponentPose=@n34, OverrideWorldGravity="0,0,-980") {
    bOverrideWorldGravity=true
    overridePhysicsAsset=&/F_MED_AnodizedMetal_LeftLeg_SkirtPhysics
    simulationSpace=BaseBoneSpace
    alphaBoolBlend {blendOption=Linear}
    baseBoneRef {boneName=thigh_l}
    componentAppliedLinearAccClamp=<3000 3000 3000>
    componentLinearAccScale=<0.3 0.3 0.3>
    componentLinearVelScale=<0.3 0.3 0.3>
    simSpaceSettings {maxAngularAcceleration=1000, maxAngularVelocity=3, maxLinearAcceleration=700, maxLinearVelocity=500, velocityScaleZ=0, worldAlpha=0.2}
   }
   n36 = Anim Dynamics - Bone: dyn_skirt_03_bck(ComponentPose=@n35, Alpha=1, ExternalForce="0,0,0") {
    angularDampingOverride=1
    angularSpringConstant=20000
    bAngularSpring=true
    bOverrideAngularDamping=true
    bUseGravityOverride=true
    gravityScale=0.35
    simulationSpace=RootRelative
    alphaBoolBlend {blendOption=Linear}
    boundBone {boneName=dyn_skirt_03_bck}
    componentAppliedLinearAccClamp=<9000 9000 9000>
    componentLinearAccScale=<0.5 0.5 0.5>
    componentLinearVelScale=<0.5 0.5 0.5>
    gravityOverride {z=-980}
    physicsBodyDefinitions [{
      boundBone {boneName=dyn_skirt_03_bck}
      boxExtents=<30 30 30>
      constraintSetup {angularLimitsMax=<0 1 0>, angularLimitsMin=<0 -20 0>, angularTarget=<1 0 0>}
      localJointOffset=<3 0 0>
     }]
   }
   n37 = Anim Dynamics - Bone: dyn_skirt_03_bck_l(ComponentPose=@n36, Alpha=1, ExternalForce="0,0,0") {
    angularDampingOverride=1
    angularSpringConstant=20000
    bAngularSpring=true
    bOverrideAngularDamping=true
    bUseGravityOverride=true
    gravityScale=0.35
    simulationSpace=RootRelative
    alphaBoolBlend {blendOption=Linear}
    boundBone {boneName=dyn_skirt_03_bck_l}
    componentAppliedLinearAccClamp=<9000 9000 9000>
    componentLinearAccScale=<0.5 0.5 0.5>
    componentLinearVelScale=<0.5 0.5 0.5>
    gravityOverride {z=-980}
    physicsBodyDefinitions [{
      boundBone {boneName=dyn_skirt_03_bck_l}
      boxExtents=<30 30 30>
      constraintSetup {angularLimitsMax=<0 1 0>, angularLimitsMin=<0 -20 0>, angularTarget=<1 0 0>}
      localJointOffset=<3 0 0>
     }]
   }
   n38 = Anim Dynamics - Bone: dyn_skirt_03_bck_r(ComponentPose=@n37, Alpha=1, ExternalForce="0,0,0") {
    angularDampingOverride=1
    angularSpringConstant=20000
    bAngularSpring=true
    bOverrideAngularDamping=true
    bUseGravityOverride=true
    gravityScale=0.35
    simulationSpace=RootRelative
    alphaBoolBlend {blendOption=Linear}
    boundBone {boneName=dyn_skirt_03_bck_r}
    componentAppliedLinearAccClamp=<9000 9000 9000>
    componentLinearAccScale=<0.5 0.5 0.5>
    componentLinearVelScale=<0.5 0.5 0.5>
    gravityOverride {z=-980}
    physicsBodyDefinitions [{
      boundBone {boneName=dyn_skirt_03_bck_r}
      boxExtents=<30 30 30>
      constraintSetup {angularLimitsMax=<0 1 0>, angularLimitsMin=<0 -20 0>, angularTarget=<1 0 0>}
      localJointOffset=<3 0 0>
     }]
   }
   n39 = Anim Dynamics - Bone: dyn_skirt_04_bck(ComponentPose=@n38, Alpha=1, ExternalForce="0,0,0") {
    angularDampingOverride=1
    angularSpringConstant=20000
    bAngularSpring=true
    bOverrideAngularDamping=true
    bUseGravityOverride=true
    gravityScale=0.35
    simulationSpace=RootRelative
    alphaBoolBlend {blendOption=Linear}
    boundBone {boneName=dyn_skirt_04_bck}
    componentAppliedLinearAccClamp=<9000 9000 9000>
    componentLinearAccScale=<0.5 0.5 0.5>
    componentLinearVelScale=<0.5 0.5 0.5>
    gravityOverride {z=-980}
    physicsBodyDefinitions [{
      boundBone {boneName=dyn_skirt_04_bck}
      boxExtents=<30 30 30>
      constraintSetup {angularLimitsMax=<0 1 0>, angularLimitsMin=<0 -20 0>, angularTarget=<1 0 0>}
      localJointOffset=<3 0 0>
     }]
   }
   n40 = Anim Dynamics - Bone: dyn_skirt_04_bck_l(ComponentPose=@n39, Alpha=1, ExternalForce="0,0,0") {
    angularDampingOverride=1
    angularSpringConstant=20000
    bAngularSpring=true
    bOverrideAngularDamping=true
    bUseGravityOverride=true
    gravityScale=0.35
    simulationSpace=RootRelative
    alphaBoolBlend {blendOption=Linear}
    boundBone {boneName=dyn_skirt_04_bck_l}
    componentAppliedLinearAccClamp=<9000 9000 9000>
    componentLinearAccScale=<0.5 0.5 0.5>
    componentLinearVelScale=<0.5 0.5 0.5>
    gravityOverride {z=-980}
    physicsBodyDefinitions [{
      boundBone {boneName=dyn_skirt_04_bck_l}
      boxExtents=<30 30 30>
      constraintSetup {angularLimitsMax=<0 1 0>, angularLimitsMin=<0 -20 0>, angularTarget=<1 0 0>}
      localJointOffset=<3 0 0>
     }]
   }
   n41 = Anim Dynamics - Bone: dyn_skirt_04_bck_r(ComponentPose=@n40, Alpha=1, ExternalForce="0,0,0") {
    angularDampingOverride=1
    angularSpringConstant=20000
    bAngularSpring=true
    bOverrideAngularDamping=true
    bUseGravityOverride=true
    gravityScale=0.35
    simulationSpace=RootRelative
    alphaBoolBlend {blendOption=Linear}
    boundBone {boneName=dyn_skirt_04_bck_r}
    componentAppliedLinearAccClamp=<9000 9000 9000>
    componentLinearAccScale=<0.5 0.5 0.5>
    componentLinearVelScale=<0.5 0.5 0.5>
    gravityOverride {z=-980}
    physicsBodyDefinitions [{
      boundBone {boneName=dyn_skirt_04_bck_r}
      boxExtents=<30 30 30>
      constraintSetup {angularLimitsMax=<0 1 0>, angularLimitsMin=<0 -20 0>, angularTarget=<1 0 0>}
      localJointOffset=<3 0 0>
     }]
   }
   n42 = Anim Dynamics - Bone: dyn_skirt_05_bck(ComponentPose=@n41, Alpha=1, ExternalForce="0,0,0") {
    angularDampingOverride=1
    angularSpringConstant=20000
    bAngularSpring=true
    bOverrideAngularDamping=true
    bUseGravityOverride=true
    gravityScale=0.35
    simulationSpace=RootRelative
    alphaBoolBlend {blendOption=Linear}
    boundBone {boneName=dyn_skirt_05_bck}
    componentAppliedLinearAccClamp=<9000 9000 9000>
    componentLinearAccScale=<0.5 0.5 0.5>
    componentLinearVelScale=<0.5 0.5 0.5>
    gravityOverride {z=-980}
    physicsBodyDefinitions [{
      boundBone {boneName=dyn_skirt_05_bck}
      boxExtents=<30 30 30>
      constraintSetup {angularLimitsMax=<0 5 0>, angularLimitsMin=<0 -20 0>, angularTarget=<1 0 0>}
      localJointOffset=<3 0 0>
     }]
   }
   n43 = Anim Dynamics - Bone: dyn_skirt_05_bck_l(ComponentPose=@n42, Alpha=1, ExternalForce="0,0,0") {
    angularDampingOverride=1
    angularSpringConstant=20000
    bAngularSpring=true
    bOverrideAngularDamping=true
    bUseGravityOverride=true
    gravityScale=0.35
    simulationSpace=RootRelative
    alphaBoolBlend {blendOption=Linear}
    boundBone {boneName=dyn_skirt_05_bck_l}
    componentAppliedLinearAccClamp=<9000 9000 9000>
    componentLinearAccScale=<0.5 0.5 0.5>
    componentLinearVelScale=<0.5 0.5 0.5>
    gravityOverride {z=-980}
    physicsBodyDefinitions [{
      boundBone {boneName=dyn_skirt_05_bck_l}
      boxExtents=<30 30 30>
      constraintSetup {angularLimitsMax=<0 5 0>, angularLimitsMin=<0 -20 0>, angularTarget=<1 0 0>}
      localJointOffset=<3 0 0>
     }]
   }
   n44 = Anim Dynamics - Bone: dyn_skirt_05_bck_r(ComponentPose=@n43, Alpha=1, ExternalForce="0,0,0") {
    angularDampingOverride=1
    angularSpringConstant=20000
    bAngularSpring=true
    bOverrideAngularDamping=true
    bUseGravityOverride=true
    gravityScale=0.35
    simulationSpace=RootRelative
    alphaBoolBlend {blendOption=Linear}
    boundBone {boneName=dyn_skirt_05_bck_r}
    componentAppliedLinearAccClamp=<9000 9000 9000>
    componentLinearAccScale=<0.5 0.5 0.5>
    componentLinearVelScale=<0.5 0.5 0.5>
    gravityOverride {z=-980}
    physicsBodyDefinitions [{
      boundBone {boneName=dyn_skirt_05_bck_r}
      boxExtents=<30 30 30>
      constraintSetup {angularLimitsMax=<0 5 0>, angularLimitsMin=<0 -20 0>, angularTarget=<1 0 0>}
      localJointOffset=<3 0 0>
     }]
   }
   n45 = "Constraint - dyn_skirt_03_fr (RR)"!(ComponentPose=@n44, Alpha=1, ConstraintWeights_0=0.5, ConstraintWeights_1:float=0.5) -> (Pose) {
    alphaBoolBlend {blendOption=Linear}
    boneToModify {boneName=dyn_skirt_03_fr}
    constraintSetup [{transformType=Rotation, targetBone {boneName=dyn_skirt_03_front_l}}, {transformType=Rotation, targetBone {boneName=dyn_skirt_03_front_r}}]
    constraintWeights [1,1]
   }
   n46 = "Constraint - dyn_skirt_04_fr (RR)"!(ComponentPose=@n45, Alpha=1, ConstraintWeights_0=0.5, ConstraintWeights_1:float=0.5) -> (Pose) {
    alphaBoolBlend {blendOption=Linear}
    boneToModify {boneName=dyn_skirt_04_fr}
    constraintSetup [{transformType=Rotation, targetBone {boneName=dyn_skirt_04_front_l}}, {transformType=Rotation, targetBone {boneName=dyn_skirt_04_front_r}}]
    constraintWeights [1,1]
   }
   n47 = "Constraint - dyn_skirt_05_fr (RR)"!(ComponentPose=@n46, Alpha=1, ConstraintWeights_0=0.5, ConstraintWeights_1:float=0.5) -> (Pose) {
    alphaBoolBlend {blendOption=Linear}
    boneToModify {boneName=dyn_skirt_05_fr}
    constraintSetup [{transformType=Rotation, targetBone {boneName=dyn_skirt_05_front_l}}, {transformType=Rotation, targetBone {boneName=dyn_skirt_05_front_r}}]
    constraintWeights [1,1]
   }
   n48 = "Transform (Modify) Bone - Bone: dyn_skirt_03_bck_l"!(ComponentPose=@n47, bAlphaBoolEnabled:bool=True) -> (Pose) {
    alphaInputType=Bool
    rotationMode=BMM_Additive
    rotationSpace=BCS_BoneSpace
    translationMode=BMM_Additive
    translationSpace=BCS_BoneSpace
    alphaBoolBlend {blendInTime=0.2, blendOption=Linear, blendOutTime=0.2}
    boneToModify {boneName=dyn_skirt_03_bck_l}
    rotation=<rot 14.2552 0.002 0.0021>
    translation=<-0.4431 -0.2249 1.7409>
   }
   n49 = "Transform (Modify) Bone - Bone: dyn_skirt_03_bck"!(ComponentPose=@n48, bAlphaBoolEnabled:bool=True) -> (Pose) {
    alphaInputType=Bool
    rotationMode=BMM_Additive
    rotationSpace=BCS_BoneSpace
    translationMode=BMM_Additive
    translationSpace=BCS_BoneSpace
    alphaBoolBlend {blendInTime=0.2, blendOption=Linear, blendOutTime=0.2}
    boneToModify {boneName=dyn_skirt_03_bck}
    rotation=<rot 18.4749 0.025 0.0123>
    translation=<-2.4629 0.8213 0.2305>
   }
   n50 = "Transform (Modify) Bone - Bone: dyn_skirt_03_bck_r"!(ComponentPose=@n49, bAlphaBoolEnabled:bool=True) -> (Pose) {
    alphaInputType=Bool
    rotationMode=BMM_Additive
    rotationSpace=BCS_BoneSpace
    translationMode=BMM_Additive
    translationSpace=BCS_BoneSpace
    alphaBoolBlend {blendInTime=0.2, blendOption=Linear, blendOutTime=0.2}
    boneToModify {boneName=dyn_skirt_03_bck_r}
    rotation=<rot 13.9072 -0.0053 -0.004>
   }
   Save cached pose 'CRCache_baseSkirt_setup'(Pose=Component To Local(ComponentPose=@n50)) {cacheName=CRCache_baseSkirt_setup}
   n51 = Use cached pose 'CRCache_baseSkirt_setup'()
   n52 = Use cached pose 'CRCache_KneeSliding_setup'()
   n53 = Get bIsKneeSliding()
   Save cached pose 'CRCache_Skirt_Dynamics'(Pose=Blend Poses by bool(
     BlendPose_0=Use cached pose 'CRCache_BaseballSliding_Setup'()
     BlendPose_1=Blend Poses by bool(BlendPose_0=@n52, BlendPose_1=@n51, BlendTime_0=0.2, bActiveValue=@n53) {blendType=Linear}
     BlendTime_0=0.5
     BlendTime_1=0.2
     bActiveValue=Get bIsBaseballSliding()
    ) {blendType=Linear}) {cacheName=CRCache_Skirt_Dynamics}
   n54 = "Transform (Modify) Bone - Bone: dyn_skirt_01_fr"!(ComponentPose=Local To Component(LocalPose=Use cached pose 'CRCache_InputPose'()), Alpha=1) -> (Pose) {
    rotationMode=BMM_Additive
    rotationSpace=BCS_BoneSpace
    alphaBoolBlend {blendOption=Linear}
    boneToModify {boneName=dyn_skirt_01_fr}
    rotation=<rot 26.773 8.281 -20.375>
   }
   n55 = "Transform (Modify) Bone - Bone: dyn_skirt_01_bck_l"!(ComponentPose=@n54, Alpha=1) -> (Pose) {
    rotationMode=BMM_Additive
    rotationSpace=BCS_BoneSpace
    translationMode=BMM_Additive
    translationSpace=BCS_BoneSpace
    alphaBoolBlend {blendOption=Linear}
    boneToModify {boneName=dyn_skirt_01_bck_l}
    rotation=<rot 32.5962 -3.3755 48.157>
    translation=<-2.9945 0.8991 -0.4229>
   }
   n56 = "Transform (Modify) Bone - Bone: dyn_skirt_02_bck_r"!(ComponentPose=@n55, Alpha=1) -> (Pose) {
    rotationMode=BMM_Additive
    rotationSpace=BCS_BoneSpace
    translationMode=BMM_Additive
    translationSpace=BCS_BoneSpace
    alphaBoolBlend {blendOption=Linear}
    boneToModify {boneName=dyn_skirt_02_bck_r}
    rotation=<rot -24.7388 2.3235 -33.9964>
    translation=<-0.4667 -2.442 7.1788>
   }
   n57 = "Transform (Modify) Bone - Bone: dyn_skirt_01_bck"!(ComponentPose=@n56, Alpha=1) -> (Pose) {
    rotationMode=BMM_Additive
    rotationSpace=BCS_BoneSpace
    translationMode=BMM_Additive
    translationSpace=BCS_BoneSpace
    alphaBoolBlend {blendOption=Linear}
    boneToModify {boneName=dyn_skirt_01_bck}
    rotation {pitch=8.408}
    translation=<-5.3262 0 -1.5362>
   }
   n58 = "Transform (Modify) Bone - Bone: dyn_skirt_01_l"!(ComponentPose=@n57, Alpha:float=1) -> (Pose) {
    rotationMode=BMM_Additive
    rotationSpace=BCS_BoneSpace
    alphaBoolBlend {blendOption=Linear}
    boneToModify {boneName=dyn_skirt_01_l}
    rotation=<rot 41.4529 -1.5658 23.7392>
   }
   n59 = "Transform (Modify) Bone - Bone: dyn_skirt_01_r"!(ComponentPose=@n58, Alpha:float=1) -> (Pose) {
    rotationMode=BMM_Additive
    rotationSpace=BCS_BoneSpace
    alphaBoolBlend {blendOption=Linear}
    boneToModify {boneName=dyn_skirt_01_r}
    rotation=<rot 32.54 -12.717 -40.151>
   }
   n60 = "Transform (Modify) Bone - Bone: base_beltLower_r"!(ComponentPose=@n59, Alpha=1) -> (Pose) {
    rotationMode=BMM_Additive
    rotationSpace=BCS_BoneSpace
    translationMode=BMM_Additive
    translationSpace=BCS_BoneSpace
    alphaBoolBlend {blendOption=Linear}
    boneToModify {boneName=base_beltLower_r}
    rotation=<rot 11.6814 -18.9314 -0.959>
    translation=<-1.5016 -1.0655 -1.2697>
   }
   n61 = "Transform (Modify) Bone - Bone: base_katana_r"!(ComponentPose=@n60, Alpha=1) -> (Pose) {
    rotationMode=BMM_Additive
    rotationSpace=BCS_BoneSpace
    translationMode=BMM_Additive
    translationSpace=BCS_BoneSpace
    alphaBoolBlend {blendOption=Linear}
    boneToModify {boneName=base_katana_r}
    rotation=<rot 13.4574 -31.9143 30.0386>
   }
   n62 = "Transform (Modify) Bone - Bone: base_bag_l"!(ComponentPose=@n61, Alpha=1) -> (Pose) {
    rotationMode=BMM_Additive
    rotationSpace=BCS_BoneSpace
    translationMode=BMM_Additive
    translationSpace=BCS_BoneSpace
    alphaBoolBlend {blendOption=Linear}
    boneToModify {boneName=base_bag_l}
    rotation=<rot 44.8806 0.0001 0>
    translation=<-4.3696 0.0002 1.6068>
   }
   n63 = "Transform (Modify) Bone - Bone: base_bow01_r"!(ComponentPose=@n62, Alpha=1) -> (Pose) {
    rotationMode=BMM_Additive
    rotationSpace=BCS_BoneSpace
    translationMode=BMM_Additive
    translationSpace=BCS_BoneSpace
    alphaBoolBlend {blendOption=Linear}
    boneToModify {boneName=base_bow01_r}
    rotation=<rot 29.6833 0 0>
    translation=<-0.6934 -1.5826 1.2164>
   }
   n64 = "Transform (Modify) Bone - Bone: base_bow01_l"!(ComponentPose=@n63, Alpha=1) -> (Pose) {
    rotationMode=BMM_Additive
    rotationSpace=BCS_BoneSpace
    translationMode=BMM_Additive
    translationSpace=BCS_BoneSpace
    alphaBoolBlend {blendOption=Linear}
    boneToModify {boneName=base_bow01_l}
    rotation=<rot 48.2429 11.2618 7.8479>
    translation=<-1.3415 -0.0186 1.1778>
   }
   n65 = RigidBody(ComponentPose=@n64, OverrideWorldGravity="0,0,-980") {
    bOverrideWorldGravity=true
    overridePhysicsAsset=&/F_MED_AnodizedMetal_SkirtKneeSlidingPhysics
    simulationSpace=BaseBoneSpace
    alphaBoolBlend {blendOption=Linear}
    baseBoneRef {boneName=pelvis}
    componentAppliedLinearAccClamp=<3000 3000 3000>
    componentLinearAccScale=<0.3 0.3 0.3>
    componentLinearVelScale=<0.3 0.3 0.3>
    simSpaceSettings {maxAngularAcceleration=1000, maxAngularVelocity=3, maxLinearAcceleration=700, maxLinearVelocity=500, velocityScaleZ=0}
   }
   n66 = "Transform (Modify) Bone - Bone: base_mask_a_r"!(ComponentPose=@n65, bAlphaBoolEnabled=True) -> (Pose) {
    alphaInputType=Bool
    rotationMode=BMM_Additive
    rotationSpace=BCS_BoneSpace
    translationMode=BMM_Additive
    translationSpace=BCS_BoneSpace
    alphaBoolBlend {blendInTime=0.2, blendOption=Linear, blendOutTime=0.2}
    boneToModify {boneName=base_mask_a_r}
    rotation=<rot 9.5924 -30.4118 -24.6231>
    translation=<-0.8668 -0.1234 3.3906>
   }
   n67 = "Transform (Modify) Bone - Bone: dyn_mask_b01_r"!(ComponentPose=@n66, bAlphaBoolEnabled=True) -> (Pose) {
    alphaInputType=Bool
    rotationMode=BMM_Additive
    rotationSpace=BCS_BoneSpace
    translationMode=BMM_Additive
    translationSpace=BCS_BoneSpace
    alphaBoolBlend {blendInTime=0.2, blendOption=Linear, blendOutTime=0.2}
    boneToModify {boneName=dyn_mask_b01_r}
    rotation=<rot 19.5555 -0.0001 0.0013>
   }
   Save cached pose 'CRCache_KneeSliding_setup'(Pose=Component To Local(ComponentPose=@n67)) {cacheName=CRCache_KneeSliding_setup}
   n68 = "Transform (Modify) Bone - Bone: dyn_skirt_01_fr"(ComponentPose=Local To Component(LocalPose=Use cached pose 'CRCache_InputPose'()), Alpha=1) {
    rotationMode=BMM_Additive
    rotationSpace=BCS_BoneSpace
    alphaBoolBlend {blendOption=Linear}
    boneToModify {boneName=dyn_skirt_01_fr}
    rotation=<rot -44.86 5.828 -49.8102>
   }
   n69 = "Transform (Modify) Bone - Bone: dyn_skirt_01_bck_l"(ComponentPose=@n68, Alpha=1) {
    rotationMode=BMM_Additive
    rotationSpace=BCS_BoneSpace
    translationMode=BMM_Additive
    translationSpace=BCS_BoneSpace
    alphaBoolBlend {blendOption=Linear}
    boneToModify {boneName=dyn_skirt_01_bck_l}
    rotation=<rot -50.8229 0 0>
    translation=<4.7255 0 3.8509>
   }
   n70 = "Transform (Modify) Bone - Bone: dyn_skirt_01_bck_r"!(ComponentPose=@n69, Alpha=1) -> (Pose) {
    rotationMode=BMM_Additive
    rotationSpace=BCS_BoneSpace
    translationMode=BMM_Additive
    translationSpace=BCS_BoneSpace
    alphaBoolBlend {blendOption=Linear}
    boneToModify {boneName=dyn_skirt_01_bck_r}
    rotation=<rot -61.2104 0 0>
    translation=<7.3948 0 4.0636>
   }
   n71 = "Transform (Modify) Bone - Bone: dyn_skirt_01_bck"(ComponentPose=@n70, Alpha=1) {
    rotationMode=BMM_Additive
    rotationSpace=BCS_BoneSpace
    translationMode=BMM_Additive
    translationSpace=BCS_BoneSpace
    alphaBoolBlend {blendOption=Linear}
    boneToModify {boneName=dyn_skirt_01_bck}
    rotation=<rot -63.695 0.0001 -0.0001>
    translation=<12.8683 0 6.3613>
   }
   n72 = "Transform (Modify) Bone - Bone: dyn_skirt_01_l"(ComponentPose=@n71, Alpha=1) {
    rotationMode=BMM_Additive
    rotationSpace=BCS_BoneSpace
    alphaBoolBlend {blendOption=Linear}
    boneToModify {boneName=dyn_skirt_01_l}
    rotation=<rot -9.3355 8.7195 -1.4252>
   }
   n73 = "Transform (Modify) Bone - Bone: dyn_skirt_01_r"(ComponentPose=@n72, Alpha=1) {
    rotationMode=BMM_Additive
    rotationSpace=BCS_BoneSpace
    alphaBoolBlend {blendOption=Linear}
    boneToModify {boneName=dyn_skirt_01_r}
    rotation=<rot -12.0593 -31.6727 -23.8143>
   }
   n74 = "Transform (Modify) Bone - Bone: base_beltLower_r"(ComponentPose=@n73, Alpha=1) {
    rotationMode=BMM_Additive
    rotationSpace=BCS_BoneSpace
    translationMode=BMM_Additive
    translationSpace=BCS_BoneSpace
    alphaBoolBlend {blendOption=Linear}
    boneToModify {boneName=base_beltLower_r}
    rotation=<rot 17.6871 -21.1837 -14.3864>
    translation=<0 0 -5.2459>
   }
   n75 = "Transform (Modify) Bone - Bone: base_skirt_01_front_r"!(ComponentPose=@n74, Alpha=1) -> (Pose) {
    rotationMode=BMM_Additive
    rotationSpace=BCS_BoneSpace
    translationMode=BMM_Additive
    translationSpace=BCS_BoneSpace
    alphaBoolBlend {blendOption=Linear}
    boneToModify {boneName=base_skirt_01_front_r}
    rotation=<rot -3.2675 7.3743 -37.5186>
   }
   n76 = "Transform (Modify) Bone - Bone: dyn_skirt_01_front_l"!(ComponentPose=@n75, Alpha:float=1) -> (Pose) {
    rotationMode=BMM_Additive
    rotationSpace=BCS_BoneSpace
    translationMode=BMM_Additive
    translationSpace=BCS_BoneSpace
    alphaBoolBlend {blendOption=Linear}
    boneToModify {boneName=dyn_skirt_01_front_l}
    rotation=<rot -6.2938 -0.0001 -22.6339>
   }
   n77 = "Transform (Modify) Bone - Bone: base_bow01_r"(ComponentPose=@n76, Alpha=1) {
    rotationMode=BMM_Additive
    rotationSpace=BCS_BoneSpace
    translationMode=BMM_Additive
    translationSpace=BCS_BoneSpace
    alphaBoolBlend {blendOption=Linear}
    boneToModify {boneName=base_bow01_r}
    rotation=<rot 73.747 -20.1571 -19.4231>
   }
   n78 = "Transform (Modify) Bone - Bone: base_bow01_l"(ComponentPose=@n77, Alpha=1) {
    rotationMode=BMM_Additive
    rotationSpace=BCS_BoneSpace
    translationMode=BMM_Additive
    translationSpace=BCS_BoneSpace
    alphaBoolBlend {blendOption=Linear}
    boneToModify {boneName=base_bow01_l}
    rotation=<rot 37.2004 -14.7013 -7.867>
    translation=<0.4295 2.4545 0.2733>
   }
   n79 = RigidBody(ComponentPose=@n78, OverrideWorldGravity="0,0,-980") {
    bOverrideWorldGravity=true
    overridePhysicsAsset=&/F_MED_AnodizedMetal_SkirtSliding_Physics
    simulationSpace=BaseBoneSpace
    alphaBoolBlend {blendOption=Linear}
    baseBoneRef {boneName=pelvis}
    componentAppliedLinearAccClamp=<3000 3000 3000>
    componentLinearAccScale=<0.3 0.3 0.3>
    componentLinearVelScale=<0.3 0.3 0.3>
    simSpaceSettings {maxAngularAcceleration=1000, maxAngularVelocity=3, maxLinearAcceleration=700, maxLinearVelocity=500, velocityScaleZ=0}
   }
   n80 = "Transform (Modify) Bone - Bone: base_mask_a_r"(ComponentPose=@n79, bAlphaBoolEnabled=True) {
    alphaInputType=Bool
    rotationMode=BMM_Additive
    rotationSpace=BCS_BoneSpace
    translationMode=BMM_Additive
    translationSpace=BCS_BoneSpace
    alphaBoolBlend {blendInTime=0.2, blendOption=Linear, blendOutTime=0.2}
    boneToModify {boneName=base_mask_a_r}
    rotation=<rot 62.7261 -43.5759 -64.7946>
    translation=<-7.3388 -6.6183 3.4445>
   }
   n81 = "Transform (Modify) Bone - Bone: dyn_mask_b01_r"(ComponentPose=@n80, bAlphaBoolEnabled=True) {
    alphaInputType=Bool
    rotationMode=BMM_Additive
    rotationSpace=BCS_BoneSpace
    translationMode=BMM_Additive
    translationSpace=BCS_BoneSpace
    alphaBoolBlend {blendInTime=0.2, blendOption=Linear, blendOutTime=0.2}
    boneToModify {boneName=dyn_mask_b01_r}
    rotation=<rot 66.9191 0.0001 -34.6154>
    translation=<-4.9759 -3.7335 2.1205>
   }
   Save cached pose 'CRCache_BaseballSliding_Setup'(Pose=Component To Local(ComponentPose=@n81)) {cacheName=CRCache_BaseballSliding_Setup}
  ]}
 EventGraph {nodes [
   n1 = Event BlueprintUpdateAnimation()
   > n2 = Sequence!(execute) -> (then_0, then_1, then_2:Exec, then_3:Exec, then_4:Exec, then_5:Exec)
   n3 = Get bIsMoving2D()
   Set Moving2d(
    execute=@n2.then_1
    Moving2d:double=FInterpTo(Current=Get Moving2d(), Target="To Float (Boolean)"(InBool=@n3), DeltaTime=@n1.DeltaTimeX, InterpSpeed=2.5)
   ) -> (Output_Get:double)
   n4 = Get bIsSkydiveFloating()
   > Set isFloating(isFloating:double=FInterpTo(Current=Get isFloating(), Target="To Float (Boolean)"(InBool=@n4), DeltaTime=@n1.DeltaTimeX, InterpSpeed=3)) -> (Output_Get:double)
   n5 = Get bIsMoving2D()
   > Set CrouchIdle(CrouchIdle:bool=AND Boolean(A=Get bIsCrouching(), B=NOT Boolean(A=@n5))) -> (Output_Get:bool)
   > n6 = Branch(Condition=Get OnZipline() -> (OnZipline:bool))
   > Set ACC/VEL Zipline(ACC/VEL Zipline:Vector="0, 0, 0") -> (Output_Get:Vector)
   > Set ZipDynamics(ZipDynamics:double=0) -> (Output_Get:double)
   > Set ZipRigidbodiesMASKs(ZipRigidbodiesMASKs:double=0.6) -> (Output_Get:double)
   Set ACC/VEL Zipline(execute=@n6.else, ACC/VEL Zipline="1,1,1")
   > Set ZipDynamics(ZipDynamics=1)
   > Set ZipRigidbodiesMASKs(ZipRigidbodiesMASKs=1)
   > Branch(Condition=Get bIsInFrontEnd())
   > Set ZipDynamics(ZipDynamics=0.3)
   Branch(execute=@n2.then_5, Condition=Get bIsInFrontEnd())
   > Set Is Playing Custom FrontEnd Idle(Is Playing Custom FrontEnd Idle:double=GetCurveValue(self=GetOwnerFrontEndAnimInstance(), CurveName=Frontend_Idle_Bypass)) -> (Output_Get:double)
   n7 = Cast To FortPlayerAnimInstance(execute=@n2.then_2, Object=GetAnimInstance(self=GetOwnerSkeletalMeshComponent()))
   > Set goingCommando(goingCommando:bool=false) -> (Output_Get:bool)
   > n8 = Switch on EFortWeaponCoreAnimation!(execute, Selection:EFortWeaponCoreAnimation Enum=Get WeaponCoreAnim(self=@n7)) -> (
    Melee:Exec
    Pistol:Exec
    Shotgun:Exec
    PaperBlueprint:Exec
    Rifle:Exec
    MeleeOneHand:Exec
    MachinePistol:Exec
    RocketLauncher:Exec
    GrenadeLauncher:Exec
    GoingCommando:Exec
    AssaultRifle:Exec
    TacticalShotgun:Exec
    SniperRifle:Exec
    TrapPlacement:Exec
    ShoulderLauncher:Exec
    AbilityDecoTool:Exec
    Crossbow:Exec
    C4:Exec
    RemoteControl:Exec
    DualWield:Exec
    AR_BullPup:Exec
    AR_ForwardGrip:Exec
    MedPackPaddles:Exec
    SMG_P90:Exec
    AR_DrumGun:Exec
    Consumable_Small:Exec
    Consumable_Large:Exec
    Balloon:Exec
    MountedTurret:Exec
    CreativeTool:Exec
    ExplosiveBow:Exec
    AshtonIndigo:Exec
    AshtonChicago:Exec
    MeleeDualWield:Exec
    Unarmed:Exec
   )
   Branch(execute=@n8.GoingCommando)
   > Set goingCommando(goingCommando=1)
   > Branch(Condition=Get bIsSkyDiving())
   > Set goingCommando(goingCommando=false)
   n9 = Gravity Override State Function DEPRECATED(
    execute=@n2.then_0
    CharacterPartAnimInstance=Self-Reference()
    GameTimeInSeconds=GetGameTimeInSeconds()
    GravityOverrideParams=Get Gravity Override Params()
   )
   > Set GravityOverride(GravityOverride=@n9)
   n10 = GetAnimInstance(self=GetOwnerSkeletalMeshComponent())
   > Is Valid(InputObject=@n10)
   > Set celebration(celebration:bool=Montage_IsPlaying(self=@n10, Montage=/BRCosmetics/Animation/Game/MainPlayer/Emotes/Emote_Celebration_Female_M)) -> (Output_Get:bool)
   > Set isCrouch(isCrouch:double="To Float (Boolean)"(InBool=Get bIsCrouching())) -> (Output_Get:double)
   Set transformAlpha(
    execute=@n2.then_3
    transformAlpha:double=Remap Float(Input=GetCurveValue(CurveName=KatanaDynamics), OldRange=(X=0,Y=1), NewRange=(X=1,Y=0), Clamp=true) -> (Output=~)
   ) -> (Output_Get:double)
   n11 = Event BlueprintProcessAnimInputEvent()
   > n12 = Branch(Condition=IsNamed(self=@n11.AnimInputEvent, InEventName=aie.Player.Zipline.Entered))
   > Set OnZipline(OnZipline:bool=1) -> (Output_Get:bool)
   Branch(execute=@n12.else, Condition=IsNamed(self=@n11.AnimInputEvent, InEventName=aie.Player.Zipline.Exited))
   > Set OnZipline(OnZipline=false)
   Event OnParachuteOpened()
   > Set openParachute(openParachute:double=1) -> (Output_Get:double)
   > Delay(Duration=2.5)
   > Set openParachute(openParachute=0)
   Set ZipDynamics(ZipDynamics=1)
  ]}
]
variables [
 ACC/VEL Zipline {type=Vector}
 celebration {type=boolean}
 CrouchIdle {type=boolean}
 goingCommando {type=boolean}
 Gravity Override Params {type=GravityOverrideParamsStruct}
 GravityOverride {type=Vector}
 Is Playing Custom FrontEnd Idle {type=number}
 isCrouch {type=number}
 isFloating {type=number}
 Moving2d {type=number}
 OnZipline {type=boolean}
 openParachute {type=number}
 transformAlpha {type=number}
 ZipDynamics {type=number}
 ZipRigidbodiesMASKs {type=number}
]
