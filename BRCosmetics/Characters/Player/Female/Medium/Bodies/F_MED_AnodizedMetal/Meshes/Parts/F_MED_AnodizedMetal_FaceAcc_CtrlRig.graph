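# Control Rig for the F_MED_AnodizedMetal face accessory, procedurally posing its hair and hat bones against body motion.
# Back hair nulls constrained to head, neck and spine accompany a spline IK chain with twist, which drives left and right strands.
# Spherical pose readers on upper arms and head-to-neck rotation weight corrective offsets.
# State variables like crouching, gliding, skydiving, aiming and cape equipped blend further hair and hat corrections.
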ControlRigBlueprint /BRCosmetics/Characters/Player/Female/Medium/Bodies/F_MED_AnodizedMetal/Meshes/Parts/F_MED_AnodizedMetal_FaceAcc_CtrlRig
graphs [
 Multiply_1_1_1_ContainedGraph {nodes [
   n1 = RigVMFunctionEntryNode()
   n2 = &RigVMFunction_MathDoubleMul(A=@n1.A, B=@n1.B)
   RigVMFunctionReturnNode(Result=&RigVMFunction_MathDoubleMul(B=@n1.D, A=&RigVMFunction_MathDoubleMul(B=@n1.C, A=@n2.Result).Result).Result)
  ]}
 Multiply_1_1_1_ContainedGraph {nodes [
   n1 = RigVMFunctionEntryNode()
   n2 = &RigVMFunction_MathDoubleMul(A=@n1.A, B=@n1.B)
   RigVMFunctionReturnNode(Result=&RigVMFunction_MathDoubleMul(B=@n1.D, A=&RigVMFunction_MathDoubleMul(B=@n1.C, A=@n2.Result).Result).Result)
  ]}
 Multiply_1_1_ContainedGraph {nodes [
   n1 = RigVMFunctionEntryNode()
   RigVMFunctionReturnNode(Result=&RigVMFunction_MathDoubleMul(B=@n1.C, A=&RigVMFunction_MathDoubleMul(A=@n1.A, B=@n1.B).Result).Result)
  ]}
 Multiply_1_2_1_ContainedGraph {nodes [
   n1 = RigVMFunctionEntryNode()
   RigVMFunctionReturnNode(Result=&RigVMFunction_MathDoubleMul(B=@n1.C, A=&RigVMFunction_MathDoubleMul(A=@n1.A, B=@n1.B).Result).Result)
  ]}
 Multiply_1_2_1_ContainedGraph {nodes [
   n1 = RigVMFunctionEntryNode()
   RigVMFunctionReturnNode(Result=&RigVMFunction_MathDoubleMul(B=@n1.C, A=&RigVMFunction_MathDoubleMul(A=@n1.A, B=@n1.B).Result).Result)
  ]}
 Multiply_1_2_1_ContainedGraph {nodes [
   n1 = RigVMFunctionEntryNode()
   RigVMFunctionReturnNode(Result=&RigVMFunction_MathDoubleMul(B=@n1.C, A=&RigVMFunction_MathDoubleMul(A=@n1.A, B=@n1.B).Result).Result)
  ]}
 Multiply_1_2_ContainedGraph {nodes [
   n1 = RigVMFunctionEntryNode()
   RigVMFunctionReturnNode(Result=&RigVMFunction_MathDoubleMul(B=@n1.C, A=&RigVMFunction_MathDoubleMul(A=@n1.A, B=@n1.B).Result).Result)
  ]}
 New Function_ContainedGraph {nodes [
   RigVMFunctionEntryNode()
   > RigVMFunction_Sequence = RigVMAggregateNode()
   &RigUnit_ParentConstraint(
    AdvancedSettings=(RotationOrderForFilter=XZY)
    Child=(Type=Null,Name="dyn_Hair_Back_01_Null")
    Filter=(TranslationFilter=(bX=true,bY=true,bZ=true),RotationFilter=(bX=false,bY=false,bZ=false),ScaleFilter=(bX=True,bY=True,bZ=True))
    Parents=((Item=(Type=Bone,Name="head"),Weight=0.9),(Item=(Type=Bone,Name="spine_05"),Weight=0.1))
    Weight=1
    ExecutePin=@RigVMFunction_Sequence.A
   )
   > &RigUnit_ParentConstraint(
    AdvancedSettings=(RotationOrderForFilter=XZY)
    Child=(Type=Null,Name="dyn_Hair_Back_02_Null")
    Filter=(TranslationFilter=(bX=True,bY=True,bZ=True),RotationFilter=(bX=false,bY=false,bZ=false),ScaleFilter=(bX=True,bY=True,bZ=True))
    Parents=((Item=(Type=Bone,Name="neck_02"),Weight=0.6),(Item=(Type=Bone,Name="spine_05"),Weight=0.4))
    Weight=1
   )
   > &RigUnit_ParentConstraint(
    AdvancedSettings=(RotationOrderForFilter=XZY)
    Child=(Type=Null,Name="dyn_Hair_Back_03_Null")
    Filter=(TranslationFilter=(bX=True,bY=True,bZ=True),RotationFilter=(bX=false,bY=false,bZ=false),ScaleFilter=(bX=True,bY=True,bZ=True))
    Parents=((Item=(Type=Bone,Name="neck_02"),Weight=0.5),(Item=(Type=Bone,Name="spine_05"),Weight=0.5))
    Weight=1
   )
   > &RigUnit_ParentConstraint(
    AdvancedSettings=(RotationOrderForFilter=XZY)
    Child=(Type=Null,Name="dyn_Hair_Back_04_Null")
    Filter=(TranslationFilter=(bX=True,bY=True,bZ=True),RotationFilter=(bX=false,bY=false,bZ=false),ScaleFilter=(bX=True,bY=True,bZ=True))
    Parents=((Item=(Type=Bone,Name="clavicle_r"),Weight=0),(Item=(Type=Bone,Name="spine_04"),Weight=1),(Item=(Type=Bone,Name="clavicle_l"),Weight=0))
    Weight=1
   )
   > &RigUnit_ParentConstraint(
    AdvancedSettings=(RotationOrderForFilter=XZY)
    Child=(Type=Null,Name="dyn_Hair_Back_05_Null")
    Filter=(TranslationFilter=(bX=True,bY=True,bZ=True),RotationFilter=(bX=false,bY=false,bZ=false),ScaleFilter=(bX=True,bY=True,bZ=True))
    Parents=((Item=(Type=Bone,Name="clavicle_l"),Weight=0),(Item=(Type=Bone,Name="spine_03"),Weight=1),(Item=(Type=Bone,Name="clavicle_r"),Weight=0))
    Weight=1
   )
   > &RigUnit_ParentConstraint(
    AdvancedSettings=(RotationOrderForFilter=XZY)
    Child=(Type=Null,Name="dyn_Hair_Back_06_Null")
    Filter=(TranslationFilter=(bX=True,bY=True,bZ=True),RotationFilter=(bX=false,bY=false,bZ=false),ScaleFilter=(bX=True,bY=True,bZ=True))
    Parents=((Item=(Type=Bone,Name="dyn_Hair_Back_06"),Weight=0),(Item=(Type=Bone,Name="spine_03"),Weight=1))
    Weight=1
   )
   RerouteNode_2 = RigVMDispatchNode(Values="((Type=Bone,Name=\"dyn_Hair_Back_01\"),((),Type=Bone,Name=\"dyn_Hair_Back_02\"),((),Type=Bone,Name=\"dyn_Hair_Back_03\"),((),Type=Bone,Name=\"dyn_Hair_Back_04\"),((),Type=Bone,Name=\"dyn_Hair_Back_05\"),((),Type=Bone,Name=\"dyn_Hair_Back_06\"),((),Type=Bone,Name=\"dyn_ ...(truncated)") {template="DISPATCH_RigVMDispatch_ArrayMake(in Values,out Array)"}
   > SplineIK_2 = RigVMFunctionReferenceNode(
    Primary Axis=(X=1,Y=0,Z=-0.15)
    Samples Per Segment=1
    Secondary Spline Direction=(X=-1,Y=-1,Z=-1)
    Spline Mode=BSpline
    Stretch=false
    Up Axis=(X=0,Y=0,Z=0)
    debug=false
    Bones=@RerouteNode_2.Array
    Controls=RigVMVariableNode() {variable=dyn_Hair_Back_ctrls}.Value
   )
   > TwistBones = RigVMFunctionReferenceNode(
    Axis to Twist=(bX=true,bY=true,bZ=true)
    Bones to Twist=((Type=Bone,Name="dyn_Hair_Back_03"),(Type=Bone,Name="dyn_Hair_Back_04"),(Type=Bone,Name="dyn_Hair_Back_06"))
    Debug=false
    End=(Type=Bone,Name="spine_05")
    Start=(Type=Bone,Name="spine_05")
   )
   > &RigUnit_ModifyTransforms(
    ItemToModify="((Item=(Type=Bone,Name=\"dyn_Hair_Back_07\"),Transform=(Rotation=(X=0,Y=0,Z=0,W=1),Translation=(X=0,Y=0,Z=3),Scale3D=(X=1,Y=1,Z=1))),(Item=(Type=Bone,Name=\"dyn_Hair_Back_04\"),Transform=(Rotation=(X=0,Y=0,Z=0,W=1),Translation=(X=0,Y=0,Z=0),Scale3D=(X=1,Y=1,Z= ...(truncated)"
    Weight=1
    WeightMinimum=0
    ItemToModify.0.Transform.Rotation=&RigVMFunction_MathQuaternionFromEuler(Euler=(X=0,Y=-20,Z=0), RotationOrder=YXZ).Result
    ItemToModify.1.Transform.Rotation=&RigVMFunction_MathQuaternionFromEuler(Euler=(X=0,Y=0,Z=0), RotationOrder=YXZ).Result
   )
   &RigUnit_ParentConstraint(
    AdvancedSettings=(RotationOrderForFilter=XZY)
    Child=(Type=Bone,Name="dyn_Hair_BackL_01")
    Filter=(TranslationFilter=(bX=true,bY=true,bZ=true),RotationFilter=(bX=true,bY=true,bZ=true),ScaleFilter=(bX=True,bY=True,bZ=True))
    Parents=((Item=(Type=Bone,Name="dyn_Hair_Back_01"),Weight=1),(Item=(Type=Bone,Name="spine_05"),Weight=0))
    Weight=1
    ExecutePin=@RigVMFunction_Sequence.B
   )
   > &RigUnit_ParentConstraint(
    AdvancedSettings=(RotationOrderForFilter=XZY)
    Child=(Type=Bone,Name="dyn_Hair_BackL_02")
    Filter=(TranslationFilter=(bX=True,bY=True,bZ=True),RotationFilter=(bX=true,bY=true,bZ=true),ScaleFilter=(bX=True,bY=True,bZ=True))
    Parents=((Item=(Type=Bone,Name="dyn_Hair_Back_02"),Weight=1),(Item=(Type=Bone,Name="spine_05"),Weight=0))
    Weight=1
   )
   > &RigUnit_ParentConstraint(
    AdvancedSettings=(RotationOrderForFilter=XZY)
    Child=(Type=Bone,Name="dyn_Hair_BackL_03")
    Filter=(TranslationFilter=(bX=True,bY=True,bZ=True),RotationFilter=(bX=true,bY=true,bZ=true),ScaleFilter=(bX=True,bY=True,bZ=True))
    Parents=((Item=(Type=Bone,Name="dyn_Hair_Back_03"),Weight=1),(Item=(Type=Bone,Name="spine_05"),Weight=0))
    Weight=1
   )
   > &RigUnit_ParentConstraint(
    AdvancedSettings=(RotationOrderForFilter=XZY)
    Child=(Type=Bone,Name="dyn_Hair_BackL_04")
    Filter=(TranslationFilter=(bX=True,bY=True,bZ=True),RotationFilter=(bX=true,bY=true,bZ=true),ScaleFilter=(bX=True,bY=True,bZ=True))
    Parents=((Item=(Type=Bone,Name="dyn_Hair_Back_04"),Weight=1),(Item=(Type=Bone,Name="spine_05"),Weight=0))
    Weight=1
   )
   > &RigUnit_ParentConstraint(
    AdvancedSettings=(RotationOrderForFilter=XZY)
    Child=(Type=Bone,Name="dyn_Hair_BackL_05")
    Filter=(TranslationFilter=(bX=True,bY=True,bZ=True),RotationFilter=(bX=true,bY=true,bZ=true),ScaleFilter=(bX=True,bY=True,bZ=True))
    Parents=((Item=(Type=Bone,Name="dyn_Hair_Back_05"),Weight=1),(Item=(Type=Bone,Name="spine_04"),Weight=0))
    Weight=1
   )
   > &RigUnit_ParentConstraint(
    AdvancedSettings=(RotationOrderForFilter=XZY)
    Child=(Type=Bone,Name="dyn_Hair_BackL_06")
    Filter=(TranslationFilter=(bX=true,bY=true,bZ=true),RotationFilter=(bX=true,bY=true,bZ=true),ScaleFilter=(bX=True,bY=True,bZ=True))
    Parents=((Item=(Type=Bone,Name="dyn_Hair_Back_06"),Weight=1),(Item=(Type=Bone,Name="spine_02"),Weight=0))
    Weight=1
   )
   &RigUnit_ParentConstraint(
    AdvancedSettings=(RotationOrderForFilter=XZY)
    Child=(Type=Bone,Name="dyn_Hair_BackR_01")
    Filter=(TranslationFilter=(bX=true,bY=true,bZ=true),RotationFilter=(bX=true,bY=true,bZ=true),ScaleFilter=(bX=True,bY=True,bZ=True))
    Parents=((Item=(Type=Bone,Name="dyn_Hair_Back_01"),Weight=1),(Item=(Type=Bone,Name="spine_05"),Weight=0))
    Weight=1
    ExecutePin=@RigVMFunction_Sequence.C
   )
   > &RigUnit_ParentConstraint(
    AdvancedSettings=(RotationOrderForFilter=XZY)
    Child=(Type=Bone,Name="dyn_Hair_BackR_02")
    Filter=(TranslationFilter=(bX=True,bY=True,bZ=True),RotationFilter=(bX=true,bY=true,bZ=true),ScaleFilter=(bX=True,bY=True,bZ=True))
    Parents=((Item=(Type=Bone,Name="dyn_Hair_Back_02"),Weight=1),(Item=(Type=Bone,Name="spine_05"),Weight=0))
    Weight=1
   )
   > &RigUnit_ParentConstraint(
    AdvancedSettings=(RotationOrderForFilter=XZY)
    Child=(Type=Bone,Name="dyn_Hair_BackR_03")
    Filter=(TranslationFilter=(bX=True,bY=True,bZ=True),RotationFilter=(bX=true,bY=true,bZ=true),ScaleFilter=(bX=True,bY=True,bZ=True))
    Parents=((Item=(Type=Bone,Name="dyn_Hair_Back_03"),Weight=1),(Item=(Type=Bone,Name="spine_05"),Weight=0))
    Weight=1
   )
   > &RigUnit_ParentConstraint(
    AdvancedSettings=(RotationOrderForFilter=XZY)
    Child=(Type=Bone,Name="dyn_Hair_BackR_04")
    Filter=(TranslationFilter=(bX=True,bY=True,bZ=True),RotationFilter=(bX=true,bY=true,bZ=true),ScaleFilter=(bX=True,bY=True,bZ=True))
    Parents=((Item=(Type=Bone,Name="dyn_Hair_Back_04"),Weight=1),(Item=(Type=Bone,Name="spine_05"),Weight=0))
    Weight=1
   )
   > &RigUnit_ParentConstraint(
    AdvancedSettings=(RotationOrderForFilter=XZY)
    Child=(Type=Bone,Name="dyn_Hair_BackR_05")
    Filter=(TranslationFilter=(bX=True,bY=True,bZ=True),RotationFilter=(bX=true,bY=true,bZ=true),ScaleFilter=(bX=True,bY=True,bZ=True))
    Parents=((Item=(Type=Bone,Name="dyn_Hair_Back_05"),Weight=1),(Item=(Type=Bone,Name="spine_04"),Weight=0))
    Weight=1
   )
   > &RigUnit_ParentConstraint(
    AdvancedSettings=(RotationOrderForFilter=XZY)
    Child=(Type=Bone,Name="dyn_Hair_BackR_06")
    Filter=(TranslationFilter=(bX=true,bY=true,bZ=true),RotationFilter=(bX=true,bY=true,bZ=true),ScaleFilter=(bX=True,bY=True,bZ=True))
    Parents=((Item=(Type=Bone,Name="dyn_Hair_Back_06"),Weight=1),(Item=(Type=Bone,Name="spine_02"),Weight=0))
    Weight=1
   )
   RigVMFunctionReturnNode(ExecuteContext=@RigVMFunction_Sequence.D)
  ]}
 New Function_ContainedGraph {nodes [
   RigVMFunctionEntryNode()
   > RigVMFunction_Sequence = RigVMAggregateNode()
   Delta Rotation = RigVMFunctionReferenceNode(Child=head, Parent=neck_01, Source Maximum=-43, Source Minimum=-5, ExecuteContext=@RigVMFunction_Sequence.B)
   > &RigUnit_ModifyTransforms(
    ItemToModify=((Item=(Type=Bone,Name="dyn_Hair_BackR_03"),Transform=(Rotation=(X=0,Y=0,Z=0,W=1),Translation=(X=1.5,Y=0,Z=0),Scale3D=(X=1,Y=1,Z=1))))
    Weight=@"Delta Rotation".Delta Z
   )
   Multiply_1 = RigVMAggregateNode(
    C=&RigVMFunction_MathDoubleRemap(SourceMaximum=1, SourceMinimum=0, TargetMaximum=0, TargetMinimum=1, bClamp=False, Value=RigVMVariableNode() {variable=isHeadUpDelta}.Value).Result
    B=RigVMVariableNode() {variable=CloseGroundRemap}.Value
    A=RigVMVariableNode() {variable=isCrouching}.Value
   )
   > &RigUnit_ModifyTransforms(
    ItemToModify=((Item=(Type=Bone,Name="dyn_Hair_BackR_01"),Transform=(Rotation=(X=0,Y=0,Z=0,W=1),Translation=(X=-1,Y=2.2,Z=0),Scale3D=(X=1,Y=1,Z=1))))
    ItemToModify.0.Transform.Rotation=&RigVMFunction_MathQuaternionFromEuler(Euler=(X=0,Y=0,Z=-4)).Result
    Weight=@Multiply_1.Result
   )
   n1 = &RigVMFunction_MathQuaternionFromEuler(Euler=(X=0,Y=-2,Z=0))
   Remap = &RigVMFunction_MathDoubleRemap(SourceMaximum=1, SourceMinimum=0, TargetMaximum=0, TargetMinimum=1, bClamp=False, Value=RigVMVariableNode() {variable=isHeadUpDelta}.Value)
   > &RigUnit_ModifyTransforms(
    ItemToModify="((Item=(Type=Bone,Name=\"dyn_Hair_Back_03\"),Transform=(Rotation=(X=0,Y=0,Z=0,W=1),Translation=(X=0,Y=0,Z=1),Scale3D=(X=1,Y=1,Z=1))),(Item=(Type=Bone,Name=\"dyn_Hair_BackL_03\"),Transform=(Rotation=(X=0,Y=0,Z=0,W=1),Translation=(X=0,Y=0,Z=0),Scale3D=(X=1,Y=1,Z ...(truncated)"
    ItemToModify.1.Transform.Rotation=@n1.Result
    ItemToModify.2.Transform.Rotation=@n1.Result
    Weight=@Remap.Result
   )
   n2 = &RigVMFunction_MathQuaternionFromEuler(Euler=(X=0,Y=5,Z=0))
   > n3 = &RigUnit_ModifyTransforms(
    ItemToModify="((Item=(Type=Bone,Name=\"dyn_Hair_Back_03\"),Transform=(Rotation=(X=0,Y=0,Z=0,W=1),Translation=(X=0,Y=0,Z=0),Scale3D=(X=1,Y=1,Z=1))),(Item=(Type=Bone,Name=\"dyn_Hair_BackL_03\"),Transform=(Rotation=(X=0,Y=0,Z=0,W=1),Translation=(X=0,Y=0,Z=1.5),Scale3D=(X=1,Y=1 ...(truncated)"
    ItemToModify.0.Transform.Rotation=&RigVMFunction_MathQuaternionFromEuler(Euler=(X=0,Y=5,Z=0)).Result
    ItemToModify.1.Transform.Rotation=@n2.Result
    ItemToModify.2.Transform.Rotation=@n2.Result
    Weight=RigVMVariableNode() {variable=isHeadUpDelta}.Value
   )
   RigVMFunction_ControlFlowBranch = &RigVMFunction_ControlFlowBranch(ExecuteContext=@n3.ExecutePin, Condition=RigVMVariableNode() {variable=isFrontEnd}.Value)
   HairCrouch_DEF = RigVMFunctionReferenceNode(ExecuteContext=@RigVMFunction_ControlFlowBranch.False)
   &RigUnit_ModifyTransforms(
    ItemToModify="((Item=(Type=Bone,Name=\"dyn_Hair_BackR_02\"),Transform=(Rotation=(X=0,Y=0,Z=0,W=1),Translation=(X=0,Y=1.5,Z=1),Scale3D=(X=1,Y=1,Z=1))),(Item=(Type=Bone,Name=\"dyn_Hair_Back_03\"),Transform=(Rotation=(X=0,Y=-0.0436,Z=0,W=0.999),Translation=(X=0,Y=0,Z=0),Scale3 ...(truncated)"
    Weight=1
    WeightMinimum=0
    ItemToModify.1.Transform.Rotation=&RigVMFunction_MathQuaternionFromEuler(Euler=(X=0,Y=-5,Z=0)).Result
    ItemToModify.3.Transform.Rotation=&RigVMFunction_MathQuaternionFromEuler(Euler=(X=0,Y=60,Z=0)).Result
    ExecutePin=@RigVMFunction_ControlFlowBranch.True
   )
   &RigUnit_ParentConstraint(
    AdvancedSettings=(InterpolationType=Average,RotationOrderForFilter=XZY)
    Child=(Type=Bone,Name="base_Hair_FrontR")
    Parents=((Item=(Type=Bone,Name="spine_05"),Weight=0.6),(Item=(Type=Bone,Name="head"),Weight=0.4))
    Weight=1
    ExecutePin=@RigVMFunction_Sequence.E
   )
   > &RigUnit_ParentConstraint(
    AdvancedSettings=(InterpolationType=Average,RotationOrderForFilter=XZY)
    Child=(Type=Bone,Name="base_Hair_FrontL")
    Parents=((Item=(Type=Bone,Name="spine_05"),Weight=0.6),(Item=(Type=Bone,Name="head"),Weight=0.4))
    Weight=1
   )
   RigVMFunctionReturnNode(ExecuteContext=@RigVMFunction_Sequence.F)
   n4 = &RigVMFunction_MathQuaternionFromEuler(Euler=(X=0,Y=-5,Z=0))
   n5 = RigVMVariableNode() {variable=isSkydiveFloating}
   &RigUnit_ModifyTransforms(
    ItemToModify="((Item=(Type=Bone,Name=\"dyn_Hair_Back_03\"),Transform=(Rotation=(X=0,Y=0,Z=0,W=1),Translation=(X=-3.9928,Y=0,Z=0.6507),Scale3D=(X=1,Y=1,Z=1))),(Item=(Type=Bone,Name=\"dyn_Hair_BackL_03\"),Transform=(Rotation=(X=0,Y=0,Z=0,W=1),Translation=(X=-2.339,Y=-1,Z=0.79 ...(truncated)"
    ItemToModify.0.Transform.Rotation=@n4.Result
    ItemToModify.1.Transform.Rotation=@n4.Result
    ItemToModify.2.Transform.Rotation=@n4.Result
    Weight=@n5.Value
    ExecutePin=@RigVMFunction_Sequence.D
   )
   > &RigUnit_ModifyTransforms(
    ItemToModify="((Item=(Type=Bone,Name=\"dyn_Hair_Back_01\"),Transform=(Rotation=(X=0,Y=0,Z=0,W=1),Translation=(X=2.4054,Y=0,Z=2.5039),Scale3D=(X=1,Y=1,Z=1))),(Item=(Type=Bone,Name=\"dyn_Hair_BackR_01\"),Transform=(Rotation=(X=0,Y=0,Z=0,W=1),Translation=(X=1.661,Y=0,Z=2.3118) ...(truncated)"
    WeightMinimum=0
    Weight=@n5.Value
   )
   n6 = &RigVMFunction_MathQuaternionFromEuler(Euler=(X=0,Y=-5,Z=0))
   > &RigUnit_ModifyTransforms(
    ItemToModify="((Item=(Type=Bone,Name=\"dyn_Hair_Back_03\"),Transform=(Rotation=(X=0,Y=0,Z=0,W=1),Translation=(X=0,Y=0,Z=1),Scale3D=(X=1,Y=1,Z=1))),(Item=(Type=Bone,Name=\"dyn_Hair_BackL_03\"),Transform=(Rotation=(X=0,Y=0,Z=0,W=1),Translation=(X=0,Y=0,Z=2),Scale3D=(X=1,Y=1,Z ...(truncated)"
    ItemToModify.3.Transform.Rotation=&RigVMFunction_MathQuaternionFromEuler(Euler=(X=0,Y=30,Z=0)).Result
    ItemToModify.0.Transform.Rotation=@n6.Result
    ItemToModify.1.Transform.Rotation=@n6.Result
    ItemToModify.2.Transform.Rotation=@n6.Result
    Weight=RigVMVariableNode() {variable=isHangGliding}.Value
   )
   n7 = &RigUnit_SphericalPoseReader(
    ActiveRegionScaleFactors=(PositiveWidth=0.6,NegativeWidth=0.6,PositiveHeight=0.6,NegativeHeight=0.6)
    ActiveRegionSize=0.1
    Debug=(bDrawDebug=false,DebugScale=25,DebugSegments=20,DebugThickness=0.25)
    DriverAxis=(X=1,Y=0,Z=0)
    DriverItem=(Type=Bone,Name="upperarm_l")
    FalloffRegionScaleFactors=(PositiveWidth=1,NegativeWidth=0.6,PositiveHeight=0.6,NegativeHeight=0.6)
    FalloffSize=0.55
    FlipHeightScaling=false
    FlipWidthScaling=false
    OptionalParentItem=(Type=Bone,Name="spine_05")
    RotationOffset=(X=-80,Y=-30,Z=-10)
    ExecutePin=@RigVMFunction_Sequence.A
   )
   Multiply = &RigVMFunction_MathDoubleMul(
    B=&RigVMFunction_MathDoubleRemap(
     SourceMaximum=1
     SourceMinimum=0
     TargetMaximum=0
     TargetMinimum=1
     bClamp=False
     Value=RigVMVariableNode() {variable=isSkydiveFloating}.Value
    ).Result
    A=@n7.OutputParam
   )
   > &RigUnit_ParentConstraint(
    AdvancedSettings=(RotationOrderForFilter=XZY)
    Child=(Type=Bone,Name="dyn_Hair_BackL_03")
    Filter=(TranslationFilter=(bX=false,bY=false,bZ=false),RotationFilter=(bX=true,bY=true,bZ=true),ScaleFilter=(bX=True,bY=True,bZ=True))
    Parents=((Item=(Type=Bone,Name="dyn_Hair_Back_03"),Weight=0.3),(Item=(Type=Bone,Name="upperarm_l"),Weight=0.7))
    Weight=@Multiply.Result
   )
   n8 = &RigVMFunction_MathQuaternionFromEuler(Euler=(X=0,Y=-10,Z=20), RotationOrder=YXZ)
   > &RigUnit_ModifyTransforms(
    ItemToModify="((Item=(Type=Bone,Name=\"dyn_Hair_BackL_02\"),Transform=(Rotation=(X=0,Y=0,Z=0,W=1),Translation=(X=1,Y=-1,Z=2),Scale3D=(X=1,Y=1,Z=1))),(Item=(Type=Bone,Name=\"dyn_Hair_BackL_04\"),Transform=(Rotation=(X=0,Y=0,Z=0,W=1),Translation=(X=0,Y=0,Z=0),Scale3D=(X=1,Y=1 ...(truncated)"
    ItemToModify.0.Transform.Rotation=@n8.Result
    ItemToModify.1.Transform.Rotation=@n8.Result
    ItemToModify.2.Transform.Rotation=&RigVMFunction_MathQuaternionFromEuler(Euler=(X=-25,Y=0,Z=20)).Result
    Weight=@Multiply.Result
   )
   n9 = &RigVMFunction_MathQuaternionFromEuler(Euler=(X=0,Y=-5,Z=0))
   > &RigUnit_ModifyTransforms(
    ItemToModify="((Item=(Type=Bone,Name=\"dyn_Hair_Back_03\"),Transform=(Rotation=(X=0,Y=0,Z=0,W=1),Translation=(X=0,Y=1,Z=1),Scale3D=(X=1,Y=1,Z=1))),(Item=(Type=Bone,Name=\"dyn_Hair_Back_04\"),Transform=(Rotation=(X=0,Y=0,Z=0,W=1),Translation=(X=0,Y=0,Z=0),Scale3D=(X=1,Y=1,Z= ...(truncated)"
    ItemToModify.2.Transform.Rotation=&RigVMFunction_MathQuaternionFromEuler(Euler=(X=0,Y=-2,Z=0)).Result
    ItemToModify.0.Transform.Rotation=@n9.Result
    ItemToModify.1.Transform.Rotation=@n9.Result
    Weight=@Multiply.Result
   )
   > n10 = &RigUnit_SphericalPoseReader(
    ActiveRegionScaleFactors=(PositiveWidth=0.6,NegativeWidth=0.6,PositiveHeight=0.6,NegativeHeight=0.6)
    ActiveRegionSize=0.1
    Debug=(bDrawDebug=false,DebugScale=25,DebugSegments=20,DebugThickness=0.25)
    DriverAxis=(X=-1,Y=0,Z=0)
    DriverItem=(Type=Bone,Name="upperarm_r")
    FalloffRegionScaleFactors=(PositiveWidth=1,NegativeWidth=0.6,PositiveHeight=0.6,NegativeHeight=0.6)
    FalloffSize=0.6
    FlipHeightScaling=false
    FlipWidthScaling=true
    OptionalParentItem=(Type=Bone,Name="spine_05")
    RotationOffset=(X=80,Y=-30,Z=-10)
   )
   n11 = &RigVMFunction_MathDoubleMul(
    B=&RigVMFunction_MathDoubleRemap(
     SourceMaximum=1
     SourceMinimum=0
     TargetMaximum=0
     TargetMinimum=1
     bClamp=False
     Value=RigVMVariableNode() {variable=isSkydiveFloating}.Value
    ).Result
    A=@n10.OutputParam
   )
   > &RigUnit_ParentConstraint(
    AdvancedSettings=(RotationOrderForFilter=XZY)
    Child=(Type=Bone,Name="dyn_Hair_BackR_03")
    Filter=(TranslationFilter=(bX=false,bY=false,bZ=false),RotationFilter=(bX=true,bY=true,bZ=true),ScaleFilter=(bX=True,bY=True,bZ=True))
    Parents=((Item=(Type=Bone,Name="dyn_Hair_Back_03"),Weight=0.3),(Item=(Type=Bone,Name="upperarm_r"),Weight=0.7))
    Weight=@n11.Result
   )
   n12 = &RigVMFunction_MathQuaternionFromEuler(Euler=(X=0,Y=-10,Z=-20), RotationOrder=YXZ)
   > &RigUnit_ModifyTransforms(
    ItemToModify="((Item=(Type=Bone,Name=\"dyn_Hair_BackR_02\"),Transform=(Rotation=(X=0,Y=0,Z=0,W=1),Translation=(X=1,Y=-1,Z=2),Scale3D=(X=1,Y=1,Z=1))),(Item=(Type=Bone,Name=\"dyn_Hair_BackR_04\"),Transform=(Rotation=(X=0,Y=0,Z=0,W=1),Translation=(X=0,Y=0,Z=0),Scale3D=(X=1,Y=1 ...(truncated)"
    ItemToModify.0.Transform.Rotation=@n12.Result
    ItemToModify.1.Transform.Rotation=@n12.Result
    ItemToModify.2.Transform.Rotation=&RigVMFunction_MathQuaternionFromEuler(Euler=(X=25,Y=0,Z=20)).Result
    Weight=@n11.Result
   )
   n13 = &RigVMFunction_MathQuaternionFromEuler(Euler=(X=0,Y=-5,Z=0))
   > &RigUnit_ModifyTransforms(
    ItemToModify="((Item=(Type=Bone,Name=\"dyn_Hair_Back_03\"),Transform=(Rotation=(X=0,Y=0,Z=0,W=1),Translation=(X=0,Y=-1,Z=0),Scale3D=(X=1,Y=1,Z=1))),(Item=(Type=Bone,Name=\"dyn_Hair_Back_04\"),Transform=(Rotation=(X=0,Y=0,Z=0,W=1),Translation=(X=0,Y=0,Z=0),Scale3D=(X=1,Y=1,Z ...(truncated)"
    ItemToModify.2.Transform.Rotation=&RigVMFunction_MathQuaternionFromEuler(Euler=(X=0,Y=-2,Z=0)).Result
    ItemToModify.0.Transform.Rotation=@n13.Result
    ItemToModify.1.Transform.Rotation=@n13.Result
    Weight=@n11.Result
   )
  ]}
 New Function_ContainedGraph {nodes [
   n1 = RigVMFunctionEntryNode()
   RigVMFunction_ControlFlowBranch = &RigVMFunction_ControlFlowBranch(Condition=@n1.Debug Controllers, ExecuteContext=@n1.ExecuteContext)
   RigVMFunctionReturnNode(ExecuteContext=@RigVMFunction_ControlFlowBranch.Completed)
   RigUnit_GetTransform = &RigUnit_GetTransform(Item=(Type=Control,Name="head_ctrl"), bInitial=False)
   &RigUnit_SetTransform(
    Item=(Type=Bone,Name="head")
    Weight=1
    bInitial=False
    Value=@RigUnit_GetTransform.Transform
    ExecutePin=@RigVMFunction_ControlFlowBranch.True
   )
   > RigUnit_SetTransform = &RigUnit_SetTransform(
    Item=(Type=Bone,Name="upperarm_l")
    Weight=1
    bInitial=False
    Value=&RigUnit_GetTransform(Item=(Type=Control,Name="upperarm_l_ctrl"), bInitial=False).Transform
   )
   > &RigUnit_SetTransform(
    Item=(Type=Bone,Name="upperarm_r")
    Weight=1
    bInitial=False
    Value=&RigUnit_GetTransform(Item=(Type=Control,Name="upperarm_r_ctrl"), bInitial=False).Transform
   )
  ]}
 New Function_ContainedGraph {nodes [
   n1 = RigVMFunctionEntryNode()
   RigVMFunction_MathQuaternionToEuler = &RigVMFunction_MathQuaternionToEuler(
    RotationOrder=ZYX
    Value=&RigUnit_GetRelativeTransformForItem(
     Child=(Type=Bone,Name="hand_l")
     Parent=(Type=Bone,Name="lowerarm_l")
     bChildInitial=False
     bParentInitial=False
     Child.Name=@n1.Child
     Parent.Name=@n1.Parent
    ).RelativeTransform.Rotation
   )
   Remap = &RigVMFunction_MathDoubleRemap(
    TargetMaximum=1
    TargetMinimum=0
    Value=@RigVMFunction_MathQuaternionToEuler.Result.Z
    SourceMaximum=@n1.Source Maximum
    SourceMinimum=@n1.Source Minimum
   )
   > RigVMFunctionReturnNode(Delta Z=@Remap.Result)
  ]}
 New Function_ContainedGraph {nodes [
   n1 = RigVMFunctionEntryNode()
   RigVMFunction_Sequence = &RigVMFunction_Sequence(ExecuteContext=@n1.ExecuteContext)
   &RigUnit_ParentConstraint(
    AdvancedSettings=(InterpolationType=Average,RotationOrderForFilter=XZY)
    Child=(Type=Bone,Name="base_hat_m")
    Filter=(TranslationFilter=(bX=false,bY=true,bZ=false),RotationFilter=(bX=false,bY=false,bZ=false),ScaleFilter=(bX=True,bY=True,bZ=True))
    Parents=((Item=(Type=Bone,Name="dyn_Hair_Back_03"),Weight=0.25),(Item=(Type=Bone,Name="head"),Weight=0.75))
    Weight=0
    ExecutePin=@RigVMFunction_Sequence.A
   )
   > &RigUnit_ParentConstraint(
    AdvancedSettings=(InterpolationType=Average,RotationOrderForFilter=XZY)
    Child=(Type=Bone,Name="base_hat_m")
    Filter=(TranslationFilter=(bX=false,bY=false,bZ=false),RotationFilter=(bX=false,bY=true,bZ=true),ScaleFilter=(bX=True,bY=True,bZ=True))
    Parents=((Item=(Type=Bone,Name="dyn_Hair_Back_02"),Weight=0.5),(Item=(Type=Bone,Name="head"),Weight=0.5))
    Weight=0.3
   )
   > n2 = &RigUnit_ParentConstraint(
    AdvancedSettings=(InterpolationType=Average,RotationOrderForFilter=XZY)
    Child=(Type=Bone,Name="dyn_branch_C01")
    Filter=(TranslationFilter=(bX=false,bY=false,bZ=false),RotationFilter=(bX=True,bY=True,bZ=True),ScaleFilter=(bX=True,bY=True,bZ=True))
    Parents=((Item=(Type=Bone,Name="dyn_hat_l_B01"),Weight=0.5),(Item=(Type=Bone,Name="dyn_hat_l_C01"),Weight=0.5))
    Weight=0.5
   )
   n3 = &RigVMFunction_Sequence(ExecuteContext=@n2.ExecutePin)
   &RigUnit_ModifyTransforms(
    ItemToModify=((Item=(Type=Bone,Name="base_hat_m"),Transform=(Rotation=(X=0,Y=-0.0872,Z=0,W=0.9962),Translation=(X=1,Y=0,Z=0),Scale3D=(X=1,Y=1,Z=1))))
    WeightMinimum=0
    ItemToModify.0.Transform.Rotation=&RigVMFunction_MathQuaternionFromEuler(Euler=(X=0,Y=10,Z=0), RotationOrder=YXZ).Result
    Weight=RigVMVariableNode() {variable=isSkydiveFloating}.Value
    ExecutePin=@n3.B
   )
   n4 = &RigVMFunction_MathQuaternionFromEuler(Euler=(X=0,Y=-10,Z=0), RotationOrder=YXZ)
   > &RigUnit_ModifyTransforms(
    ItemToModify="((Item=(Type=Bone,Name=\"base_hat_m\"),Transform=(Rotation=(X=0,Y=-0.0872,Z=0,W=0.9962),Translation=(X=1,Y=0,Z=-1),Scale3D=(X=1,Y=1,Z=1))),(Item=(Type=Bone,Name=\"base_hat_l_B\"),Transform=(Rotation=(X=0,Y=0,Z=0,W=1),Translation=(X=-2,Y=0,Z=0),Scale3D=(X=1,Y=1 ...(truncated)"
    WeightMinimum=0
    ItemToModify.0.Transform.Rotation=&RigVMFunction_MathQuaternionFromEuler(Euler=(X=0,Y=5,Z=0), RotationOrder=YXZ).Result
    ItemToModify.1.Transform.Rotation=@n4.Result
    ItemToModify.2.Transform.Rotation=@n4.Result
    Weight=RigVMVariableNode() {variable=isHangGliding}.Value
   )
   n5 = RigVMVariableNode() {variable=isHeadUpDelta}
   &RigUnit_ModifyTransforms(
    ItemToModify="((Item=(Type=Bone,Name=\"base_hat_m\"),Transform=(Rotation=(X=0,Y=-0.1564,Z=0,W=0.9877),Translation=(X=0,Y=0,Z=0),Scale3D=(X=1,Y=1,Z=1))),(Item=(Type=Bone,Name=\"dyn_branch_C02\"),Transform=(Rotation=(X=-0.0113,Y=-0.2586,Z=-0.0421,W=0.965),Translation=(X=0,Y=0 ...(truncated)"
    WeightMinimum=0
    ItemToModify.0.Transform.Rotation=&RigVMFunction_MathQuaternionFromEuler(Euler=(X=0,Y=-8,Z=5), RotationOrder=YXZ).Result
    Weight=@n5.Value
    ExecutePin=@n3.A
   )
   Multiply = &RigVMFunction_MathDoubleMul(A=RigVMVariableNode() {variable=isGunEquipped}.Value, B=@n5.Value)
   > &RigUnit_ModifyTransforms(
    ItemToModify=((Item=(Type=Bone,Name="base_hat_m"),Transform=(Rotation=(X=0,Y=0,Z=0,W=1),Translation=(X=0.5,Y=0,Z=0),Scale3D=(X=1,Y=1,Z=1))))
    WeightMinimum=0
    Weight=@Multiply.Result
   )
   Multiply_1 = RigVMAggregateNode(
    A=RigVMVariableNode() {variable=CloseGroundRemap}.Value
    B=RigVMVariableNode() {variable=isHeadUpExtreme}.Value
    C=RigVMVariableNode() {variable=isCrouching}.Value
   )
   > &RigUnit_ModifyTransforms(
    ItemToModify="((Item=(Type=Bone,Name=\"base_hat_m\"),Transform=(Rotation=(X=0,Y=-0.1564,Z=0,W=0.9877),Translation=(X=3,Y=0,Z=0),Scale3D=(X=1,Y=1,Z=1))),(Item=(Type=Bone,Name=\"base_hat_bk\"),Transform=(Rotation=(X=0,Y=0,Z=0,W=1),Translation=(X=0,Y=0,Z=0),Scale3D=(X=1,Y=1,Z= ...(truncated)"
    WeightMinimum=0
    ItemToModify.0.Transform.Rotation=&RigVMFunction_MathQuaternionFromEuler(Euler=(X=0,Y=-8,Z=0), RotationOrder=YXZ).Result
    ItemToModify.1.Transform.Rotation=&RigVMFunction_MathQuaternionFromEuler(Euler=(X=0,Y=-5,Z=0), RotationOrder=YXZ).Result
    Weight=@Multiply_1.Result
   )
   > &RigUnit_ModifyTransforms(
    ItemToModify=((Item=(Type=Bone,Name="base_hat_l_B"),Transform=(Rotation=(X=0,Y=0,Z=0,W=1),Translation=(X=0,Y=0,Z=2),Scale3D=(X=1,Y=1,Z=1))),(Item=(Type=Bone,Name="base_hat_r_B"),Transform=(Rotation=(X=0,Y=0,Z=0,W=1),Translation=(X=0,Y=0,Z=2),Scale3D=(X=1,Y=1,Z=1))))
    WeightMinimum=0
    Weight=@Multiply_1.Result
   )
   Multiply_1_1 = RigVMAggregateNode(
    D=&RigVMFunction_MathDoubleRemap(
     SourceMaximum=0.5
     SourceMinimum=0
     TargetMaximum=1
     TargetMinimum=0
     bClamp=False
     Value=RigVMVariableNode() {variable=isHeadUpExtreme}.Value
    ).Result
    B=RigVMVariableNode() {variable=isCrouching}.Value
    A=RigVMVariableNode() {variable=isGunEquipped}.Value
    C=RigVMVariableNode() {variable=isTargeting}.Value
   )
   > &RigUnit_ModifyTransforms(
    ItemToModify=((Item=(Type=Bone,Name="base_hat_m"),Transform=(Rotation=(X=0,Y=0,Z=0,W=1),Translation=(X=0.5,Y=0,Z=0),Scale3D=(X=1,Y=1,Z=1))))
    WeightMinimum=0
    ItemToModify.0.Transform.Rotation=&RigVMFunction_MathQuaternionFromEuler(Euler=(X=0,Y=-5,Z=0), RotationOrder=YXZ).Result
    Weight=&RigVMFunction_MathDoubleRemap(SourceMaximum=1, SourceMinimum=0, TargetMaximum=1, TargetMinimum=0, Value=@Multiply_1_1.Result).Result
   )
   RigVMFunctionReturnNode(ExecuteContext=@RigVMFunction_Sequence.B)
  ]}
 New Function_ContainedGraph {nodes [
   RigVMFunctionEntryNode()
   > RigVMFunction_Sequence = RigVMAggregateNode()
   RigVMFunctionReturnNode(ExecuteContext=@RigVMFunction_Sequence.C)
   n1 = &RigVMFunction_MathQuaternionFromEuler(Euler=(X=0,Y=0,Z=0), RotationOrder=YXZ)
   Multiply_1_2 = RigVMAggregateNode(
    C=&RigVMFunction_MathDoubleRemap(
     SourceMaximum=1
     SourceMinimum=0
     TargetMaximum=1
     TargetMinimum=0
     bClamp=False
     Value=RigVMVariableNode() {variable=CloseGroundRemap}.Value
    ).Result
    B=RigVMVariableNode() {variable=isCrouching}.Value
    A=RigVMVariableNode() {variable=isHeadUpDelta}.Value
   )
   &RigUnit_ModifyTransforms(
    ItemToModify="((Item=(Type=Bone,Name=\"dyn_Hair_BackR_02\"),Transform=(Rotation=(X=0,Y=0,Z=0,W=1),Translation=(X=2,Y=0,Z=3),Scale3D=(X=1,Y=1,Z=1))),(Item=(Type=Bone,Name=\"dyn_Hair_BackL_02\"),Transform=(Rotation=(X=0,Y=0,Z=0,W=1),Translation=(X=2,Y=0,Z=0),Scale3D=(X=1,Y=1, ...(truncated)"
    WeightMinimum=0
    ItemToModify.1.Transform.Rotation=@n1.Result
    ItemToModify.2.Transform.Rotation=@n1.Result
    ItemToModify.0.Transform.Rotation=&RigVMFunction_MathQuaternionFromEuler(Euler=(X=0,Y=10,Z=0), RotationOrder=YXZ).Result
    Weight=@Multiply_1_2.Result
    ExecutePin=@RigVMFunction_Sequence.B
   )
   n2 = &RigVMFunction_MathQuaternionFromEuler(Euler=(X=0,Y=15,Z=0), RotationOrder=YXZ)
   > &RigUnit_ModifyTransforms(
    ItemToModify="((Item=(Type=Bone,Name=\"dyn_Hair_BackR_03\"),Transform=(Rotation=(X=0,Y=0,Z=0,W=1),Translation=(X=2,Y=0,Z=0),Scale3D=(X=1,Y=1,Z=1))),(Item=(Type=Bone,Name=\"dyn_Hair_BackL_03\"),Transform=(Rotation=(X=0,Y=0,Z=0,W=1),Translation=(X=2,Y=0,Z=0),Scale3D=(X=1,Y=1, ...(truncated)"
    WeightMinimum=0
    ItemToModify.0.Transform.Rotation=@n2.Result
    ItemToModify.1.Transform.Rotation=@n2.Result
    ItemToModify.2.Transform.Rotation=@n2.Result
    Weight=@Multiply_1_2.Result
   )
   n3 = &RigVMFunction_MathQuaternionFromEuler(Euler=(X=0,Y=-5,Z=0), RotationOrder=YXZ)
   > &RigUnit_ModifyTransforms(
    ItemToModify="((Item=(Type=Bone,Name=\"dyn_Hair_Back_04\"),Transform=(Rotation=(X=0,Y=0,Z=0,W=1),Translation=(X=0,Y=0,Z=0),Scale3D=(X=1,Y=1,Z=1))),(Item=(Type=Bone,Name=\"dyn_Hair_BackL_04\"),Transform=(Rotation=(X=0,Y=0,Z=0,W=1),Translation=(X=0,Y=0,Z=0),Scale3D=(X=1,Y=1,Z ...(truncated)"
    WeightMinimum=0
    ItemToModify.0.Transform.Rotation=@n3.Result
    ItemToModify.1.Transform.Rotation=@n3.Result
    ItemToModify.2.Transform.Rotation=@n3.Result
    Weight=@Multiply_1_2.Result
   )
   > &RigUnit_ModifyTransforms(
    ItemToModify="((Item=(Type=Bone,Name=\"dyn_Hair_Back_07\"),Transform=(Rotation=(X=0,Y=0,Z=0,W=1),Translation=(X=0,Y=0,Z=0),Scale3D=(X=1,Y=1,Z=1))),(Item=(Type=Bone,Name=\"dyn_Hair_Back_05\"),Transform=(Rotation=(X=0,Y=0,Z=0,W=1),Translation=(X=0,Y=0,Z=2),Scale3D=(X=1,Y=1,Z= ...(truncated)"
    WeightMinimum=0
    ItemToModify.0.Transform.Rotation=&RigVMFunction_MathQuaternionFromEuler(Euler=(X=0,Y=-30,Z=0), RotationOrder=YXZ).Result
    ItemToModify.1.Transform.Rotation=&RigVMFunction_MathQuaternionFromEuler(Euler=(X=0,Y=-10,Z=0), RotationOrder=YXZ).Result
    Weight=@Multiply_1_2.Result
   )
   n4 = &RigVMFunction_MathQuaternionFromEuler(Euler=(X=0,Y=6,Z=0), RotationOrder=YXZ)
   Multiply_1_1_2 = RigVMAggregateNode(
    A=&RigVMFunction_MathDoubleRemap(
     SourceMaximum=0.5
     SourceMinimum=0
     TargetMaximum=1
     TargetMinimum=0
     bClamp=False
     Value=RigVMVariableNode() {variable=isHeadUpExtreme}.Value
    ).Result
    C=RigVMVariableNode() {variable=isCrouching}.Value
    B=RigVMVariableNode() {variable=isGunEquipped}.Value
    D=RigVMVariableNode() {variable=isTargeting}.Value
   )
   > &RigUnit_ModifyTransforms(
    ItemToModify="((Item=(Type=Bone,Name=\"dyn_Hair_Back_04\"),Transform=(Rotation=(X=0,Y=0,Z=0,W=1),Translation=(X=0,Y=0,Z=0),Scale3D=(X=1,Y=1,Z=1))),(Item=(Type=Bone,Name=\"dyn_Hair_BackL_04\"),Transform=(Rotation=(X=0,Y=0,Z=0,W=1),Translation=(X=0,Y=0,Z=0),Scale3D=(X=1,Y=1,Z ...(truncated)"
    WeightMinimum=0
    ItemToModify.0.Transform.Rotation=@n4.Result
    ItemToModify.1.Transform.Rotation=@n4.Result
    ItemToModify.2.Transform.Rotation=@n4.Result
    Weight=@Multiply_1_1_2.Result
   )
   > &RigUnit_ModifyTransforms(
    ItemToModify="((Item=(Type=Bone,Name=\"dyn_Hair_Back_07\"),Transform=(Rotation=(X=0,Y=0,Z=0,W=1),Translation=(X=0,Y=0,Z=0),Scale3D=(X=1,Y=1,Z=1))),(Item=(Type=Bone,Name=\"dyn_Hair_Back_05\"),Transform=(Rotation=(X=0,Y=0,Z=0,W=1),Translation=(X=0,Y=0,Z=0),Scale3D=(X=1,Y=1,Z= ...(truncated)"
    Weight=1
    WeightMinimum=0
    ItemToModify.1.Transform.Rotation=&RigVMFunction_MathQuaternionFromEuler(Euler=(X=0,Y=-2,Z=0), RotationOrder=YXZ).Result
    ItemToModify.0.Transform.Rotation=&RigVMFunction_MathQuaternionFromEuler(Euler=(X=0,Y=-10,Z=0), RotationOrder=YXZ).Result
   )
   Multiply_1_2_1 = RigVMAggregateNode(
    C=&RigVMFunction_MathDoubleRemap(
     SourceMaximum=1
     SourceMinimum=0
     TargetMaximum=1
     TargetMinimum=0
     bClamp=False
     Value=RigVMVariableNode() {variable=CloseGroundRemap}.Value
    ).Result
    A=RigVMVariableNode() {variable=isCapeEquipped}.Value
    B=RigVMVariableNode() {variable=isCrouching}.Value
   )
   > &RigUnit_ModifyTransforms(
    ItemToModify="((Item=(Type=Bone,Name=\"dyn_Hair_Back_05\"),Transform=(Rotation=(X=0,Y=0,Z=0,W=1),Translation=(X=0,Y=0,Z=3),Scale3D=(X=1,Y=1,Z=1))),(Item=(Type=Bone,Name=\"dyn_Hair_Back_07\"),Transform=(Rotation=(X=0,Y=0,Z=0,W=1),Translation=(X=0,Y=0,Z=2),Scale3D=(X=1,Y=1,Z= ...(truncated)"
    WeightMinimum=0
    ItemToModify.1.Transform.Rotation=&RigVMFunction_MathQuaternionFromEuler(Euler=(X=0,Y=20,Z=0), RotationOrder=YXZ).Result
    ItemToModify.0.Transform.Rotation=&RigVMFunction_MathQuaternionFromEuler(Euler=(X=0,Y=10,Z=0), RotationOrder=YXZ).Result
    Weight=@Multiply_1_2_1.Result
   )
   Multiply_1_2_1_1 = RigVMAggregateNode(A=@Multiply_1_2_1.Result, C=RigVMVariableNode() {variable=isCapeEquipped}.Value, B=RigVMVariableNode() {variable=isHeadUpDelta}.Value)
   > &RigUnit_ModifyTransforms(
    ItemToModify="((Item=(Type=Bone,Name=\"dyn_Hair_Back_05\"),Transform=(Rotation=(X=0,Y=0,Z=0,W=1),Translation=(X=0,Y=0,Z=0),Scale3D=(X=1,Y=1,Z=1))),(Item=(Type=Bone,Name=\"dyn_Hair_Back_07\"),Transform=(Rotation=(X=0,Y=0,Z=0,W=1),Translation=(X=0,Y=0,Z=0),Scale3D=(X=1,Y=1,Z= ...(truncated)"
    WeightMinimum=0
    ItemToModify.0.Transform.Rotation=&RigVMFunction_MathQuaternionFromEuler(Euler=(X=0,Y=8,Z=0), RotationOrder=YXZ).Result
    ItemToModify.1.Transform.Rotation=&RigVMFunction_MathQuaternionFromEuler(Euler=(X=0,Y=-10,Z=0), RotationOrder=YXZ).Result
    Weight=@Multiply_1_2_1_1.Result
   )
   n5 = &RigVMFunction_MathDoubleMul(
    A=&RigVMFunction_MathDoubleRemap(SourceMaximum=1, SourceMinimum=0, TargetMaximum=0, TargetMinimum=1, bClamp=False, Value=RigVMVariableNode() {variable=isGunEquipped}.Value).Result
    B=RigVMVariableNode() {variable=isCrouching}.Value
   )
   &RigUnit_ModifyTransforms(
    ItemToModify="((Item=(Type=Bone,Name=\"dyn_Hair_BackR_02\"),Transform=(Rotation=(X=0,Y=0,Z=0,W=1),Translation=(X=0,Y=0,Z=0),Scale3D=(X=1,Y=1,Z=1))),(Item=(Type=Bone,Name=\"dyn_Hair_BackR_03\"),Transform=(Rotation=(X=0,Y=0,Z=0,W=1),Translation=(X=1,Y=0,Z=1.5),Scale3D=(X=1,Y= ...(truncated)"
    WeightMinimum=0
    ItemToModify.1.Transform.Rotation=&RigVMFunction_MathQuaternionFromEuler(Euler=(X=10,Y=0,Z=0), RotationOrder=YXZ).Result
    ItemToModify.0.Transform.Rotation=&RigVMFunction_MathQuaternionFromEuler(Euler=(X=0,Y=5,Z=0), RotationOrder=YXZ).Result
    Weight=@n5.Result
    ExecutePin=@RigVMFunction_Sequence.A
   )
   > &RigUnit_ModifyTransforms(
    ItemToModify="((Item=(Type=Bone,Name=\"dyn_Hair_Back_05\"),Transform=(Rotation=(X=0,Y=0,Z=0,W=1),Translation=(X=0,Y=0,Z=-3),Scale3D=(X=1,Y=1,Z=1))),(Item=(Type=Bone,Name=\"dyn_Hair_Back_07\"),Transform=(Rotation=(X=0,Y=0,Z=0,W=1),Translation=(X=0,Y=0,Z=0),Scale3D=(X=1,Y=1,Z ...(truncated)"
    WeightMinimum=0
    ItemToModify.1.Transform.Rotation=&RigVMFunction_MathQuaternionFromEuler(Euler=(X=0,Y=-20,Z=0), RotationOrder=YXZ).Result
    ItemToModify.0.Transform.Rotation=&RigVMFunction_MathQuaternionFromEuler(Euler=(X=0,Y=-5,Z=0), RotationOrder=YXZ).Result
    Weight=@n5.Result
   )
   > &RigUnit_ModifyTransforms(
    ItemToModify="((Item=(Type=Bone,Name=\"dyn_Hair_BackR_02\"),Transform=(Rotation=(X=0,Y=0,Z=0,W=1),Translation=(X=0,Y=0,Z=1),Scale3D=(X=1,Y=1,Z=1))),(Item=(Type=Bone,Name=\"dyn_Hair_Back_05\"),Transform=(Rotation=(X=0,Y=0,Z=0,W=1),Translation=(X=0,Y=0,Z=-3),Scale3D=(X=1,Y=1, ...(truncated)"
    WeightMinimum=0
    ItemToModify.3.Transform.Rotation=&RigVMFunction_MathQuaternionFromEuler(Euler=(X=0,Y=-20,Z=0), RotationOrder=YXZ).Result
    ItemToModify.2.Transform.Rotation=&RigVMFunction_MathQuaternionFromEuler(Euler=(X=0,Y=0,Z=0), RotationOrder=YXZ).Result
    ItemToModify.0.Transform.Rotation=&RigVMFunction_MathQuaternionFromEuler(Euler=(X=0,Y=0,Z=0), RotationOrder=YXZ).Result
    Weight=&RigVMFunction_MathDoubleMul(B=RigVMVariableNode() {variable=isCrouching}.Value, A=RigVMVariableNode() {variable=isGunEquipped}.Value).Result
   )
  ]}
 New Function_ContainedGraph {nodes [
   RigVMFunctionEntryNode()
   > n1 = &RigUnit_SphericalPoseReader(
    ActiveRegionScaleFactors=(PositiveWidth=1,NegativeWidth=1,PositiveHeight=1,NegativeHeight=2)
    ActiveRegionSize=0.15
    Debug=(bDrawDebug=false,bDraw2D=False,bDrawLocalAxes=False,DebugScale=25,DebugSegments=20,DebugThickness=0.25)
    DriverAxis=(X=-1,Y=0,Z=0)
    DriverItem=(Type=Bone,Name="upperarm_r")
    FalloffRegionScaleFactors=(PositiveWidth=1,NegativeWidth=1,PositiveHeight=2,NegativeHeight=1)
    FalloffSize=0.3
    FlipHeightScaling=False
    FlipWidthScaling=False
    OptionalParentItem=(Type=Bone,Name="upperarm_r_Hold")
    RotationOffset=(X=-90,Y=20,Z=15)
   )
   > &RigUnit_ModifyTransforms(
    ItemToModify="((Item=(Type=Bone,Name=\"dyn_Hair_FrontL_01\"),Transform=(Rotation=(X=0,Y=0,Z=0.1061,W=0.9943),Translation=(X=0,Y=0.766,Z=0),Scale3D=(X=1,Y=1,Z=1))),(Item=(Type=Bone,Name=\"dyn_Hair_FrontR_02\"),Transform=(Rotation=(X=0,Y=0,Z=-0.0509,W=0.9987),Translation=(X=0 ...(truncated)"
    WeightMinimum=0
    Weight=@n1.OutputParam
   )
   > &RigUnit_ModifyTransforms(
    ItemToModify="((Item=(Type=Bone,Name=\"dyn_Hair_FrontL_01\"),Transform=(Rotation=(X=-0.0031,Y=-0.2168,Z=0.0366,W=0.9755),Translation=(X=-0.0333,Y=0.4744,Z=-0.0105),Scale3D=(X=1,Y=1,Z=1))),(Item=(Type=Bone,Name=\"dyn_Hair_FrontR_02\"),Transform=(Rotation=(X=0.0098,Y=-0.2222, ...(truncated)"
    WeightMinimum=0
    Weight=RigVMVariableNode() {variable=IsHeadDown}.Value
   )
   Multiply = &RigVMFunction_MathDoubleMul(B=RigVMVariableNode() {variable=isCrouching}.Value, A=RigVMVariableNode() {variable=isHeadUpDelta}.Value)
   > &RigUnit_ModifyTransforms(
    ItemToModify="((Item=(Type=Bone,Name=\"dyn_Hair_Back_04\"),Transform=(Rotation=(X=0,Y=-0.0753,Z=0,W=0.9972),Translation=(X=0,Y=0,Z=0),Scale3D=(X=1,Y=1,Z=1))),(Item=(Type=Bone,Name=\"dyn_Hair_BackL_04\"),Transform=(Rotation=(X=-0.0069,Y=0.017,Z=0.0749,W=0.997),Translation=(X ...(truncated)"
    WeightMinimum=0
    Weight=@Multiply.Result
   )
   > &RigUnit_ModifyTransforms(
    ItemToModify=((Item=(Type=Bone,Name="dyn_faceMask_m"),Transform=(Rotation=(X=0,Y=0,Z=0,W=1),Translation=(X=2.8805,Y=0,Z=0),Scale3D=(X=1,Y=1,Z=1))))
    WeightMinimum=0
    Weight=RigVMVariableNode() {variable=isHeadUpDelta}.Value
   )
   > RigVMFunctionReturnNode()
  ]}
 New Function_ContainedGraph {nodes [
   n1 = RigVMFunctionEntryNode()
   RigVMFunction_MathQuaternionToEuler = &RigVMFunction_MathQuaternionToEuler(
    RotationOrder=ZYX
    Value=&RigUnit_GetRelativeTransformForItem(
     Child=(Type=Bone,Name="hand_l")
     Parent=(Type=Null,Name="lowerarm_l")
     bChildInitial=False
     bParentInitial=False
     Child.Name=@n1.Child
     Parent.Name=@n1.Parent
    ).RelativeTransform.Rotation
   )
   Remap = &RigVMFunction_MathDoubleRemap(
    TargetMaximum=1
    TargetMinimum=0
    Value=@RigVMFunction_MathQuaternionToEuler.Result.X
    SourceMaximum=@n1.Source Maximum
    SourceMinimum=@n1.Source Minimum
   )
   > RigVMFunctionReturnNode(Delta X=@Remap.Result)
  ]}
 New Function_ContainedGraph {nodes [
   n1 = RigVMFunctionEntryNode()
   > DISPATCH_RigVMDispatch_ArrayIterator = RigVMDispatchNode(Array=@n1.Transform) {template="DISPATCH_RigVMDispatch_ArrayIterator(in Array,out Element,out Index,out Count,out Ratio)"}
   Multiply = &RigVMFunction_MathTransformMul(
    A=@DISPATCH_RigVMDispatch_ArrayIterator.Element.Transform
    B=&RigUnit_GetTransform(
     Item=(Type=Bone,Name="dyn_Hair_BackR_02")
     Space=LocalSpace
     bInitial=False
     Item.Name=@DISPATCH_RigVMDispatch_ArrayIterator.Element.Bone
     Item.Type=@n1.Type
    ).Transform
   ) {template="Multiply::Execute(in A,in B,out Result)"}
   > &RigUnit_SetTransform(
    Item=(Type=Bone,Name="dyn_L_hair_3")
    Space=LocalSpace
    bInitial=false
    Item.Type=@n1.Type
    Weight=@n1.Weight
    bPropagateToChildren=@n1.inheritsTransform
    Value=@Multiply.Result
    Item.Name=@DISPATCH_RigVMDispatch_ArrayIterator.Element.Bone
   )
   RigVMFunctionReturnNode(ExecuteContext=@DISPATCH_RigVMDispatch_ArrayIterator.Completed)
  ]}
 RigVMFunctionLibrary {nodes [
   AlterTransform = RigVMCollapseNode(ExecuteContext=(), Transform=(), inheritsTransform=False)
   Constraints_DEF = RigVMCollapseNode(ExecuteContext=())
   Controllers = RigVMCollapseNode(Debug Controllers=False, ExecuteContext=())
   Delta Rotation = RigVMCollapseNode(ExecuteContext=())
   Delta Rotation X = RigVMCollapseNode(ExecuteContext=())
   HairCrouch_DEF = RigVMCollapseNode(ExecuteContext=())
   HairPostFixes = RigVMCollapseNode(ExecuteContext=())
   Hair_DEF = RigVMCollapseNode(ExecuteContext=())
   Hat_DEF = RigVMCollapseNode(ExecuteContext=())
  ]}
 RigVMFunction_Sequence_1_ContainedGraph {nodes [
   n1 = RigVMFunctionEntryNode()
   RigVMFunction_Sequence = &RigVMFunction_Sequence(ExecuteContext=@n1.ExecuteContext)
   n2 = &RigVMFunction_Sequence(ExecuteContext=@RigVMFunction_Sequence.B)
   RigVMFunctionReturnNode(A=@RigVMFunction_Sequence.A, B=@n2.A, C=@n2.B)
  ]}
 RigVMFunction_Sequence_1_ContainedGraph {nodes [
   n1 = RigVMFunctionEntryNode()
   RigVMFunction_Sequence = &RigVMFunction_Sequence(ExecuteContext=@n1.ExecuteContext)
   n2 = &RigVMFunction_Sequence(ExecuteContext=@RigVMFunction_Sequence.B)
   n3 = &RigVMFunction_Sequence(ExecuteContext=@n2.B)
   RigVMFunctionReturnNode(A=@RigVMFunction_Sequence.A, B=@n2.A, C=@n3.A, D=@n3.B)
  ]}
 RigVMFunction_Sequence_1_ContainedGraph {nodes [
   n1 = RigVMFunctionEntryNode()
   RigVMFunction_Sequence = &RigVMFunction_Sequence(ExecuteContext=@n1.ExecuteContext)
   n2 = &RigVMFunction_Sequence(ExecuteContext=@RigVMFunction_Sequence.B)
   n3 = &RigVMFunction_Sequence(ExecuteContext=@n2.B)
   n4 = &RigVMFunction_Sequence(ExecuteContext=@n3.B)
   RigVMFunctionReturnNode(A=@RigVMFunction_Sequence.A, B=@n2.A, D=@n3.A, E=@n4.A, F=@n4.B)
  ]}
 RigVMFunction_Sequence_1_ContainedGraph {nodes [
   n1 = RigVMFunctionEntryNode()
   RigVMFunction_Sequence = &RigVMFunction_Sequence(ExecuteContext=@n1.ExecuteContext)
   n2 = &RigVMFunction_Sequence(ExecuteContext=@RigVMFunction_Sequence.B)
   RigVMFunctionReturnNode(A=@RigVMFunction_Sequence.A, B=@n2.A, C=@n2.B)
  ]}
 RigVMModel {nodes [
   RigUnit_BeginExecution = &RigUnit_BeginExecution()
   Controllers = RigVMFunctionReferenceNode(Debug Controllers=false, ExecuteContext=@RigUnit_BeginExecution.ExecutePin)
   > RigVMFunction_Sequence = RigVMAggregateNode()
   Delta Rotation = RigVMFunctionReferenceNode(Child=head, Parent=neck_01, Source Maximum=-43, Source Minimum=-20, ExecuteContext=@RigVMFunction_Sequence.A)
   > RigVMVariableNode(Value=@"Delta Rotation".Delta Z) {variable=isHeadUpDelta}
   > Delta Rotation_1 = RigVMFunctionReferenceNode(Child=head, Parent=neck_01, Source Maximum=-89, Source Minimum=-50)
   > RigVMVariableNode(Value=@"Delta Rotation_1".Delta Z) {variable=isHeadUpExtreme}
   > Delta Rotation_1_1 = RigVMFunctionReferenceNode(Child=head, Parent=neck_01, Source Maximum=0, Source Minimum=-20)
   > RigVMVariableNode(Value=@"Delta Rotation_1_1".Delta Z) {variable=IsHeadDown}
   > Delta Rotation X = RigVMFunctionReferenceNode(Child=head, Parent=head_null, Source Maximum=-33, Source Minimum=0)
   > RigVMVariableNode(Value=@"Delta Rotation X".Delta X) {variable=isHeadLeft}
   RigUnit_GetRelativeTransformForItem = &RigUnit_GetRelativeTransformForItem(Child=(Type=Bone,Name="pelvis"), Parent=(Type=Bone,Name="root"), bChildInitial=False, bParentInitial=False)
   Remap = &RigVMFunction_MathDoubleRemap(
    SourceMaximum=30
    SourceMinimum=60
    TargetMaximum=0
    TargetMinimum=1
    bClamp=False
    Value=@RigUnit_GetRelativeTransformForItem.RelativeTransform.Translation.Z
   )
   n1 = &RigVMFunction_AnimEvalRichCurve(
    Curve="(EditorCurveData=(Keys=((InterpMode=RCIM_Cubic,TangentMode=RCTM_Break,TangentWeightMode=RCTWM_WeightedNone,Time=0,Value=0,ArriveTangent=0,ArriveTangentWeight=0,LeaveTangent=0,LeaveTangentWeight=0),(InterpMode=RCIM_Cubic,TangentMode=RCTM_Break,TangentWeight ...(truncated)"
    SourceMinimum=0
    TargetMinimum=0
    Value=@Remap.Result
   )
   RigVMVariableNode(Value=@n1.Result, ExecuteContext=@RigVMFunction_Sequence.B) {variable=CloseGroundRemap}
   Constraints_DEF = RigVMFunctionReferenceNode(ExecuteContext=@RigVMFunction_Sequence.C)
   > Hair_DEF = RigVMFunctionReferenceNode()
   > HairPostFixes = RigVMFunctionReferenceNode()
   > AlterTransform = RigVMFunctionReferenceNode(
    Transform=((Bone="dyn_Hair_BackR_02",Transform=(Rotation=(X=0,Y=0,Z=0,W=1),Translation=(X=-2,Y=-2,Z=0),Scale3D=(X=1,Y=1,Z=1))))
    Type=Bone
    Weight=1
    inheritsTransform=False
   )
   RigVMFunction_MathBoolToFloat = &RigVMFunction_MathBoolToFloat(Value=RigVMVariableNode() {variable=isFrontEnd}.Value)
   > AlterTransform_1 = RigVMFunctionReferenceNode(
    Transform="((Bone=\"dyn_Hair_BackR_02\",Transform=(Rotation=(X=0,Y=0,Z=0,W=1),Translation=(X=0,Y=-3.5,Z=-1),Scale3D=(X=1,Y=1,Z=1))),(Bone=\"dyn_Hair_BackR_03\",Transform=(Rotation=(X=0,Y=0,Z=0,W=1),Translation=(X=0,Y=-2,Z=-3),Scale3D=(X=1,Y=1,Z=1))),(Bone=\"dyn_Hair_Front ...(truncated)"
    Type=Bone
    inheritsTransform=False
    Weight=@RigVMFunction_MathBoolToFloat.Result
   )
  ]}
 RigVMModel Construction Event Graph {nodes [
   n1 = &RigUnit_PrepareForExecution()
   RerouteNode_6_1 = RigVMDispatchNode(Values="((Type=Control,Name=\"dyn_Hair_Back_01_ctrl\"),((),Type=Control,Name=\"dyn_Hair_Back_02_ctrl\"),((),Type=Control,Name=\"dyn_Hair_Back_03_ctrl\"),((),Type=Control,Name=\"dyn_Hair_Back_04_ctrl\"),((),Type=Control,Name=\"dyn_Hair_Back_05_ctrl\"),((),Type=Control,Name=\" ...(truncated)") {template="DISPATCH_RigVMDispatch_ArrayMake(in Values,out Array)"}
   RigVMVariableNode(ExecuteContext=@n1.ExecutePin, Value=@RerouteNode_6_1.Array) {variable=dyn_Hair_Back_ctrls}
   > RigVMVariableNode(Value=/Script/ControlRig.RigUnit_HierarchyGetPoseItemArray(ElementType=All, Initial=false, ItemsToGet=@RerouteNode_6_1.Array) {is=struct RigVMUnitNode}.Pose) {variable=dyn_Hair_Back_pose}
  ]}
]
hierarchy [
 root {loc=<0 0 0>, rot=<0 0 0 1>}
 pelvis {parent=root, loc=<0 -0.5301 98.0571>, rot=<0.0127 0.707 -0.0127 -0.707>}
 spine_01 {parent=pelvis, loc=<6.1208 -0.9328 0>, rot=<0 0 0.034 -0.9994>}
 spine_02 {parent=spine_01, loc=<6.5996 -0.2095 0>, rot=<0 0 0.0238 -0.9997>}
 spine_03 {parent=spine_02, loc=<6.4914 -0.1237 0>, rot=<0 0 0.0149 -0.9999>}
 spine_04 {parent=spine_03, loc=<6.9598 0.0794 0>, rot=<0 0 -0.0132 -0.9999>}
 spine_05 {parent=spine_04, loc=<6.9987 0 0>, rot=<0 0 -0.0015 -1>}
 neck_01 {parent=spine_05, loc=<11.3138 0.1529 0>, rot=<0 0 -0.2003 -0.9797>}
 neck_02 {parent=neck_01, loc=<4.5627 0 0>, rot=<0 0 0.0504 -0.9987>}
 head {parent=neck_02, loc=<5.3914 0 0>, rot=<0 0 0.115 -0.9934>}
 base_Hair_BackL {parent=head, loc=<8.1857 -0.439 -5.0959>, rot=<0.3093 -0.6359 0.6359 -0.3093>}
 dyn_Hair_BackL_01 {parent=base_Hair_BackL, loc=<0 0 0>, rot=<0 0.0178 0 -0.9998>}
 dyn_Hair_BackL_02 {parent=dyn_Hair_BackL_01, loc=<4.4301 0 0>, rot=<-0.0017 -0.3721 -0.0041 -0.9282>}
 dyn_Hair_BackL_03 {parent=dyn_Hair_BackL_02, loc=<6.799 0 0>, rot=<0.0015 0.1503 -0.0011 -0.9886>}
 dyn_Hair_BackL_04 {parent=dyn_Hair_BackL_03, loc=<9.2764 0 0>, rot=<0 -0.008 0.0055 -1>}
 dyn_Hair_BackL_05 {parent=dyn_Hair_BackL_04, loc=<7.291 0 0>, rot=<0 -0.0709 0 -0.9975>}
 dyn_Hair_BackL_06 {parent=dyn_Hair_BackL_05, loc=<7.4303 0 0>, rot=<0 0 0 1>}
 base_Hair_Back {parent=head, loc=<8.5016 -1.2653 -0.3258>, rot=<0.3093 -0.6359 0.6359 -0.3093>}
 dyn_Hair_Back_01 {parent=base_Hair_Back, loc=<0 0 0>, rot=<0 0 0 1>}
 dyn_Hair_Back_02 {parent=dyn_Hair_Back_01, loc=<4.2813 0 0>, rot=<0 -0.3651 0 -0.931>}
 dyn_Hair_Back_03 {parent=dyn_Hair_Back_02, loc=<6.9171 0 0>, rot=<0.0034 0.1783 -0.019 -0.9838>}
 dyn_Hair_Back_04 {parent=dyn_Hair_Back_03, loc=<9.1968 0 0>, rot=<-0.0006 -0.0313 0.0193 -0.9993>}
 dyn_Hair_Back_05 {parent=dyn_Hair_Back_04, loc=<10.1314 0 0>, rot=<0 -0.145 0 -0.9894>}
 dyn_Hair_Back_06 {parent=dyn_Hair_Back_05, loc=<7.2685 0 0>, rot=<0 -0.1403 0 -0.9901>}
 dyn_Hair_Back_07 {parent=dyn_Hair_Back_06, loc=<6.2276 0 0>, rot=<0 0 0 1>}
 base_Hair_BackR {parent=head, loc=<8.1854 -0.439 5.0959>, rot=<0.3093 -0.6359 0.6359 -0.3093>}
 dyn_Hair_BackR_01 {parent=base_Hair_BackR, loc=<0 0 0>, rot=<0 0 0 1>}
 dyn_Hair_BackR_02 {parent=dyn_Hair_BackR_01, loc=<4.4274 0 0.1575>, rot=<0.0018 -0.3557 0.0046 -0.9346>}
 dyn_Hair_BackR_03 {parent=dyn_Hair_BackR_02, loc=<6.7978 0 0>, rot=<-0.0013 0.1616 -0.0016 -0.9869>}
 dyn_Hair_BackR_04 {parent=dyn_Hair_BackR_03, loc=<9.2816 0 0>, rot=<0 -0.0081 -0.0043 -1>}
 dyn_Hair_BackR_05 {parent=dyn_Hair_BackR_04, loc=<7.3337 0 0>, rot=<-0.0002 -0.0771 -0.0003 -0.997>}
 dyn_Hair_BackR_06 {parent=dyn_Hair_BackR_05, loc=<7.42 0 0>, rot=<0 0 0 1>}
 base_Hair_FrontL {parent=head, loc=<-0.9358 4.0904 -7.1705>, rot=<-0.0287 -0.7065 -0.7036 0.0708>}
 dyn_Hair_FrontL_01 {parent=base_Hair_FrontL, loc=<0 0 0>, rot=<0 0 0 1>}
 dyn_Hair_FrontL_02 {parent=dyn_Hair_FrontL_01, loc=<7.6713 0 0>, rot=<0.0237 0.1633 -0.0029 -0.9863>}
 dyn_Hair_FrontL_03 {parent=dyn_Hair_FrontL_02, loc=<6.8026 0 0>, rot=<0 0 0.0764 -0.9971>}
 base_Hair_FrontR {parent=head, loc=<-1.8 3.5677 6.7148>, rot=<0.0414 -0.7059 -0.7045 -0.0607>}
 dyn_Hair_FrontR_02 {parent=base_Hair_FrontR, loc=<0 0 0>, rot=<0 0 0 1>}
 dyn_Hair_FrontR_03 {parent=dyn_Hair_FrontR_02, loc=<7.7689 0 0>, rot=<-0.0166 0.1493 -0.0336 -0.9881>}
 dyn_Hair_FrontR_04 {parent=dyn_Hair_FrontR_03, loc=<6.7572 0 0>, rot=<0 0 0 1>}
 base_Hair_SideL {parent=head, loc=<2.5939 2.8551 -9.1679>, rot=<-0.1606 0.6774 0.7163 -0.0473>}
 dyn_Hair_SideL_01 {parent=base_Hair_SideL, loc=<0 0 0>, rot=<0 0 0 1>}
 dyn_Hair_SideL_02 {parent=dyn_Hair_SideL_01, loc=<4.1966 0 0>, rot=<0 0 0 1>}
 base_Hair_SideR {parent=head, loc=<2.5333 3.1682 9.0467>, rot=<0.0427 -0.7085 -0.6955 0.1115>}
 dyn_Hair_SideR_01 {parent=base_Hair_SideR, loc=<0 0 0>, rot=<0 0 0 1>}
 dyn_Hair_SideR_02 {parent=dyn_Hair_SideR_01, loc=<3.1682 0 0>, rot=<0 0 0 1>}
 base_HairStrand_front_01 {parent=head, loc=<10.0656 9.7295 3.4072>, rot=<0.4289 0.0088 0.9031 -0.0185>}
 dyn_HairStrand_front_01 {parent=base_HairStrand_front_01, loc=<0 0 0>, rot=<0 0 0 1>}
 dyn_HairStrand_front_02 {parent=dyn_HairStrand_front_01, loc=<2.9007 0 0>, rot=<-0.0093 -0.2359 -0.0123 -0.9717>}
 dyn_HairStrand_front_03 {parent=dyn_HairStrand_front_02, loc=<3.1593 0 0>, rot=<-0.0086 -0.1288 -0.0297 -0.9912>}
 dyn_HairStrand_front_04 {parent=dyn_HairStrand_front_03, loc=<3.4222 0 0>, rot=<0 0 0 1>}
 base_faceMask_m {parent=head, loc=<-6.2276 4.0471 0.0003>, rot=<-0.0029 0.7071 0.7071 -0.0029>}
 dyn_faceMask_m {parent=base_faceMask_m, loc=<0 0 0>, rot=<0 0 0 1>}
 head_null {parent=neck_02, type=Null, loc=<5.3914 0 0>, rot=<0 0 0.115 -0.9934>}
 clavicle_l {parent=spine_05, loc=<5.4831 -0.1053 -3.2946>, rot=<0.0465 -0.7289 -0.009 -0.6829>}
 upperarm_l {parent=clavicle_l, loc=<9.8658 -1.2136 -0.4105>, rot=<0.0164 -0.3245 -0.0269 -0.9454>}
 upperarm_twist_01_l {parent=upperarm_l, loc=<7.2135 -0.1134 0>, rot=<0 0 0 1>}
 deform_upperarm_bk_base_l {parent=upperarm_twist_01_l, loc=<-6.4053 -4.3746 0.314>, rot=<0 0 0 -1>}
 deform_upperarm_fr_base_l {parent=upperarm_twist_01_l, loc=<-5.6312 5.092 0.3354>, rot=<0 0 0 -1>}
 deform_upperarm_up_base_l {parent=upperarm_twist_01_l, loc=<-8.891 0.4825 3.9517>, rot=<0 0 0 -1>}
 base_shoulderPad_l {parent=upperarm_twist_01_l, loc=<-3.7746 -0.9552 6.5872>, rot=<0 0.107 0 -0.9943>}
 clavicle_r {parent=spine_05, loc=<5.4831 -0.1053 3.2946>, rot=<0.7289 0.0465 -0.6829 0.009>}
 upperarm_r {parent=clavicle_r, loc=<-9.8658 1.2136 0.4105>, rot=<0.0164 -0.3245 -0.0269 -0.9454>}
 upperarm_twist_01_r {parent=upperarm_r, loc=<-7.2135 0.1134 0>, rot=<0 0 0 -1>}
 deform_upperarm_bk_base_r {parent=upperarm_twist_01_r, loc=<6.4053 4.3746 -0.314>, rot=<0 0 0 -1>}
 deform_upperarm_fr_base_r {parent=upperarm_twist_01_r, loc=<5.6312 -5.092 -0.3354>, rot=<0 0 0 -1>}
 deform_upperarm_up_base_r {parent=upperarm_twist_01_r, loc=<8.891 -0.4825 -3.9517>, rot=<0 0 0 -1>}
 base_shoulderPad_r {parent=upperarm_twist_01_r, loc=<3.7743 0.9552 -6.5868>, rot=<0 0.107 0 -0.9943>}
 base_collar_bck {parent=spine_05, loc=<12.0115 -4.2544 0>, rot=<0.1953 -0.6796 0.6796 -0.1953>}
 base_collar_side_l {parent=spine_05, loc=<9.8352 0.3233 -7.3207>, rot=<0.2772 -0.9398 0.1798 -0.0871>}
 base_collar_fr_l {parent=spine_05, loc=<5.0358 7.2168 -6.8516>, rot=<0.4115 -0.9107 -0.0337 -0.012>}
 base_collar_side_r {parent=spine_05, loc=<9.8347 0.3233 7.3207>, rot=<0.0871 -0.1798 0.9398 -0.2772>}
 base_collar_fr_r {parent=spine_05, loc=<5.0355 7.2168 6.8516>, rot=<0.012 0.0337 0.9107 -0.4115>}
 base_jacket_side_l {parent=spine_04, loc=<-1.4419 -4.4831 -10.5024>, rot=<0.0388 -0.9324 0.3591 -0.015>}
 dyn_jacket_side_l {parent=base_jacket_side_l, loc=<0 0 0>, rot=<0 0 0 1>}
 base_jacket_bck {parent=spine_04, loc=<-0.8516 -8.4268 0>, rot=<0.0294 -0.7065 0.7065 -0.0294>}
 dyn_jacket_bck {parent=base_jacket_bck, loc=<0 0 0>, rot=<0 0 0 1>}
 base_jacket_side_r {parent=spine_04, loc=<-1.4254 -4.6813 10.7223>, rot=<0.9324 0.0388 -0.015 -0.3591>}
 dyn_jacket_side_r {parent=base_jacket_side_r, loc=<0 0 0>, rot=<0 0 0 1>}
 dyn_Hair_BackL_02_Null {parent=pelvis, type=Null, loc=<59.0305 -7.7237 -5.5286>, rot=<0 0 -0.6508 0.7593>}
 dyn_Hair_BackL_02_ctrl {controlType=EULER_TRANSFORM, parent=dyn_Hair_BackL_02_Null, type=Control, loc=<0 0 0>, rot=<0 0 0 1>}
 dyn_Hair_BackL_03_Null {parent=pelvis, type=Null, loc=<52.158 -8.5088 -5.5286>, rot=<0 0 0.6478 -0.7618>}
 dyn_Hair_BackL_03_ctrl {controlType=EULER_TRANSFORM, parent=dyn_Hair_BackL_03_Null, type=Control, loc=<0 0 0>, rot=<0 0 0 1>}
 dyn_Hair_BackL_04_Null {parent=pelvis, type=Null, loc=<43.581 -11.6175 -7.8636>, rot=<0 0 -0.6478 0.7618>}
 dyn_Hair_BackL_04_ctrl {controlType=EULER_TRANSFORM, parent=dyn_Hair_BackL_04_Null, type=Control, loc=<0 0 0>, rot=<0 0 0 1>}
 dyn_Hair_BackL_05_Null {parent=pelvis, type=Null, loc=<43.581 -11.6175 -7.8636>, rot=<0 0 -0.6478 0.7618>}
 dyn_Hair_BackL_05_ctrl {controlType=EULER_TRANSFORM, parent=dyn_Hair_BackL_05_Null, type=Control, loc=<0 0 0>, rot=<0 0 0 1>}
 dyn_Hair_BackL_06_Null {parent=pelvis, type=Null, loc=<43.581 -11.6175 -7.8636>, rot=<0 0 -0.6478 0.7618>}
 dyn_Hair_BackL_06_ctrl {controlType=EULER_TRANSFORM, parent=dyn_Hair_BackL_06_Null, type=Control, loc=<0 0 0>, rot=<0 0 0 1>}
 dyn_Hair_BackL_01_Null {parent=pelvis, type=Null, loc=<61.8198 -4.4758 -5.5286>, rot=<0 0 0.1148 -0.9934>}
 dyn_Hair_BackL_01_ctrl {controlType=EULER_TRANSFORM, parent=dyn_Hair_BackL_01_Null, type=Control, loc=<0 0 0>, rot=<0 0 0 1>}
 dyn_Hair_Back_01_Null {parent=pelvis, type=Null, loc=<62.3385 -4.7288 -0.3258>, rot=<0 0 0.1148 -0.9934>}
 dyn_Hair_Back_01_ctrl {controlType=EULER_TRANSFORM, parent=dyn_Hair_Back_01_Null, type=Control, loc=<0 0 0>, rot=<0 0 0 1>}
 dyn_Hair_Back_02_Null {parent=pelvis, type=Null, loc=<59.5492 -7.9768 -0.3258>, rot=<0 0 -0.6508 0.7593>}
 dyn_Hair_Back_02_ctrl {controlType=EULER_TRANSFORM, parent=dyn_Hair_Back_02_Null, type=Control, loc=<0 0 0>, rot=<0 0 0 1>}
 dyn_Hair_Back_03_Null {parent=pelvis, type=Null, loc=<52.6768 -8.7618 -0.3258>, rot=<0 0 0.6478 -0.7618>}
 dyn_Hair_Back_03_ctrl {controlType=EULER_TRANSFORM, parent=dyn_Hair_Back_03_Null, type=Control, loc=<0 0 0>, rot=<0 0 0 1>}
 dyn_Hair_Back_04_Null {parent=pelvis, type=Null, loc=<44.2983 -12.3291 -0.6812>, rot=<0 0 -0.6478 0.7618>}
 dyn_Hair_Back_04_ctrl {controlType=EULER_TRANSFORM, parent=dyn_Hair_Back_04_Null, type=Control, loc=<0 0 0>, rot=<0 0 0 1>}
 dyn_Hair_Back_05_Null {parent=pelvis, type=Null, loc=<35.1845 -16.4827 -0.6812>, rot=<0 0 0.6478 -0.7618>}
 dyn_Hair_Back_05_ctrl {controlType=EULER_TRANSFORM, parent=dyn_Hair_Back_05_Null, type=Control, loc=<0 0 0>, rot=<0 0 0 1>}
 dyn_Hair_Back_06_Null {parent=pelvis, type=Null, loc=<27.9468 -16.7205 -0.6812>, rot=<0 0 -0.6935 0.7204>}
 dyn_Hair_Back_06_ctrl {controlType=EULER_TRANSFORM, parent=dyn_Hair_Back_06_Null, type=Control, loc=<0 0 0>, rot=<0 0 0 1>}
 dyn_Hair_Back_07_null {parent=pelvis, type=Null, loc=<21.2661 -15.4804 -0.6672>, rot=<0.2803 0.6483 -0.7038 0.0761>}
 dyn_Hair_Back_07_ctrl {controlType=EULER_TRANSFORM, parent=dyn_Hair_Back_07_null, type=Control, loc=<0 0 0>, rot=<0 0 0 1>}
 dyn_Hair_BackR_01_Null {parent=pelvis, type=Null, loc=<62.0591 -3.8895 5.0959>, rot=<0 0 0.1148 -0.9934>}
 dyn_Hair_BackR_01_ctrl {controlType=EULER_TRANSFORM, parent=dyn_Hair_BackR_01_Null, type=Control, loc=<0 0 0>, rot=<0 0 0 1>}
 dyn_Hair_BackR_02_Null {parent=pelvis, type=Null, loc=<59.2698 -7.1374 5.0959>, rot=<0 0 -0.6508 0.7593>}
 dyn_Hair_BackR_02_ctrl {controlType=EULER_TRANSFORM, parent=dyn_Hair_BackR_02_Null, type=Control, loc=<0 0 0>, rot=<0 0 0 1>}
 dyn_Hair_BackR_03_Null {parent=pelvis, type=Null, loc=<52.3973 -7.9224 5.0959>, rot=<0 0 0.6478 -0.7618>}
 dyn_Hair_BackR_03_ctrl {controlType=EULER_TRANSFORM, parent=dyn_Hair_BackR_03_Null, type=Control, loc=<0 0 0>, rot=<0 0 0 1>}
 dyn_Hair_BackR_04_Null {parent=pelvis, type=Null, loc=<43.9889 -11.8327 5.7125>, rot=<0 0 -0.6478 0.7618>}
 dyn_Hair_BackR_04_ctrl {controlType=EULER_TRANSFORM, parent=dyn_Hair_BackR_04_Null, type=Control, loc=<0 0 0>, rot=<0 0 0 1>}
 dyn_Hair_BackR_05_Null {parent=pelvis, type=Null, loc=<37.3308 -14.8048 5.7125>, rot=<0 0 0.6478 -0.7618>}
 dyn_Hair_BackR_05_ctrl {controlType=EULER_TRANSFORM, parent=dyn_Hair_BackR_05_Null, type=Control, loc=<0 0 0>, rot=<0 0 0 1>}
 dyn_Hair_BackR_06_Null {parent=pelvis, type=Null, loc=<30.0547 -16.1126 5.7125>, rot=<0 0 -0.6935 0.7204>}
 dyn_Hair_BackR_06_ctrl {controlType=EULER_TRANSFORM, parent=dyn_Hair_BackR_06_Null, type=Control, loc=<0 0 0>, rot=<0 0 0 1>}
 head_ctrl {controlType=EULER_TRANSFORM, type=Control, loc=<0 0 0>, rot=<0 0 0 1>}
 upperarm_l_ctrl {controlType=EULER_TRANSFORM, type=Control, loc=<0 0 0>, rot=<0 0 0 1>}
 upperarm_r_ctrl {controlType=EULER_TRANSFORM, type=Control, loc=<0 0 0>, rot=<0 0 0 1>}
]
variables [
 CloseGroundRemap {type=double, default=0}
 IsHeadDown {type=double, default=0}
 dyn_Hair_BackL_ctrls {
  type="TArray<FRigElementKey>"
  default=((Type=Control,Name="dyn_Hair_BackL_02"),(Type=Control,Name="dyn_Hair_BackL_03"),(Type=Control,Name="dyn_Hair_BackL_04"),(Type=Control,Name="dyn_Hair_BackL_05"),(Type=Control,Name="dyn_Hair_BackL_06"))
 }
 dyn_Hair_BackL_pose {type=FRigPose, default=(Elements=,HierarchyTopologyVersion=-1,PoseHash=-1)}
 dyn_Hair_BackR_ctrls {
  type="TArray<FRigElementKey>"
  default=((Type=Control,Name="dfrm_arm_spline_01_r_ctrl"),(Type=Control,Name="dfrm_arm_spline_03_r_ctrl"),(Type=Control,Name="dfrm_arm_spline_05_r_ctrl"),(Type=Control,Name="dfrm_arm_spline_07_r_ctrl"),(Type=Control,Name="dfrm_arm_spline_09_r_ctrl"))
 }
 dyn_Hair_BackR_pose {type=FRigPose, default=(Elements=,HierarchyTopologyVersion=-1,PoseHash=-1)}
 dyn_Hair_Back_ctrls {
  type="TArray<FRigElementKey>"
  default=((Type=Control,Name="dyn_Hair_BackL_02"),(Type=Control,Name="dyn_Hair_BackL_03"),(Type=Control,Name="dyn_Hair_BackL_04"),(Type=Control,Name="dyn_Hair_BackL_05"),(Type=Control,Name="dyn_Hair_BackL_06"))
 }
 dyn_Hair_Back_pose {type=FRigPose, default=(Elements=,HierarchyTopologyVersion=-1,PoseHash=-1)}
 isCapeEquipped {type=double, default=0}
 isCrouching {type=double, default=0}
 isFrontEnd {type=bool, default=False}
 isGunEquipped {type=double, default=0}
 isHangGliding {type=double, default=0}
 isHeadLeft {type=double, default=0}
 isHeadUpDelta {type=double, default=0}
 isHeadUpExtreme {type=double, default=0}
 isSkydiveFloating {type=double, default=0}
 isTargeting {type=double, default=0}
]
